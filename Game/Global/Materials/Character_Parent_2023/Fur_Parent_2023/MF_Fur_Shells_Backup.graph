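# Material function, apparently a backup, that layers shell-based fur onto incoming character material attributes, also outputting fur length.
# Mask and flow maps control fur length, clothing flattening and world position offset along the shells.
# Pattern textures drive opacity and normals, alongside root color, anisotropy and backlit fresnel.
# FOV-scaled distance blends in far fur and thins shells by power-of-two index modulo, with NaN offsets discarding culled geometry.
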MaterialFunction /Game/Global/Materials/Character_Parent_2023/Fur_Parent_2023/MF_Fur_Shells_Backup
graph {expressions {
  MaterialExpressionAdd_0 {
   class=MaterialExpressionAdd
   inputs [
    A {from=MaterialExpressionNamedRerouteUsage_89}
    B {from=""}
   ]
   properties {constB=0.1}
  }
  MaterialExpressionAdd_1 {
   class=MaterialExpressionAdd
   inputs [
    A {from=MaterialExpressionNamedRerouteUsage_19}
    B {from=MaterialExpressionNamedRerouteUsage_20}
   ]
  }
  MaterialExpressionAdd_11 {
   class=MaterialExpressionAdd
   inputs [
    A {from=MaterialExpressionScalarParameter_37}
    B {from=MaterialExpressionScalarParameter_38}
   ]
  }
  MaterialExpressionAdd_12 {
   class=MaterialExpressionAdd
   inputs [
    A {from=MaterialExpressionStaticSwitchParameter_2}
    B {from=MaterialExpressionMultiply_27}
   ]
  }
  MaterialExpressionAdd_2 {
   class=MaterialExpressionAdd
   inputs [
    A {from=MaterialExpressionNamedRerouteUsage_45}
    B {from=MaterialExpressionMultiply_19}
   ]
  }
  MaterialExpressionAdd_49 {
   class=MaterialExpressionAdd
   inputs [
    A {from=MaterialExpressionAppendVector_34}
    B {from=MaterialExpressionTextureCoordinate_6}
   ]
  }
  MaterialExpressionAdd_6 {
   class=MaterialExpressionAdd
   inputs [
    A {from=MaterialExpressionNamedRerouteUsage_175}
    B {from=MaterialExpressionMultiply_11}
   ]
  }
  MaterialExpressionAdd_8 {
   class=MaterialExpressionAdd
   inputs [
    A {from=MaterialExpressionScalarParameter_40}
    B {from=MaterialExpressionScalarParameter_41}
   ]
  }
  MaterialExpressionAdd_9 {
   class=MaterialExpressionAdd
   inputs [
    A {from=MaterialExpressionScalarParameter_30}
    B {from=MaterialExpressionScalarParameter_32}
   ]
  }
  MaterialExpressionAppendVector_0 {
   class=MaterialExpressionAppendVector
   inputs [
    A {from=MaterialExpressionScalarParameter_35, output=""}
    B {from=MaterialExpressionScalarParameter_35, output=""}
   ]
  }
  MaterialExpressionAppendVector_1 {
   class=MaterialExpressionAppendVector
   inputs [
    A {from=MaterialExpressionAppendVector_0, output=""}
    B {from=MaterialExpressionConstant_2, output=""}
   ]
  }
  MaterialExpressionAppendVector_27 {
   class=MaterialExpressionAppendVector
   inputs [
    A {from=MaterialExpressionScalarParameter_137, output=""}
    B {from=MaterialExpressionScalarParameter_137, output=""}
   ]
  }
  MaterialExpressionAppendVector_28 {
   class=MaterialExpressionAppendVector
   inputs [
    A {from=MaterialExpressionAppendVector_27, output=""}
    B {from=MaterialExpressionConstant_12, output=""}
   ]
  }
  MaterialExpressionAppendVector_34 {
   class=MaterialExpressionAppendVector
   inputs [
    A {from=MaterialExpressionScalarParameter_163, output=""}
    B {from=MaterialExpressionScalarParameter_164, output=""}
   ]
  }
  MaterialExpressionAtmosphericLightVector_5 {class=MaterialExpressionAtmosphericLightVector, inputs [], properties {}}
  MaterialExpressionCeil_0 {class=MaterialExpressionCeil, inputs [None {from=MaterialExpressionLogarithm2_0, output=""}], properties {}}
  MaterialExpressionChannelMaskParameter_1 {
   class=MaterialExpressionChannelMaskParameter
   inputs [None {from=MaterialExpressionTextureSampleParameter2D_20, output=RGB}]
   properties {group=Fur_Shell_01 - Base Settings, maskChannel=Green, parameterName=FurLegnthChannel, sortPriority=1, defaultValue=<rgba 0 1 0 0>}
  }
  MaterialExpressionChannelMaskParameter_2 {
   class=MaterialExpressionChannelMaskParameter
   inputs [None {from=MaterialExpressionTextureSampleParameter2D_20, output=RGB}]
   properties {group=Fur_Shell_01 - Base Settings, parameterName=FurNormalMaskChannel, sortPriority=2}
  }
  MaterialExpressionChannelMaskParameter_3 {
   class=MaterialExpressionChannelMaskParameter
   inputs [None {from=MaterialExpressionTextureSampleParameter2D_20, output=RGB}]
   properties {group=Fur_Shell_01 - Base Settings, maskChannel=Blue, parameterName=FurClothingMaskChannel, sortPriority=2, defaultValue=<rgba 0 0 1 0>}
  }
  MaterialExpressionChannelMaskParameter_4 {
   class=MaterialExpressionChannelMaskParameter
   inputs [None {from=MaterialExpressionFunctionInput_7, output=""}]
   properties {
    desc=Coming from the TA mask. Determines what area of the mesh should have fur.
    group=Fur_Shell_01 - Base Settings
    parameterName=ShellFurMaskChannel
    sortPriority=-1
   }
  }
  MaterialExpressionClamp_0 {
   class=MaterialExpressionClamp
   inputs [
    None {from=MaterialExpressionNamedRerouteDeclaration_24, output=""}
    Min {from="", output=""}
    Max {from="", output=""}
   ]
   properties {maxDefault=0.975}
  }
  MaterialExpressionComponentMask_0 {
   class=MaterialExpressionComponentMask
   inputs [None {from=MaterialExpressionTextureSampleParameter2D_21, output=RGB}]
   properties {g=true, r=true}
  }
  MaterialExpressionConstant3Vector_0 {class=MaterialExpressionConstant3Vector, properties {constant=<rgba 0 1 0 0>}}
  MaterialExpressionConstant3Vector_1 {class=MaterialExpressionConstant3Vector, properties {constant=<rgba 0 0 1 0>}}
  MaterialExpressionConstant3Vector_2 {class=MaterialExpressionConstant3Vector, properties {constant=<rgba 1 0 0 0>}}
  MaterialExpressionConstant_0 {class=MaterialExpressionConstant}
  MaterialExpressionConstant_1 {class=MaterialExpressionConstant}
  MaterialExpressionConstant_12 {class=MaterialExpressionConstant, properties {r=1}}
  MaterialExpressionConstant_2 {class=MaterialExpressionConstant, properties {r=1}}
  MaterialExpressionConstant_30 {class=MaterialExpressionConstant, properties {r=0.7972}}
  MaterialExpressionConstant_31 {class=MaterialExpressionConstant, properties {r=1}}
  MaterialExpressionConstant_32 {class=MaterialExpressionConstant, properties {r=1.5}}
  MaterialExpressionConstant_33 {class=MaterialExpressionConstant, properties {r=0.01}}
  MaterialExpressionConstant_34 {class=MaterialExpressionConstant, properties {r=1}}
  MaterialExpressionConstant_4 {class=MaterialExpressionConstant, properties {r=2}}
  MaterialExpressionConstant_5 {class=MaterialExpressionConstant, properties {r=1}}
  MaterialExpressionConvert_3 {
   class=MaterialExpressionConvert
   inputs [None {from=MaterialExpressionNamedRerouteUsage_131}]
   properties {
    nodeName=Break Float3
    convertInputs [{
      type=Vector3
      expressionInput {
       expression=&self:MaterialExpressionNamedRerouteUsage_131
       inputName=None
       mask=0
       maskA=0
       maskB=0
       maskG=0
       maskR=0
       outputIndex=0
      }
     }]
    convertOutputs [{type=Scalar}, {type=Scalar}, {type=Scalar}]
   }
  }
  MaterialExpressionConvert_4 {
   class=MaterialExpressionConvert
   inputs [
    None {from=MaterialExpressionMultiply_2}
    None {from=MaterialExpressionConvert_3}
    None {from=MaterialExpressionConvert_3}
   ]
   properties {
    nodeName=Make Float3
    convertInputs [
     {
      type=Scalar
      expressionInput {
       expression=&self:MaterialExpressionMultiply_2
       inputName=None
       mask=0
       maskA=0
       maskB=0
       maskG=0
       maskR=0
       outputIndex=0
      }
     }
     {
      type=Scalar
      expressionInput {
       expression=&self:MaterialExpressionConvert_3
       inputName=None
       mask=0
       maskA=0
       maskB=0
       maskG=0
       maskR=0
       outputIndex=1
      }
     }
     {
      type=Scalar
      expressionInput {
       expression=&self:MaterialExpressionConvert_3
       inputName=None
       mask=0
       maskA=0
       maskB=0
       maskG=0
       maskR=0
       outputIndex=2
      }
     }
    ]
    convertOutputs [{type=Vector3}]
   }
  }
  MaterialExpressionCustom_0 {class=MaterialExpressionCustom, inputs [], properties {code="const float fNaN = 0.0f / 0.0f;\r\nreturn fNaN;", showCode=true, inputs []}}
  MaterialExpressionDeriveNormalZ_1 {class=MaterialExpressionDeriveNormalZ, inputs [InXY {from=MaterialExpressionSubtract_2, output=""}], properties {}}
  MaterialExpressionDivide_0 {
   class=MaterialExpressionDivide
   inputs [
    A {from=MaterialExpressionSubtract_0}
    B {from=MaterialExpressionOneMinus_7}
   ]
  }
  MaterialExpressionDivide_4 {
   class=MaterialExpressionDivide
   inputs [
    A {from=MaterialExpressionNamedRerouteUsage_64}
    B {from=MaterialExpressionNamedRerouteUsage_63}
   ]
  }
  MaterialExpressionFmod_0 {
   class=MaterialExpressionFmod
   inputs [
    A {from=ShellMeshMaterialExpressionShellLayerIndex_2, output=""}
    B {from=MaterialExpressionPower_0, output=""}
   ]
   properties {}
  }
  MaterialExpressionFresnel_2 {
   class=MaterialExpressionFresnel
   inputs [
    ExponentIn {from=MaterialExpressionConstant_32, output=""}
    BaseReflectFractionIn {from=MaterialExpressionConstant_33, output=""}
    Normal {from="", output=""}
   ]
   properties {baseReflectFraction=0.01, exponent=9}
  }
  MaterialExpressionFunctionInput_6 {class=MaterialExpressionFunctionInput, inputs [Preview {from="", output=""}], properties {inputType=FunctionInput_MaterialAttributes}}
  MaterialExpressionFunctionInput_7 {
   class=MaterialExpressionFunctionInput
   inputs [Preview {from="", output=""}]
   properties {bUsePreviewValueAsDefault=true, inputName=TechArtMask, sortPriority=1, previewValue=<1 1 1 1>}
  }
  MaterialExpressionFunctionOutput_0 {
   class=MaterialExpressionFunctionOutput
   inputs [None {from=MaterialExpressionNamedRerouteDeclaration_82, output=""}]
   properties {outputName=Fur Length, sortPriority=1}
  }
  MaterialExpressionFunctionOutput_3 {class=MaterialExpressionFunctionOutput, inputs [None {from=MaterialExpressionStaticSwitchParameter_1, output=""}]}
  MaterialExpressionGetMaterialAttributes_3 {
   class=MaterialExpressionGetMaterialAttributes
   inputs [None {from=MaterialExpressionNamedRerouteDeclaration_13, output=""}]
   properties {attributeGetTypes [
     #1
     #2
     #3
     #4
    ]}
  }
  MaterialExpressionIf_0 {
   class=MaterialExpressionIf
   inputs [
    A {from=MaterialExpressionVertexInterpolator_1, output=PS}
    B {from=MaterialExpressionScalarParameter_2, output=""}
    "A > B" {from=MaterialExpressionConstant_0, output=""}
    "A == B" {from=MaterialExpressionConstant_0, output=""}
    "A < B" {from=MaterialExpressionNamedRerouteUsage_27, output=""}
   ]
   properties {}
  }
  MaterialExpressionIf_1 {
   class=MaterialExpressionIf
   inputs [
    A {from=MaterialExpressionNamedRerouteUsage_75, output=""}
    B {from="", output=""}
    "A > B" {from=MaterialExpressionCustom_0, output=""}
    "A == B" {from=MaterialExpressionMultiply_6, output=""}
    "A < B" {from=MaterialExpressionMultiply_6, output=""}
   ]
   properties {constB=0.5}
  }
  MaterialExpressionIf_2 {
   class=MaterialExpressionIf
   inputs [
    A {from=ShellMeshMaterialExpressionShellCount_0, output=""}
    B {from=MaterialExpressionScalarParameter_43, output=""}
    "A > B" {from=MaterialExpressionMultiply_16, output=""}
    "A == B" {from=MaterialExpressionConstant_1, output=""}
    "A < B" {from=MaterialExpressionConstant_1, output=""}
   ]
   properties {}
  }
  MaterialExpressionInverseLinearInterpolate_0 {
   class=MaterialExpressionInverseLinearInterpolate
   inputs [
    A {from=MaterialExpressionScalarParameter_30, output=""}
    B {from=MaterialExpressionAdd_9, output=""}
    Value {from=MaterialExpressionNamedRerouteUsage_26, output=""}
   ]
   properties {constB=0}
  }
  MaterialExpressionInverseLinearInterpolate_1 {
   class=MaterialExpressionInverseLinearInterpolate
   inputs [
    A {from=MaterialExpressionScalarParameter_40, output=""}
    B {from=MaterialExpressionAdd_8, output=""}
    Value {from=MaterialExpressionNamedRerouteDeclaration_31, output=""}
   ]
   properties {constB=0}
  }
  MaterialExpressionInverseLinearInterpolate_2 {
   class=MaterialExpressionInverseLinearInterpolate
   inputs [
    A {from=MaterialExpressionScalarParameter_37, output=""}
    B {from=MaterialExpressionAdd_11, output=""}
    Value {from=MaterialExpressionNamedRerouteUsage_55, output=""}
   ]
   properties {constB=0}
  }
  MaterialExpressionInverseLinearInterpolate_3 {
   class=MaterialExpressionInverseLinearInterpolate
   inputs [
    A {from="", output=""}
    B {from="", output=""}
    Value {from=MaterialExpressionMaterialFunctionCall_1, output=FOV angle in degrees}
   ]
   properties {constA=45, constB=90}
  }
  MaterialExpressionLength_0 {class=MaterialExpressionLength, inputs [None {from=MaterialExpressionWorldPosition_0, output=XYZ}], properties {}}
  MaterialExpressionLinearInterpolate_0 {
   class=MaterialExpressionLinearInterpolate
   inputs [
    A {from=MaterialExpressionConstant3Vector_0}
    B {from=MaterialExpressionConstant3Vector_2}
    Alpha {from=MaterialExpressionNamedRerouteUsage_8}
   ]
  }
  MaterialExpressionLinearInterpolate_1 {
   class=MaterialExpressionLinearInterpolate
   inputs [
    A {from=MaterialExpressionLinearInterpolate_2}
    B {from=MaterialExpressionMultiply_13}
    Alpha {from=MaterialExpressionClamp_0}
   ]
  }
  MaterialExpressionLinearInterpolate_10 {
   class=MaterialExpressionLinearInterpolate
   inputs [
    A {from=MaterialExpressionScalarParameter_39}
    B {from=MaterialExpressionMax_0}
    Alpha {from=MaterialExpressionNamedRerouteUsage_58}
   ]
  }
  MaterialExpressionLinearInterpolate_11 {
   class=MaterialExpressionLinearInterpolate
   inputs [
    A {from=MaterialExpressionNamedRerouteUsage_94}
    B {from=MaterialExpressionNamedRerouteUsage_57}
    Alpha {from=MaterialExpressionSaturate_10}
   ]
  }
  MaterialExpressionLinearInterpolate_12 {
   class=MaterialExpressionLinearInterpolate
   inputs [
    A {from=MaterialExpressionNamedRerouteDeclaration_49}
    B {from=MaterialExpressionSaturate_13}
    Alpha {from=MaterialExpressionScalarParameter_42}
   ]
  }
  MaterialExpressionLinearInterpolate_13 {
   class=MaterialExpressionLinearInterpolate
   inputs [
    A {from=MaterialExpressionTextureSample_2, output=RGB}
    B {from=MaterialExpressionNamedRerouteUsage_60}
    Alpha {from=MaterialExpressionNamedRerouteUsage_61}
   ]
  }
  MaterialExpressionLinearInterpolate_14 {
   class=MaterialExpressionLinearInterpolate
   inputs [
    A {from=""}
    B {from=MaterialExpressionScalarParameter_46}
    Alpha {from=MaterialExpressionNamedRerouteUsage_0}
   ]
   properties {constA=1}
  }
  MaterialExpressionLinearInterpolate_15 {
   class=MaterialExpressionLinearInterpolate
   inputs [
    A {from=MaterialExpressionScalarParameter_5}
    B {from=MaterialExpressionScalarParameter_25}
    Alpha {from=MaterialExpressionNamedRerouteUsage_35}
   ]
   properties {constA=1}
  }
  MaterialExpressionLinearInterpolate_16 {
   class=MaterialExpressionLinearInterpolate
   inputs [
    A {from=""}
    B {from=MaterialExpressionNamedRerouteUsage_22}
    Alpha {from=MaterialExpressionScalarParameter_28}
   ]
   properties {constA=1}
  }
  MaterialExpressionLinearInterpolate_17 {
   class=MaterialExpressionLinearInterpolate
   inputs [
    A {from=MaterialExpressionNormalize_2}
    B {from=MaterialExpressionNamedRerouteUsage_65}
    Alpha {from=MaterialExpressionNamedRerouteUsage_66}
   ]
  }
  MaterialExpressionLinearInterpolate_18 {
   class=MaterialExpressionLinearInterpolate
   inputs [
    A {from=MaterialExpressionScalarParameter_155}
    B {from=MaterialExpressionScalarParameter_44}
    Alpha {from=MaterialExpressionNamedRerouteUsage_67}
   ]
  }
  MaterialExpressionLinearInterpolate_19 {
   class=MaterialExpressionLinearInterpolate
   inputs [
    A {from=MaterialExpressionScalarParameter_156}
    B {from=MaterialExpressionScalarParameter_45}
    Alpha {from=MaterialExpressionNamedRerouteUsage_51}
   ]
  }
  MaterialExpressionLinearInterpolate_2 {
   class=MaterialExpressionLinearInterpolate
   inputs [
    A {from=MaterialExpressionNamedRerouteDeclaration_50}
    B {from=MaterialExpressionNamedRerouteDeclaration_51}
    Alpha {from=MaterialExpressionNamedRerouteUsage_128}
   ]
  }
  MaterialExpressionLinearInterpolate_20 {
   class=MaterialExpressionLinearInterpolate
   inputs [
    A {from=MaterialExpressionScalarParameter_47}
    B {from=""}
    Alpha {from=MaterialExpressionSaturate_11}
   ]
   properties {constA=1}
  }
  MaterialExpressionLinearInterpolate_3 {
   class=MaterialExpressionLinearInterpolate
   inputs [
    A {from=MaterialExpressionNamedRerouteUsage_21}
    B {from=MaterialExpressionLinearInterpolate_0}
    Alpha {from=MaterialExpressionScalarParameter_16}
   ]
  }
  MaterialExpressionLinearInterpolate_4 {
   class=MaterialExpressionLinearInterpolate
   inputs [
    A {from=MaterialExpressionScalarParameter_6}
    B {from=MaterialExpressionScalarParameter_8}
    Alpha {from=MaterialExpressionPower_2}
   ]
  }
  MaterialExpressionLinearInterpolate_5 {
   class=MaterialExpressionLinearInterpolate
   inputs [
    A {from=MaterialExpressionConstant3Vector_1}
    B {from=MaterialExpressionNormalize_15}
    Alpha {from=MaterialExpressionPower_3}
   ]
  }
  MaterialExpressionLinearInterpolate_6 {
   class=MaterialExpressionLinearInterpolate
   inputs [
    A {from=MaterialExpressionScalarParameter_3}
    B {from=MaterialExpressionScalarParameter_27}
    Alpha {from=MaterialExpressionPower_8}
   ]
  }
  MaterialExpressionLinearInterpolate_67 {
   class=MaterialExpressionLinearInterpolate
   inputs [
    A {from=MaterialExpressionOneMinus_26}
    B {from=MaterialExpressionOneMinus_25}
    Alpha {from=MaterialExpressionPower_43}
   ]
  }
  MaterialExpressionLinearInterpolate_72 {
   class=MaterialExpressionLinearInterpolate
   inputs [
    A {from=MaterialExpressionMultiply_149}
    B {from=MaterialExpressionMaterialFunctionCall_7, output=Result}
    Alpha {from=MaterialExpressionTwoSidedSign_3}
   ]
  }
  MaterialExpressionLinearInterpolate_8 {
   class=MaterialExpressionLinearInterpolate
   inputs [
    A {from=MaterialExpressionNamedRerouteUsage_25}
    B {from=MaterialExpressionMultiply_14}
    Alpha {from=MaterialExpressionPower_4}
   ]
  }
  MaterialExpressionLinearInterpolate_9 {
   class=MaterialExpressionLinearInterpolate
   inputs [
    A {from=MaterialExpressionNamedRerouteUsage_29}
    B {from=MaterialExpressionMultiply_15}
    Alpha {from=MaterialExpressionPower_5}
   ]
  }
  MaterialExpressionLogarithm2_0 {class=MaterialExpressionLogarithm2, inputs [X {from=MaterialExpressionLinearInterpolate_10, output=""}], properties {}}
  MaterialExpressionMaterialFunctionCall_0 {
   class=MaterialExpressionMaterialFunctionCall
   inputs [
    Normal {from=MaterialExpressionNamedRerouteUsage_150}
    Flatness {from=MaterialExpressionScalarParameter_15}
   ]
   properties {materialFunction=&FlattenNormal}
  }
  MaterialExpressionMaterialFunctionCall_1 {
   class=MaterialExpressionMaterialFunctionCall
   inputs []
   properties {materialFunction=/Engine/Functions/Engine_MaterialFunctions02/Utility/FOV}
  }
  MaterialExpressionMaterialFunctionCall_3 {
   class=MaterialExpressionMaterialFunctionCall
   inputs [
    Normal {from=MaterialExpressionStaticSwitchParameter_0}
    Flatness {from=MaterialExpressionNamedRerouteDeclaration_19}
   ]
   properties {materialFunction=&FlattenNormal}
  }
  MaterialExpressionMaterialFunctionCall_34 {
   class=MaterialExpressionMaterialFunctionCall
   inputs [
    Brightness {from=MaterialExpressionConstant_30}
    Exponent {from=MaterialExpressionConstant_31}
    Camera Vector {from=MaterialExpressionNormalize_8}
    Vertex / Pixel Normal {from=MaterialExpressionPixelNormalWS_5}
   ]
   properties {materialFunction=&MF_Fresnel_FX}
  }
  MaterialExpressionMaterialFunctionCall_7 {
   class=MaterialExpressionMaterialFunctionCall
   inputs [
    BaseNormal {from=MaterialExpressionMaterialFunctionCall_0, output=Result}
    AdditionalNormal {from=MaterialExpressionMaterialFunctionCall_3, output=Result}
   ]
   properties {materialFunction=&BlendAngleCorrectedNormals}
  }
  MaterialExpressionMax_0 {
   class=MaterialExpressionMax
   inputs [
    A {from=MaterialExpressionScalarParameter_36}
    B {from=""}
   ]
   properties {constB=2}
  }
  MaterialExpressionMultiply_0 {
   class=MaterialExpressionMultiply
   inputs [
    A {from=MaterialExpressionMultiply_4}
    B {from=MaterialExpressionNamedRerouteUsage_1}
   ]
  }
  MaterialExpressionMultiply_1 {
   class=MaterialExpressionMultiply
   inputs [
    A {from=MaterialExpressionScalarParameter_34}
    B {from=MaterialExpressionTextureCoordinate_2}
   ]
  }
  MaterialExpressionMultiply_10 {
   class=MaterialExpressionMultiply
   inputs [
    A {from=MaterialExpressionTextureCoordinate_1}
    B {from=MaterialExpressionScalarParameter_22}
   ]
  }
  MaterialExpressionMultiply_11 {
   class=MaterialExpressionMultiply
   inputs [
    A {from=MaterialExpressionNamedRerouteUsage_175}
    B {from=MaterialExpressionMultiply_9}
   ]
  }
  MaterialExpressionMultiply_114 {
   class=MaterialExpressionMultiply
   inputs [
    A {from=MaterialExpressionTextureSample_1, output=RGB}
    B {from=MaterialExpressionAppendVector_28}
   ]
  }
  MaterialExpressionMultiply_12 {
   class=MaterialExpressionMultiply
   inputs [
    A {from=MaterialExpressionStaticSwitchParameter_4}
    B {from=MaterialExpressionLinearInterpolate_16}
   ]
  }
  MaterialExpressionMultiply_13 {
   class=MaterialExpressionMultiply
   inputs [
    A {from=MaterialExpressionLinearInterpolate_2}
    B {from=MaterialExpressionNamedRerouteUsage_44}
   ]
  }
  MaterialExpressionMultiply_137 {
   class=MaterialExpressionMultiply
   inputs [
    A {from=MaterialExpressionScalarParameter_162}
    B {from=MaterialExpressionAdd_49}
   ]
  }
  MaterialExpressionMultiply_14 {
   class=MaterialExpressionMultiply
   inputs [
    A {from=MaterialExpressionNamedRerouteUsage_25}
    B {from=MaterialExpressionScalarParameter_17}
   ]
  }
  MaterialExpressionMultiply_148 {
   class=MaterialExpressionMultiply
   inputs [
    A {from=MaterialExpressionSubtract_13}
    B {from=MaterialExpressionNamedRerouteUsage_54}
   ]
  }
  MaterialExpressionMultiply_149 {
   class=MaterialExpressionMultiply
   inputs [
    A {from=MaterialExpressionMaterialFunctionCall_7, output=Result}
    B {from=""}
   ]
   properties {constB=-1}
  }
  MaterialExpressionMultiply_15 {
   class=MaterialExpressionMultiply
   inputs [
    A {from=MaterialExpressionNamedRerouteUsage_29}
    B {from=MaterialExpressionScalarParameter_19}
   ]
  }
  MaterialExpressionMultiply_16 {
   class=MaterialExpressionMultiply
   inputs [
    A {from=MaterialExpressionStep_3}
    B {from=MaterialExpressionStep_2}
   ]
  }
  MaterialExpressionMultiply_17 {
   class=MaterialExpressionMultiply
   inputs [
    A {from=MaterialExpressionNamedRerouteUsage_52}
    B {from=MaterialExpressionOneMinus_2}
   ]
  }
  MaterialExpressionMultiply_18 {
   class=MaterialExpressionMultiply
   inputs [
    A {from=MaterialExpressionLength_0}
    B {from=MaterialExpressionSaturate_7}
   ]
  }
  MaterialExpressionMultiply_19 {
   class=MaterialExpressionMultiply
   inputs [
    A {from=MaterialExpressionStep_0}
    B {from=MaterialExpressionStep_1}
   ]
  }
  MaterialExpressionMultiply_2 {
   class=MaterialExpressionMultiply
   inputs [
    A {from=MaterialExpressionConvert_3}
    B {from=""}
   ]
   properties {constB=-1}
  }
  MaterialExpressionMultiply_20 {
   class=MaterialExpressionMultiply
   inputs [
    A {from=MaterialExpressionTextureSample_0, output=RGB}
    B {from=MaterialExpressionScalarParameter_9}
   ]
   properties {constB=2}
  }
  MaterialExpressionMultiply_22 {
   class=MaterialExpressionMultiply
   inputs [
    A {from=MaterialExpressionLinearInterpolate_6}
    B {from=MaterialExpressionNamedRerouteUsage_31}
   ]
  }
  MaterialExpressionMultiply_23 {
   class=MaterialExpressionMultiply
   inputs [
    A {from=MaterialExpressionTextureSample_4, output=RGB}
    B {from=MaterialExpressionAppendVector_1}
   ]
  }
  MaterialExpressionMultiply_24 {
   class=MaterialExpressionMultiply
   inputs [
    A {from=MaterialExpressionScalarParameter_145}
    B {from=MaterialExpressionLinearInterpolate_14}
   ]
  }
  MaterialExpressionMultiply_25 {
   class=MaterialExpressionMultiply
   inputs [
    A {from=MaterialExpressionLinearInterpolate_14}
    B {from=MaterialExpressionScalarParameter_1}
   ]
  }
  MaterialExpressionMultiply_27 {
   class=MaterialExpressionMultiply
   inputs [
    A {from=MaterialExpressionScalarParameter_50}
    B {from=MaterialExpressionNamedRerouteUsage_38}
   ]
  }
  MaterialExpressionMultiply_3 {
   class=MaterialExpressionMultiply
   inputs [
    A {from=MaterialExpressionNamedRerouteUsage_93}
    B {from=MaterialExpressionNamedRerouteUsage_84}
   ]
  }
  MaterialExpressionMultiply_4 {
   class=MaterialExpressionMultiply
   inputs [
    A {from=MaterialExpressionMultiply_5}
    B {from=MaterialExpressionPower_1}
   ]
  }
  MaterialExpressionMultiply_5 {
   class=MaterialExpressionMultiply
   inputs [
    A {from=MaterialExpressionStaticSwitch_11}
    B {from=MaterialExpressionNamedRerouteDeclaration_53}
   ]
  }
  MaterialExpressionMultiply_6 {
   class=MaterialExpressionMultiply
   inputs [
    A {from=MaterialExpressionAdd_1}
    B {from=MaterialExpressionNamedRerouteUsage_6}
   ]
  }
  MaterialExpressionMultiply_65 {
   class=MaterialExpressionMultiply
   inputs [
    A {from=MaterialExpressionAtmosphericLightVector_5}
    B {from=""}
   ]
   properties {constB=-1}
  }
  MaterialExpressionMultiply_66 {
   class=MaterialExpressionMultiply
   inputs [
    A {from=MaterialExpressionFresnel_2}
    B {from=MaterialExpressionMaterialFunctionCall_34, output=Fresnel Inverted}
   ]
  }
  MaterialExpressionMultiply_7 {
   class=MaterialExpressionMultiply
   inputs [
    A {from=MaterialExpressionTransform_0}
    B {from=MaterialExpressionLinearInterpolate_4}
   ]
  }
  MaterialExpressionMultiply_8 {
   class=MaterialExpressionMultiply
   inputs [
    A {from=MaterialExpressionComponentMask_0}
    B {from=""}
   ]
   properties {constB=2}
  }
  MaterialExpressionMultiply_9 {
   class=MaterialExpressionMultiply
   inputs [
    A {from=MaterialExpressionNamedRerouteUsage_17}
    B {from=MaterialExpressionScalarParameter_14}
   ]
  }
  MaterialExpressionNamedRerouteDeclaration_0 {
   class=MaterialExpressionNamedRerouteDeclaration
   inputs [None {from=MaterialExpressionChannelMaskParameter_3, output=R}]
   properties {name=FurClothingMask, nodeColor=<rgba 0.8281 0 1 1>}
  }
  MaterialExpressionNamedRerouteDeclaration_1 {
   class=MaterialExpressionNamedRerouteDeclaration
   inputs [None {from=MaterialExpressionSaturate_4}]
   properties {name=FlowMask, nodeColor=<rgba 1 0.9375 0 1>}
  }
  MaterialExpressionNamedRerouteDeclaration_10 {
   class=MaterialExpressionNamedRerouteDeclaration
   inputs [None {from=MaterialExpressionMultiply_7}]
   properties {name=WPO_NormalInfluence, nodeColor=<rgba 1 0 0.7266 1>}
  }
  MaterialExpressionNamedRerouteDeclaration_104 {
   class=MaterialExpressionNamedRerouteDeclaration
   inputs [None {from=MaterialExpressionAdd_6}]
   properties {name=FurColor_Backlit, nodeColor=<rgba 1 0.6959 0 1>}
  }
  MaterialExpressionNamedRerouteDeclaration_11 {
   class=MaterialExpressionNamedRerouteDeclaration
   inputs [None {from=MaterialExpressionMultiply_0}]
   properties {name=WPO_FlowMap, nodeColor=<rgba 1 0 0.8203 1>}
  }
  MaterialExpressionNamedRerouteDeclaration_12 {
   class=MaterialExpressionNamedRerouteDeclaration
   inputs [None {from=MaterialExpressionSaturate_8}]
   properties {name=Aniso, nodeColor=<rgba 0.0605 1 0.8716 1>}
  }
  MaterialExpressionNamedRerouteDeclaration_13 {
   class=MaterialExpressionNamedRerouteDeclaration
   inputs [None {from=MaterialExpressionFunctionInput_6}]
   properties {name=IN_Material, nodeColor=<rgba 0.125 0 1 1>}
  }
  MaterialExpressionNamedRerouteDeclaration_14 {
   class=MaterialExpressionNamedRerouteDeclaration
   inputs [None {from=MaterialExpressionLinearInterpolate_9}]
   properties {name=Fur_Color, nodeColor=<rgba 0.2422 1 0 1>}
  }
  MaterialExpressionNamedRerouteDeclaration_15 {
   class=MaterialExpressionNamedRerouteDeclaration
   inputs [None {from=MaterialExpressionLinearInterpolate_8}]
   properties {name=FurColor_Roots, nodeColor=<rgba 0.7344 1 0 1>}
  }
  MaterialExpressionNamedRerouteDeclaration_16 {
   class=MaterialExpressionNamedRerouteDeclaration
   inputs [None {from=MaterialExpressionNamedRerouteDeclaration_3}]
   properties {name=Tangent, nodeColor=<rgba 0 1 0.7891 1>}
  }
  MaterialExpressionNamedRerouteDeclaration_17 {
   class=MaterialExpressionNamedRerouteDeclaration
   inputs [None {from=MaterialExpressionLinearInterpolate_12}]
   properties {name=RenormalizedShellLayer, nodeColor=<rgba 0 1 0.1328 1>}
  }
  MaterialExpressionNamedRerouteDeclaration_18 {
   class=MaterialExpressionNamedRerouteDeclaration
   inputs [None {from=MaterialExpressionStaticSwitch_3}]
   properties {name=FurPattern, nodeColor=<rgba 0.3125 1 0 1>}
  }
  MaterialExpressionNamedRerouteDeclaration_19 {
   class=MaterialExpressionNamedRerouteDeclaration
   inputs [None {from=MaterialExpressionMultiply_17}]
   properties {name=ClothingMaskInverse, nodeColor=<rgba 0 1 1 1>}
  }
  MaterialExpressionNamedRerouteDeclaration_2 {
   class=MaterialExpressionNamedRerouteDeclaration
   inputs [None {from=MaterialExpressionIf_1}]
   properties {name=FurWPO, nodeColor=<rgba 1 0.8672 0 1>}
  }
  MaterialExpressionNamedRerouteDeclaration_20 {
   class=MaterialExpressionNamedRerouteDeclaration
   inputs [None {from=MaterialExpressionScalarParameter_21}]
   properties {name=FurTextureMipBias, nodeColor=<rgba 1 0 0.2344 1>}
  }
  MaterialExpressionNamedRerouteDeclaration_22 {
   class=MaterialExpressionNamedRerouteDeclaration
   inputs [None {from=MaterialExpressionStaticSwitch_4}]
   properties {name=FurNormalNormal, nodeColor=<rgba 0.0312 1 0 1>}
  }
  MaterialExpressionNamedRerouteDeclaration_24 {
   class=MaterialExpressionNamedRerouteDeclaration
   inputs [None {from=MaterialExpressionScalarParameter_23}]
   properties {name=FurClothing, nodeColor=<rgba 0.7812 1 0 1>}
  }
  MaterialExpressionNamedRerouteDeclaration_25 {
   class=MaterialExpressionNamedRerouteDeclaration
   inputs [None {from=MaterialExpressionChannelMaskParameter_4, output=R}]
   properties {name=ShellFurMask, nodeColor=<rgba 1 0.8203 0 1>}
  }
  MaterialExpressionNamedRerouteDeclaration_26 {
   class=MaterialExpressionNamedRerouteDeclaration
   inputs [None {from=MaterialExpressionOneMinus_4}]
   properties {name=AnisoBreakup, nodeColor=<rgba 0 1 0.5547 1>}
  }
  MaterialExpressionNamedRerouteDeclaration_27 {
   class=MaterialExpressionNamedRerouteDeclaration
   inputs [None {from=MaterialExpressionLinearInterpolate_3}]
   properties {name=FurDebugColor, nodeColor=<rgba 0.0312 1 0 1>}
  }
  MaterialExpressionNamedRerouteDeclaration_28 {
   class=MaterialExpressionNamedRerouteDeclaration
   inputs [None {from=MaterialExpressionTextureSampleParameter2D_21, output=B}]
   properties {name=AnisoMask, nodeColor=<rgba 0.8281 0 1 1>}
  }
  MaterialExpressionNamedRerouteDeclaration_29 {
   class=MaterialExpressionNamedRerouteDeclaration
   inputs [None {from=MaterialExpressionStaticSwitch_2}]
   properties {name=FurDebugOpacity, nodeColor=<rgba 0.0547 0 1 1>}
  }
  MaterialExpressionNamedRerouteDeclaration_3 {
   class=MaterialExpressionNamedRerouteDeclaration
   inputs [None {from=MaterialExpressionNormalize_0}]
   properties {name=FlowMap_Swizzled, nodeColor=<rgba 1 0.9375 0 1>}
  }
  MaterialExpressionNamedRerouteDeclaration_30 {
   class=MaterialExpressionNamedRerouteDeclaration
   inputs [None {from=MaterialExpressionStaticBoolParameter_0}]
   properties {name=UseFurDebug, nodeColor=<rgba 0 0.2266 1 1>}
  }
  MaterialExpressionNamedRerouteDeclaration_31 {
   class=MaterialExpressionNamedRerouteDeclaration
   inputs [None {from=MaterialExpressionMultiply_18}]
   properties {name=PixelDistance, nodeColor=<rgba 1 0 0.2109 1>}
  }
  MaterialExpressionNamedRerouteDeclaration_32 {
   class=MaterialExpressionNamedRerouteDeclaration
   inputs [None {from=MaterialExpressionSaturate_0}]
   properties {name=FarFurBlend, nodeColor=<rgba 0 0.5312 1 1>}
  }
  MaterialExpressionNamedRerouteDeclaration_33 {
   class=MaterialExpressionNamedRerouteDeclaration
   inputs [None {from=MaterialExpressionIf_2}]
   properties {name=ShellLODCullMask, nodeColor=<rgba 0.8281 0 1 1>}
  }
  MaterialExpressionNamedRerouteDeclaration_34 {
   class=MaterialExpressionNamedRerouteDeclaration
   inputs [None {from=MaterialExpressionSaturate_6}]
   properties {name=ShellLODCullBlend, nodeColor=<rgba 0 0.5312 1 1>}
  }
  MaterialExpressionNamedRerouteDeclaration_35 {
   class=MaterialExpressionNamedRerouteDeclaration
   inputs [None {from=MaterialExpressionLinearInterpolate_1}]
   properties {name=FurLength, nodeColor=<rgba 0.6875 0 1 1>}
  }
  MaterialExpressionNamedRerouteDeclaration_36 {
   class=MaterialExpressionNamedRerouteDeclaration
   inputs [None {from=MaterialExpressionTextureSample_3, output=RGB}]
   properties {name=FurPattern_Distance, nodeColor=<rgba 0.047 0.6685 1 1>}
  }
  MaterialExpressionNamedRerouteDeclaration_37 {
   class=MaterialExpressionNamedRerouteDeclaration
   inputs [None {from=MaterialExpressionMultiply_1}]
   properties {name=FurDistanceUVs, nodeColor=<rgba 1 0.5859 0 1>}
  }
  MaterialExpressionNamedRerouteDeclaration_38 {
   class=MaterialExpressionNamedRerouteDeclaration
   inputs [None {from=MaterialExpressionMultiply_3}]
   properties {name=ScaledShellLayer, nodeColor=<rgba 0.2188 1 0 1>}
  }
  MaterialExpressionNamedRerouteDeclaration_39 {
   class=MaterialExpressionNamedRerouteDeclaration
   inputs [None {from=MaterialExpressionNormalize_3}]
   properties {name=FurNormalDistance, nodeColor=<rgba 0.2656 0 1 1>}
  }
  MaterialExpressionNamedRerouteDeclaration_4 {
   class=MaterialExpressionNamedRerouteDeclaration
   inputs [None {from=MaterialExpressionLinearInterpolate_11}]
   properties {name=WPO_FurLength, nodeColor=<rgba 0.8281 0 1 1>}
  }
  MaterialExpressionNamedRerouteDeclaration_40 {
   class=MaterialExpressionNamedRerouteDeclaration
   inputs [None {from=MaterialExpressionStaticBoolParameter_2}]
   properties {name=UseFarFur, nodeColor=<rgba 0.0547 1 0 1>}
  }
  MaterialExpressionNamedRerouteDeclaration_41 {
   class=MaterialExpressionNamedRerouteDeclaration
   inputs [None {from=MaterialExpressionInverseLinearInterpolate_2}]
   properties {name=ShellLODCullBlendUnclamped, nodeColor=<rgba 0 1 0.1797 1>}
  }
  MaterialExpressionNamedRerouteDeclaration_42 {
   class=MaterialExpressionNamedRerouteDeclaration
   inputs [None {from=MaterialExpressionTextureObjectParameter_0}]
   properties {name=FurPatternTexture, nodeColor=<rgba 0.9922 1 0 1>}
  }
  MaterialExpressionNamedRerouteDeclaration_43 {
   class=MaterialExpressionNamedRerouteDeclaration
   inputs [None {from=MaterialExpressionTextureObjectParameter_1}]
   properties {name=FurPatternNormalTexture, nodeColor=<rgba 0.8516 1 0 1>}
  }
  MaterialExpressionNamedRerouteDeclaration_46 {
   class=MaterialExpressionNamedRerouteDeclaration
   inputs [None {from=MaterialExpressionStaticSwitchParameter_3}]
   properties {name=ShellCullingMask, nodeColor=<rgba 0.125 1 0 1>}
  }
  MaterialExpressionNamedRerouteDeclaration_49 {
   class=MaterialExpressionNamedRerouteDeclaration
   inputs [None {from=MaterialExpressionShaderStageSwitch_1}]
   properties {name=NormalizedShellLayer, nodeColor=<rgba 0.2656 0 1 1>}
  }
  MaterialExpressionNamedRerouteDeclaration_5 {
   class=MaterialExpressionNamedRerouteDeclaration
   inputs [None {from=MaterialExpressionSetMaterialAttributes_3}]
   properties {name=FurShellMaterial, nodeColor=<rgba 1 0.4453 0 1>}
  }
  MaterialExpressionNamedRerouteDeclaration_50 {
   class=MaterialExpressionNamedRerouteDeclaration
   inputs [None {from=MaterialExpressionStaticSwitch_7}]
   properties {name=FurMin, nodeColor=<rgba 0.3828 0 1 1>}
  }
  MaterialExpressionNamedRerouteDeclaration_51 {
   class=MaterialExpressionNamedRerouteDeclaration
   inputs [None {from=MaterialExpressionStaticSwitch_8}]
   properties {name=Fur_Max, nodeColor=<rgba 0 1 0.0625 1>}
  }
  MaterialExpressionNamedRerouteDeclaration_52 {
   class=MaterialExpressionNamedRerouteDeclaration
   inputs [None {from=MaterialExpressionStaticBoolParameter_1}]
   properties {name=UseShellCullingDebug, nodeColor=<rgba 0 0.2266 1 1>}
  }
  MaterialExpressionNamedRerouteDeclaration_53 {
   class=MaterialExpressionNamedRerouteDeclaration
   inputs [None {from=MaterialExpressionTransform_1}]
   properties {name=FlowMap_WS, nodeColor=<rgba 0.4062 1 0 1>}
  }
  MaterialExpressionNamedRerouteDeclaration_62 {
   class=MaterialExpressionNamedRerouteDeclaration
   inputs [None {from=MaterialExpressionSaturate_26}]
   properties {name=fresnel_backlight, nodeColor=<rgba 0 1 0.2969 1>}
  }
  MaterialExpressionNamedRerouteDeclaration_80 {
   class=MaterialExpressionNamedRerouteDeclaration
   inputs [None {from=MaterialExpressionLinearInterpolate_72}]
   properties {name=Fur_Normal, nodeColor=<rgba 0.4247 0.3851 1 1>}
  }
  MaterialExpressionNamedRerouteDeclaration_82 {
   class=MaterialExpressionNamedRerouteDeclaration
   inputs [None {from=MaterialExpressionChannelMaskParameter_1, output=R}]
   properties {name=FurLengthMap, nodeColor=<rgba 0 1 0.2002 1>}
  }
  MaterialExpressionNamedRerouteDeclaration_85 {
   class=MaterialExpressionNamedRerouteDeclaration
   inputs [None {from=MaterialExpressionDeriveNormalZ_1}]
   properties {name=FlowMap, nodeColor=<rgba 1 0.4525 0 1>}
  }
  MaterialExpressionNamedRerouteDeclaration_86 {
   class=MaterialExpressionNamedRerouteDeclaration
   inputs [None {from=MaterialExpressionMultiply_137}]
   properties {name=FurPattern_UVs, nodeColor=<rgba 0.9735 0 1 1>}
  }
  MaterialExpressionNamedRerouteDeclaration_88 {
   class=MaterialExpressionNamedRerouteDeclaration
   inputs [None {from=MaterialExpressionMultiply_148}]
   properties {name=Fur_Opacity, nodeColor=<rgba 0.3906 0.2103 0.0483 1>}
  }
  MaterialExpressionNamedRerouteDeclaration_9 {
   class=MaterialExpressionNamedRerouteDeclaration
   inputs [None {from=MaterialExpressionChannelMaskParameter_2, output=R}]
   properties {name=FurNormalMask, nodeColor=<rgba 0.7109 1 0 1>}
  }
  MaterialExpressionNamedRerouteDeclaration_90 {
   class=MaterialExpressionNamedRerouteDeclaration
   inputs [None {from=MaterialExpressionGetMaterialAttributes_3, output=BaseColor}]
   properties {name=In_BaseColor, nodeColor=<rgba 1 0 0.5396 1>}
  }
  MaterialExpressionNamedRerouteDeclaration_93 {
   class=MaterialExpressionNamedRerouteDeclaration
   inputs [None {from=MaterialExpressionGetMaterialAttributes_3, output=Normal}]
   properties {name=In_Normal, nodeColor=<rgba 1 0 0.5396 1>}
  }
  MaterialExpressionNamedRerouteDeclaration_94 {
   class=MaterialExpressionNamedRerouteDeclaration
   inputs [None {from=MaterialExpressionGetMaterialAttributes_3, output=Tangent}]
   properties {name=In_Tangent, nodeColor=<rgba 1 0 0.5396 1>}
  }
  MaterialExpressionNamedRerouteDeclaration_95 {
   class=MaterialExpressionNamedRerouteDeclaration
   inputs [None {from=MaterialExpressionGetMaterialAttributes_3, output=Anisotropy}]
   properties {name=In_Aniso, nodeColor=<rgba 1 0 0.5396 1>}
  }
  MaterialExpressionNamedRerouteDeclaration_96 {
   class=MaterialExpressionNamedRerouteDeclaration
   inputs [None {from=MaterialExpressionGetMaterialAttributes_3, output=MaterialAttributes}]
   properties {name=In_MA, nodeColor=<rgba 1 0 0.5396 1>}
  }
  MaterialExpressionNamedRerouteUsage_0 {class=MaterialExpressionNamedRerouteUsage}
  MaterialExpressionNamedRerouteUsage_1 {class=MaterialExpressionNamedRerouteUsage}
  MaterialExpressionNamedRerouteUsage_12 {class=MaterialExpressionNamedRerouteUsage}
  MaterialExpressionNamedRerouteUsage_128 {class=MaterialExpressionNamedRerouteUsage}
  MaterialExpressionNamedRerouteUsage_131 {class=MaterialExpressionNamedRerouteUsage}
  MaterialExpressionNamedRerouteUsage_135 {class=MaterialExpressionNamedRerouteUsage}
  MaterialExpressionNamedRerouteUsage_149 {class=MaterialExpressionNamedRerouteUsage}
  MaterialExpressionNamedRerouteUsage_150 {class=MaterialExpressionNamedRerouteUsage}
  MaterialExpressionNamedRerouteUsage_151 {class=MaterialExpressionNamedRerouteUsage}
  MaterialExpressionNamedRerouteUsage_17 {class=MaterialExpressionNamedRerouteUsage}
  MaterialExpressionNamedRerouteUsage_175 {class=MaterialExpressionNamedRerouteUsage}
  MaterialExpressionNamedRerouteUsage_18 {class=MaterialExpressionNamedRerouteUsage}
  MaterialExpressionNamedRerouteUsage_19 {class=MaterialExpressionNamedRerouteUsage}
  MaterialExpressionNamedRerouteUsage_2 {class=MaterialExpressionNamedRerouteUsage}
  MaterialExpressionNamedRerouteUsage_20 {class=MaterialExpressionNamedRerouteUsage}
  MaterialExpressionNamedRerouteUsage_21 {class=MaterialExpressionNamedRerouteUsage}
  MaterialExpressionNamedRerouteUsage_22 {class=MaterialExpressionNamedRerouteUsage}
  MaterialExpressionNamedRerouteUsage_23 {class=MaterialExpressionNamedRerouteUsage}
  MaterialExpressionNamedRerouteUsage_24 {class=MaterialExpressionNamedRerouteUsage}
  MaterialExpressionNamedRerouteUsage_25 {class=MaterialExpressionNamedRerouteUsage}
  MaterialExpressionNamedRerouteUsage_26 {class=MaterialExpressionNamedRerouteUsage}
  MaterialExpressionNamedRerouteUsage_27 {class=MaterialExpressionNamedRerouteUsage}
  MaterialExpressionNamedRerouteUsage_28 {class=MaterialExpressionNamedRerouteUsage}
  MaterialExpressionNamedRerouteUsage_29 {class=MaterialExpressionNamedRerouteUsage}
  MaterialExpressionNamedRerouteUsage_3 {class=MaterialExpressionNamedRerouteUsage}
  MaterialExpressionNamedRerouteUsage_30 {class=MaterialExpressionNamedRerouteUsage}
  MaterialExpressionNamedRerouteUsage_31 {class=MaterialExpressionNamedRerouteUsage}
  MaterialExpressionNamedRerouteUsage_32 {class=MaterialExpressionNamedRerouteUsage}
  MaterialExpressionNamedRerouteUsage_33 {class=MaterialExpressionNamedRerouteUsage}
  MaterialExpressionNamedRerouteUsage_35 {class=MaterialExpressionNamedRerouteUsage}
  MaterialExpressionNamedRerouteUsage_36 {class=MaterialExpressionNamedRerouteUsage}
  MaterialExpressionNamedRerouteUsage_37 {class=MaterialExpressionNamedRerouteUsage}
  MaterialExpressionNamedRerouteUsage_38 {class=MaterialExpressionNamedRerouteUsage}
  MaterialExpressionNamedRerouteUsage_4 {class=MaterialExpressionNamedRerouteUsage}
  MaterialExpressionNamedRerouteUsage_40 {class=MaterialExpressionNamedRerouteUsage}
  MaterialExpressionNamedRerouteUsage_42 {class=MaterialExpressionNamedRerouteUsage}
  MaterialExpressionNamedRerouteUsage_43 {class=MaterialExpressionNamedRerouteUsage}
  MaterialExpressionNamedRerouteUsage_44 {class=MaterialExpressionNamedRerouteUsage}
  MaterialExpressionNamedRerouteUsage_45 {class=MaterialExpressionNamedRerouteUsage}
  MaterialExpressionNamedRerouteUsage_46 {class=MaterialExpressionNamedRerouteUsage}
  MaterialExpressionNamedRerouteUsage_48 {class=MaterialExpressionNamedRerouteUsage}
  MaterialExpressionNamedRerouteUsage_5 {class=MaterialExpressionNamedRerouteUsage}
  MaterialExpressionNamedRerouteUsage_51 {class=MaterialExpressionNamedRerouteUsage}
  MaterialExpressionNamedRerouteUsage_52 {class=MaterialExpressionNamedRerouteUsage}
  MaterialExpressionNamedRerouteUsage_53 {class=MaterialExpressionNamedRerouteUsage}
  MaterialExpressionNamedRerouteUsage_54 {class=MaterialExpressionNamedRerouteUsage}
  MaterialExpressionNamedRerouteUsage_55 {class=MaterialExpressionNamedRerouteUsage}
  MaterialExpressionNamedRerouteUsage_56 {class=MaterialExpressionNamedRerouteUsage}
  MaterialExpressionNamedRerouteUsage_57 {class=MaterialExpressionNamedRerouteUsage}
  MaterialExpressionNamedRerouteUsage_58 {class=MaterialExpressionNamedRerouteUsage}
  MaterialExpressionNamedRerouteUsage_59 {class=MaterialExpressionNamedRerouteUsage}
  MaterialExpressionNamedRerouteUsage_6 {class=MaterialExpressionNamedRerouteUsage}
  MaterialExpressionNamedRerouteUsage_60 {class=MaterialExpressionNamedRerouteUsage}
  MaterialExpressionNamedRerouteUsage_61 {class=MaterialExpressionNamedRerouteUsage}
  MaterialExpressionNamedRerouteUsage_62 {class=MaterialExpressionNamedRerouteUsage}
  MaterialExpressionNamedRerouteUsage_63 {class=MaterialExpressionNamedRerouteUsage}
  MaterialExpressionNamedRerouteUsage_64 {class=MaterialExpressionNamedRerouteUsage}
  MaterialExpressionNamedRerouteUsage_65 {class=MaterialExpressionNamedRerouteUsage}
  MaterialExpressionNamedRerouteUsage_66 {class=MaterialExpressionNamedRerouteUsage}
  MaterialExpressionNamedRerouteUsage_67 {class=MaterialExpressionNamedRerouteUsage}
  MaterialExpressionNamedRerouteUsage_68 {class=MaterialExpressionNamedRerouteUsage}
  MaterialExpressionNamedRerouteUsage_69 {class=MaterialExpressionNamedRerouteUsage}
  MaterialExpressionNamedRerouteUsage_7 {class=MaterialExpressionNamedRerouteUsage}
  MaterialExpressionNamedRerouteUsage_70 {class=MaterialExpressionNamedRerouteUsage}
  MaterialExpressionNamedRerouteUsage_71 {class=MaterialExpressionNamedRerouteUsage}
  MaterialExpressionNamedRerouteUsage_72 {class=MaterialExpressionNamedRerouteUsage}
  MaterialExpressionNamedRerouteUsage_73 {class=MaterialExpressionNamedRerouteUsage}
  MaterialExpressionNamedRerouteUsage_74 {class=MaterialExpressionNamedRerouteUsage}
  MaterialExpressionNamedRerouteUsage_75 {class=MaterialExpressionNamedRerouteUsage}
  MaterialExpressionNamedRerouteUsage_76 {class=MaterialExpressionNamedRerouteUsage}
  MaterialExpressionNamedRerouteUsage_77 {class=MaterialExpressionNamedRerouteUsage}
  MaterialExpressionNamedRerouteUsage_8 {class=MaterialExpressionNamedRerouteUsage}
  MaterialExpressionNamedRerouteUsage_84 {class=MaterialExpressionNamedRerouteUsage}
  MaterialExpressionNamedRerouteUsage_89 {class=MaterialExpressionNamedRerouteUsage}
  MaterialExpressionNamedRerouteUsage_9 {class=MaterialExpressionNamedRerouteUsage}
  MaterialExpressionNamedRerouteUsage_90 {class=MaterialExpressionNamedRerouteUsage}
  MaterialExpressionNamedRerouteUsage_91 {class=MaterialExpressionNamedRerouteUsage}
  MaterialExpressionNamedRerouteUsage_93 {class=MaterialExpressionNamedRerouteUsage}
  MaterialExpressionNamedRerouteUsage_94 {class=MaterialExpressionNamedRerouteUsage}
  MaterialExpressionNamedRerouteUsage_95 {class=MaterialExpressionNamedRerouteUsage}
  MaterialExpressionNormalize_0 {class=MaterialExpressionNormalize, inputs [VectorInput {from=MaterialExpressionConvert_4, output=""}]}
  MaterialExpressionNormalize_15 {class=MaterialExpressionNormalize, inputs [VectorInput {from=MaterialExpressionNamedRerouteUsage_48, output=""}]}
  MaterialExpressionNormalize_2 {class=MaterialExpressionNormalize, inputs [VectorInput {from=MaterialExpressionMultiply_114, output=""}]}
  MaterialExpressionNormalize_3 {class=MaterialExpressionNormalize, inputs [VectorInput {from=MaterialExpressionMultiply_23, output=""}]}
  MaterialExpressionNormalize_8 {class=MaterialExpressionNormalize, inputs [VectorInput {from=MaterialExpressionMultiply_65, output=""}]}
  MaterialExpressionOneMinus_0 {class=MaterialExpressionOneMinus, inputs [None {from=MaterialExpressionNamedRerouteUsage_32}]}
  MaterialExpressionOneMinus_1 {class=MaterialExpressionOneMinus, inputs [None {from=MaterialExpressionSaturate_14}]}
  MaterialExpressionOneMinus_2 {class=MaterialExpressionOneMinus, inputs [None {from=MaterialExpressionNamedRerouteUsage_53}]}
  MaterialExpressionOneMinus_25 {class=MaterialExpressionOneMinus, inputs [None {from=MaterialExpressionStaticSwitch_6}]}
  MaterialExpressionOneMinus_26 {class=MaterialExpressionOneMinus, inputs [None {from=MaterialExpressionStaticSwitch_5}]}
  MaterialExpressionOneMinus_4 {class=MaterialExpressionOneMinus, inputs [None {from=MaterialExpressionSaturate_9}]}
  MaterialExpressionOneMinus_6 {class=MaterialExpressionOneMinus, inputs [None {from=MaterialExpressionSaturate_5}]}
  MaterialExpressionOneMinus_7 {class=MaterialExpressionOneMinus, inputs [None {from=MaterialExpressionScalarParameter_0}]}
  MaterialExpressionPixelNormalWS_5 {class=MaterialExpressionPixelNormalWS, inputs [], properties {}}
  MaterialExpressionPower_0 {
   class=MaterialExpressionPower
   inputs [
    Base {from=MaterialExpressionConstant_4}
    Exp {from=MaterialExpressionCeil_0}
   ]
  }
  MaterialExpressionPower_1 {
   class=MaterialExpressionPower
   inputs [
    Base {from=MaterialExpressionAdd_0}
    Exp {from=MaterialExpressionScalarParameter_4}
   ]
  }
  MaterialExpressionPower_13 {
   class=MaterialExpressionPower
   inputs [
    Base {from=MaterialExpressionMultiply_66}
    Exp {from=MaterialExpressionConstant_34}
   ]
   properties {constExponent=3}
  }
  MaterialExpressionPower_2 {
   class=MaterialExpressionPower
   inputs [
    Base {from=MaterialExpressionNamedRerouteUsage_90}
    Exp {from=MaterialExpressionScalarParameter_7}
   ]
  }
  MaterialExpressionPower_3 {
   class=MaterialExpressionPower
   inputs [
    Base {from=MaterialExpressionNamedRerouteUsage_18}
    Exp {from=MaterialExpressionScalarParameter_10}
   ]
  }
  MaterialExpressionPower_4 {
   class=MaterialExpressionPower
   inputs [
    Base {from=MaterialExpressionOneMinus_0}
    Exp {from=MaterialExpressionScalarParameter_18}
   ]
  }
  MaterialExpressionPower_43 {
   class=MaterialExpressionPower
   inputs [
    Base {from=MaterialExpressionNamedRerouteUsage_40}
    Exp {from=MaterialExpressionScalarParameter_157}
   ]
  }
  MaterialExpressionPower_5 {
   class=MaterialExpressionPower
   inputs [
    Base {from=MaterialExpressionNamedRerouteUsage_30}
    Exp {from=MaterialExpressionScalarParameter_20}
   ]
  }
  MaterialExpressionPower_8 {
   class=MaterialExpressionPower
   inputs [
    Base {from=MaterialExpressionNamedRerouteUsage_28}
    Exp {from=MaterialExpressionScalarParameter_29}
   ]
  }
  MaterialExpressionSaturate_0 {class=MaterialExpressionSaturate, inputs [None {from=MaterialExpressionInverseLinearInterpolate_0}]}
  MaterialExpressionSaturate_10 {class=MaterialExpressionSaturate, inputs [None {from=MaterialExpressionAdd_12}]}
  MaterialExpressionSaturate_11 {class=MaterialExpressionSaturate, inputs [None {from=MaterialExpressionInverseLinearInterpolate_3}]}
  MaterialExpressionSaturate_13 {class=MaterialExpressionSaturate, inputs [None {from=MaterialExpressionDivide_4}]}
  MaterialExpressionSaturate_14 {class=MaterialExpressionSaturate, inputs [None {from=MaterialExpressionInverseLinearInterpolate_1}]}
  MaterialExpressionSaturate_26 {class=MaterialExpressionSaturate, inputs [None {from=MaterialExpressionPower_13}]}
  MaterialExpressionSaturate_4 {class=MaterialExpressionSaturate, inputs [None {from=MaterialExpressionDivide_0}]}
  MaterialExpressionSaturate_5 {class=MaterialExpressionSaturate, inputs [None {from=MaterialExpressionScalarParameter_26}]}
  MaterialExpressionSaturate_6 {class=MaterialExpressionSaturate, inputs [None {from=MaterialExpressionNamedRerouteDeclaration_41}]}
  MaterialExpressionSaturate_7 {class=MaterialExpressionSaturate, inputs [None {from=MaterialExpressionLinearInterpolate_20}]}
  MaterialExpressionSaturate_8 {class=MaterialExpressionSaturate, inputs [None {from=MaterialExpressionMultiply_12}]}
  MaterialExpressionSaturate_9 {class=MaterialExpressionSaturate, inputs [None {from=MaterialExpressionMultiply_20}]}
  MaterialExpressionScalarParameter_0 {
   class=MaterialExpressionScalarParameter
   properties {
    defaultValue=0.1
    desc=Buffer stoping the flow map from effecting the first few shells to avoid shell missalignment relative to the mesh
    group=Fur_Shell_03 - WPO
    parameterName=FlowMapMaskRootBuffer
    sortPriority=3
   }
  }
  MaterialExpressionScalarParameter_1 {
   class=MaterialExpressionScalarParameter
   properties {defaultValue=1, group=Fur_Shell_01 - Base Settings, parameterName=FurLengthMax, sortPriority=9}
  }
  MaterialExpressionScalarParameter_10 {
   class=MaterialExpressionScalarParameter
   properties {defaultValue=1, group=Fur_Shell_02 - Strands, parameterName=FurNormalMaskExponent, sortPriority=17}
  }
  MaterialExpressionScalarParameter_137 {
   class=MaterialExpressionScalarParameter
   properties {defaultValue=1, desc=Strength of the normal map., group=Fur_Shell_02 - Strands, parameterName=FurPatternNormalStrength, sortPriority=12}
  }
  MaterialExpressionScalarParameter_14 {class=MaterialExpressionScalarParameter, properties {defaultValue=1, group=Fur_Shell_02 - Strands, parameterName=FurBacklightContribution}}
  MaterialExpressionScalarParameter_145 {
   class=MaterialExpressionScalarParameter
   properties {defaultValue=0.2, group=Fur_Shell_01 - Base Settings, parameterName=FurLengthMin, sortPriority=8}
  }
  MaterialExpressionScalarParameter_15 {
   class=MaterialExpressionScalarParameter
   properties {
    desc="Reduces the Mesh's normal map against the fur "
    group=Fur_Shell_01 - Base Settings
    parameterName=MeshNormalMapFlatten
    sortPriority=40
   }
  }
  MaterialExpressionScalarParameter_155 {
   class=MaterialExpressionScalarParameter
   properties {
    defaultValue=1.5
    desc=How thick is the fur strand at the base/root.
    group=Fur_Shell_02 - Strands
    parameterName=FurThicknessRoot
    sortPriority=55
   }
  }
  MaterialExpressionScalarParameter_156 {
   class=MaterialExpressionScalarParameter
   properties {defaultValue=0.5, desc=How thick is the fur strand at the tip., group=Fur_Shell_02 - Strands, parameterName=FurThicknessTip, sortPriority=56}
  }
  MaterialExpressionScalarParameter_157 {
   class=MaterialExpressionScalarParameter
   properties {
    defaultValue=1
    desc=How is the transition between root and tip thickness distributed along the strand.
    group=Fur_Shell_02 - Strands
    parameterName=FurThicknessRootTipDistribution
    sortPriority=57
   }
  }
  MaterialExpressionScalarParameter_16 {class=MaterialExpressionScalarParameter, properties {defaultValue=1, group=Fur_Shell_06 - Debug, parameterName=FurDebugColorStrength}}
  MaterialExpressionScalarParameter_162 {
   class=MaterialExpressionScalarParameter
   properties {defaultValue=10, group=Fur_Shell_02 - Strands, parameterName=FurPatternUVTile, sortPriority=3}
  }
  MaterialExpressionScalarParameter_163 {class=MaterialExpressionScalarParameter, properties {group=Fur_Shell_02 - Strands, parameterName=FurPatternUVOffsetX, sortPriority=3}}
  MaterialExpressionScalarParameter_164 {class=MaterialExpressionScalarParameter, properties {group=Fur_Shell_02 - Strands, parameterName=FurPatternUVOffsetY, sortPriority=3}}
  MaterialExpressionScalarParameter_17 {class=MaterialExpressionScalarParameter, properties {defaultValue=0.5, group=Fur_Shell_02 - Strands, parameterName=FurColorRootMulitplier}}
  MaterialExpressionScalarParameter_18 {class=MaterialExpressionScalarParameter, properties {defaultValue=1, group=Fur_Shell_02 - Strands, parameterName=FurColorRootExponent}}
  MaterialExpressionScalarParameter_19 {class=MaterialExpressionScalarParameter, properties {defaultValue=1, group=Fur_Shell_02 - Strands, parameterName=FurColorTipMulitplier}}
  MaterialExpressionScalarParameter_2 {class=MaterialExpressionScalarParameter, properties {defaultValue=50, group=Fur_Shell_06 - Debug, parameterName=FurDebugClipShellIndex}}
  MaterialExpressionScalarParameter_20 {class=MaterialExpressionScalarParameter, properties {defaultValue=1, group=Fur_Shell_02 - Strands, parameterName=FurColorTipExponent}}
  MaterialExpressionScalarParameter_21 {
   class=MaterialExpressionScalarParameter
   properties {defaultValue=-1, group=Fur_Shell_02 - Strands, parameterName=FurPatternMipBias, sortPriority=1}
  }
  MaterialExpressionScalarParameter_22 {
   class=MaterialExpressionScalarParameter
   properties {defaultValue=3, group=Fur_Shell_04 - Aniso, parameterName=AnisoNoiseTiling, sortPriority=42}
  }
  MaterialExpressionScalarParameter_23 {
   class=MaterialExpressionScalarParameter
   properties {desc="Fur Clothing Mask Toggle ", group=Fur_Shell_01 - Base Settings, parameterName=FurClothing, sortPriority=41}
  }
  MaterialExpressionScalarParameter_25 {
   class=MaterialExpressionScalarParameter
   properties {defaultValue=1, group=Fur_Shell_07 - Far Fur, parameterName=FarFlowMapStrength, sortPriority=13}
  }
  MaterialExpressionScalarParameter_26 {class=MaterialExpressionScalarParameter, properties {group=Fur_Shell_05 - Culling, parameterName=FurShellCullAmount, sortPriority=2}}
  MaterialExpressionScalarParameter_27 {class=MaterialExpressionScalarParameter, properties {defaultValue=0.6, group=Fur_Shell_04 - Aniso, parameterName=AnisoTipAmount}}
  MaterialExpressionScalarParameter_28 {
   class=MaterialExpressionScalarParameter
   properties {defaultValue=1, group=Fur_Shell_04 - Aniso, parameterName=AnisoMaskStrength, sortPriority=35}
  }
  MaterialExpressionScalarParameter_29 {
   class=MaterialExpressionScalarParameter
   properties {
    defaultValue=1
    desc=How is the transition between root and tip aniso effect distributed along the strand.
    group=Fur_Shell_04 - Aniso
    parameterName=AnisoRootTipDistribution
    sortPriority=33
   }
  }
  MaterialExpressionScalarParameter_3 {class=MaterialExpressionScalarParameter, properties {group=Fur_Shell_04 - Aniso, parameterName=AnisoRootAmount, sortPriority=31}}
  MaterialExpressionScalarParameter_30 {class=MaterialExpressionScalarParameter, properties {group=Fur_Shell_07 - Far Fur, parameterName=FarFurNearDistance, sortPriority=1}}
  MaterialExpressionScalarParameter_31 {
   class=MaterialExpressionScalarParameter
   properties {defaultValue=1, group=Fur_Shell_05 - Culling, parameterName=FurShellMinCullLayer, sortPriority=5}
  }
  MaterialExpressionScalarParameter_32 {class=MaterialExpressionScalarParameter, properties {group=Fur_Shell_07 - Far Fur, parameterName=FarFurRange, sortPriority=2}}
  MaterialExpressionScalarParameter_34 {
   class=MaterialExpressionScalarParameter
   properties {defaultValue=10, group=Fur_Shell_07 - Far Fur, parameterName=FarFurPattern_01_UVTile, sortPriority=10}
  }
  MaterialExpressionScalarParameter_35 {
   class=MaterialExpressionScalarParameter
   properties {
    defaultValue=1
    desc=Strength of the normal map.
    group=Fur_Shell_07 - Far Fur
    parameterName=FarFurPattern_01_NormalStrength
    sortPriority=20
   }
  }
  MaterialExpressionScalarParameter_36 {
   class=MaterialExpressionScalarParameter
   properties {defaultValue=2, group=Fur_Shell_05 - Culling, parameterName=ShellLODCullMin, sortPriority=103}
  }
  MaterialExpressionScalarParameter_37 {
   class=MaterialExpressionScalarParameter
   properties {defaultValue=50, group=Fur_Shell_05 - Culling, parameterName=ShellLODCullDistance, sortPriority=101}
  }
  MaterialExpressionScalarParameter_38 {
   class=MaterialExpressionScalarParameter
   properties {defaultValue=250, group=Fur_Shell_05 - Culling, parameterName=ShellLODCullRange, sortPriority=102}
  }
  MaterialExpressionScalarParameter_39 {
   class=MaterialExpressionScalarParameter
   properties {defaultValue=16, group=Fur_Shell_05 - Culling, parameterName=ShellLODCullMax, sortPriority=104}
  }
  MaterialExpressionScalarParameter_4 {
   class=MaterialExpressionScalarParameter
   properties {
    defaultValue=1
    desc=Distribution of the flow map effect on shells between root and tip of the fur strand.
    group=Fur_Shell_03 - WPO
    parameterName=FlowMapRootTipDistribution
    sortPriority=2
   }
  }
  MaterialExpressionScalarParameter_40 {
   class=MaterialExpressionScalarParameter
   properties {defaultValue=50, group=Fur_Shell_05 - Culling, parameterName=FurShellCullNearDistance, sortPriority=3}
  }
  MaterialExpressionScalarParameter_41 {
   class=MaterialExpressionScalarParameter
   properties {defaultValue=100, group=Fur_Shell_05 - Culling, parameterName=FurShellCullRange, sortPriority=4}
  }
  MaterialExpressionScalarParameter_42 {class=MaterialExpressionScalarParameter, properties {group=Fur_Shell_05 - Culling, parameterName=RenormalizeShellLayers, sortPriority=10}}
  MaterialExpressionScalarParameter_43 {
   class=MaterialExpressionScalarParameter
   properties {defaultValue=8, group=Fur_Shell_05 - Culling, parameterName=ShellLODMinLayerCount, sortPriority=105}
  }
  MaterialExpressionScalarParameter_44 {
   class=MaterialExpressionScalarParameter
   properties {
    defaultValue=1.5
    desc=How thick is the fur strand at the base/root.
    group=Fur_Shell_07 - Far Fur
    parameterName=FarFurThicknessRoot
    sortPriority=11
   }
  }
  MaterialExpressionScalarParameter_45 {
   class=MaterialExpressionScalarParameter
   properties {
    defaultValue=0.5
    desc=How thick is the fur strand at the tip.
    group=Fur_Shell_07 - Far Fur
    parameterName=FarFurThicknessTip
    sortPriority=12
   }
  }
  MaterialExpressionScalarParameter_46 {
   class=MaterialExpressionScalarParameter
   properties {defaultValue=1, group=Fur_Shell_07 - Far Fur, parameterName=FarFurLengthMultiplier, sortPriority=13}
  }
  MaterialExpressionScalarParameter_47 {
   class=MaterialExpressionScalarParameter
   properties {defaultValue=0.5, group=Fur_Shell_05 - Culling, parameterName=FOVDistanceMultiplier, sortPriority=100}
  }
  MaterialExpressionScalarParameter_48 {class=MaterialExpressionScalarParameter, properties {defaultValue=99}}
  MaterialExpressionScalarParameter_5 {
   class=MaterialExpressionScalarParameter
   properties {defaultValue=1, group=Fur_Shell_03 - WPO, parameterName=FlowMapStrength, sortPriority=1}
  }
  MaterialExpressionScalarParameter_50 {class=MaterialExpressionScalarParameter, properties {group=Fur_Shell_05 - Culling, parameterName=FurShellCullClothingBias, sortPriority=100}}
  MaterialExpressionScalarParameter_6 {
   class=MaterialExpressionScalarParameter
   properties {
    desc="Shells offset strength along the vert normals. Controls distance between shells at root "
    group=Fur_Shell_03 - WPO
    parameterName=NormalVSRootStrength
    sortPriority=4
   }
  }
  MaterialExpressionScalarParameter_7 {
   class=MaterialExpressionScalarParameter
   properties {
    defaultValue=1
    desc=Distribution of the shells along the vertex normals between root and tip values. Controlling overall distance between shells along strand.
    group=Fur_Shell_03 - WPO
    parameterName=NormalVSRootTipDistribution
    sortPriority=6
   }
  }
  MaterialExpressionScalarParameter_8 {
   class=MaterialExpressionScalarParameter
   properties {
    defaultValue=0.4
    desc="Shells offset strength along the vert normals. Controls distance between shells at tip. "
    group=Fur_Shell_03 - WPO
    parameterName=NormalVSTipStrength
    sortPriority=5
   }
  }
  MaterialExpressionScalarParameter_9 {class=MaterialExpressionScalarParameter, properties {group=Fur_Shell_04 - Aniso, parameterName=AnisoNoiseStrength, sortPriority=41}}
  MaterialExpressionSetMaterialAttributes_3 {
   class=MaterialExpressionSetMaterialAttributes
   inputs [
    MaterialAttributes {from=MaterialExpressionNamedRerouteUsage_151, output=""}
    BaseColor {from=MaterialExpressionStaticSwitch_0, output=""}
    Anisotropy {from=MaterialExpressionNamedRerouteUsage_23, output=""}
    OpacityMask {from=MaterialExpressionStaticSwitch_1, output=""}
    Normal {from=MaterialExpressionNamedRerouteUsage_149, output=""}
    Tangent {from=MaterialExpressionNamedRerouteUsage_24, output=""}
    WorldPositionOffset {from=MaterialExpressionNamedRerouteUsage_4, output=""}
    ShadingModel {from=MaterialExpressionShadingModel_3, output=""}
   ]
   properties {attributeSetTypes [
     #1
     #4
     #5
     #2
     #3
     #6
     #7
    ]}
  }
  MaterialExpressionShaderStageSwitch_1 {
   class=MaterialExpressionShaderStageSwitch
   inputs [
    PixelShader {from=MaterialExpressionVertexInterpolator_0, output=PS}
    VertexShader {from=ShellMeshMaterialExpressionNormalizedShellLayer_0, output=""}
   ]
   properties {}
  }
  MaterialExpressionShadingModel_3 {class=MaterialExpressionShadingModel, inputs [], properties {}}
  MaterialExpressionStaticBoolParameter_0 {class=MaterialExpressionStaticBoolParameter, inputs [], properties {group=Fur_Shell_06 - Debug, parameterName=UseFurShellDebug}}
  MaterialExpressionStaticBoolParameter_1 {class=MaterialExpressionStaticBoolParameter, inputs [], properties {group=Fur_Shell_06 - Debug, parameterName=UseFurShellCullingDebug}}
  MaterialExpressionStaticBoolParameter_2 {class=MaterialExpressionStaticBoolParameter, inputs [], properties {group=Fur_Shell_07 - Far Fur, parameterName=UseFarFur, sortPriority=0}}
  MaterialExpressionStaticSwitchParameter_0 {
   class=MaterialExpressionStaticSwitchParameter
   inputs [
    True {from=MaterialExpressionLinearInterpolate_5, output=""}
    False {from=MaterialExpressionNormalize_15, output=""}
   ]
   properties {
    defaultValue=true
    desc="Reduces fur normal map intensiy on areas of the mask "
    group=Fur_Shell_02 - Strands
    parameterName=UseFurNormalMask
    sortPriority=16
   }
  }
  MaterialExpressionStaticSwitchParameter_1 {
   class=MaterialExpressionStaticSwitchParameter
   inputs [
    True {from=MaterialExpressionNamedRerouteUsage_3, output=""}
    False {from=MaterialExpressionNamedRerouteUsage_7, output=""}
   ]
   properties {group=Fur_Shell_01 - Base Settings, parameterName=UseFurShells, sortPriority=-2}
  }
  MaterialExpressionStaticSwitchParameter_2 {
   class=MaterialExpressionStaticSwitchParameter
   inputs [
    True {from=MaterialExpressionOneMinus_1, output=""}
    False {from=MaterialExpressionOneMinus_6, output=""}
   ]
   properties {group=Fur_Shell_05 - Culling, parameterName=UseCameraDistanceFurCulling, sortPriority=1}
  }
  MaterialExpressionStaticSwitchParameter_3 {
   class=MaterialExpressionStaticSwitchParameter
   inputs [
    True {from=MaterialExpressionAdd_2, output=""}
    False {from=MaterialExpressionMultiply_19, output=""}
   ]
   properties {group=Fur_Shell_05 - Culling, parameterName=UseShellLODCulling, sortPriority=100}
  }
  MaterialExpressionStaticSwitchParameter_4 {
   class=MaterialExpressionStaticSwitchParameter
   inputs [
    True {from=MaterialExpressionMultiply_22, output=""}
    False {from=MaterialExpressionLinearInterpolate_6, output=""}
   ]
   properties {group=Fur_Shell_04 - Aniso, parameterName=UseAnisoNoise, sortPriority=40}
  }
  MaterialExpressionStaticSwitch_0 {
   class=MaterialExpressionStaticSwitch
   inputs [
    True {from=MaterialExpressionNamedRerouteUsage_9, output=""}
    False {from=MaterialExpressionNamedRerouteUsage_2, output=""}
    Value {from=MaterialExpressionNamedRerouteUsage_12, output=""}
   ]
   properties {}
  }
  MaterialExpressionStaticSwitch_1 {
   class=MaterialExpressionStaticSwitch
   inputs [
    True {from=MaterialExpressionNamedRerouteUsage_42, output=""}
    False {from=MaterialExpressionNamedRerouteUsage_43, output=""}
    Value {from=MaterialExpressionNamedRerouteUsage_46, output=""}
   ]
   properties {}
  }
  MaterialExpressionStaticSwitch_11 {
   class=MaterialExpressionStaticSwitch
   inputs [
    True {from=MaterialExpressionLinearInterpolate_15, output=""}
    False {from=MaterialExpressionScalarParameter_5, output=""}
    Value {from=MaterialExpressionNamedRerouteUsage_37, output=""}
   ]
   properties {}
  }
  MaterialExpressionStaticSwitch_2 {
   class=MaterialExpressionStaticSwitch
   inputs [
    True {from=MaterialExpressionConstant_5, output=""}
    False {from=MaterialExpressionIf_0, output=""}
    Value {from=MaterialExpressionNamedRerouteUsage_95, output=""}
   ]
   properties {}
  }
  MaterialExpressionStaticSwitch_3 {
   class=MaterialExpressionStaticSwitch
   inputs [
    True {from=MaterialExpressionLinearInterpolate_13, output=""}
    False {from=MaterialExpressionTextureSample_2, output=RGB}
    Value {from=MaterialExpressionNamedRerouteUsage_68, output=""}
   ]
   properties {}
  }
  MaterialExpressionStaticSwitch_4 {
   class=MaterialExpressionStaticSwitch
   inputs [
    True {from=MaterialExpressionLinearInterpolate_17, output=""}
    False {from=MaterialExpressionNormalize_2, output=""}
    Value {from=MaterialExpressionNamedRerouteUsage_69, output=""}
   ]
   properties {}
  }
  MaterialExpressionStaticSwitch_5 {
   class=MaterialExpressionStaticSwitch
   inputs [
    True {from=MaterialExpressionLinearInterpolate_18, output=""}
    False {from=MaterialExpressionScalarParameter_155, output=""}
    Value {from=MaterialExpressionNamedRerouteUsage_70, output=""}
   ]
   properties {}
  }
  MaterialExpressionStaticSwitch_6 {
   class=MaterialExpressionStaticSwitch
   inputs [
    True {from=MaterialExpressionLinearInterpolate_19, output=""}
    False {from=MaterialExpressionScalarParameter_156, output=""}
    Value {from=MaterialExpressionNamedRerouteUsage_71, output=""}
   ]
   properties {}
  }
  MaterialExpressionStaticSwitch_7 {
   class=MaterialExpressionStaticSwitch
   inputs [
    True {from=MaterialExpressionMultiply_24, output=""}
    False {from=MaterialExpressionScalarParameter_145, output=""}
    Value {from=MaterialExpressionNamedRerouteUsage_72, output=""}
   ]
   properties {}
  }
  MaterialExpressionStaticSwitch_8 {
   class=MaterialExpressionStaticSwitch
   inputs [
    True {from=MaterialExpressionMultiply_25, output=""}
    False {from=MaterialExpressionScalarParameter_1, output=""}
    Value {from=MaterialExpressionNamedRerouteUsage_73, output=""}
   ]
   properties {}
  }
  MaterialExpressionStaticSwitch_9 {
   class=MaterialExpressionStaticSwitch
   inputs [
    True {from="", output=""}
    False {from="", output=""}
    Value {from=MaterialExpressionNamedRerouteUsage_33, output=""}
   ]
   properties {}
  }
  MaterialExpressionStep_0 {
   class=MaterialExpressionStep
   inputs [
    Y {from=MaterialExpressionNamedRerouteUsage_56, output=""}
    X {from=MaterialExpressionNamedRerouteDeclaration_38, output=""}
   ]
   properties {}
  }
  MaterialExpressionStep_1 {
   class=MaterialExpressionStep
   inputs [
    Y {from=MaterialExpressionScalarParameter_31, output=""}
    X {from=ShellMeshMaterialExpressionShellLayerIndex_1, output=""}
   ]
   properties {}
  }
  MaterialExpressionStep_2 {
   class=MaterialExpressionStep
   inputs [
    Y {from="", output=""}
    X {from=MaterialExpressionNamedRerouteUsage_62, output=""}
   ]
   properties {}
  }
  MaterialExpressionStep_3 {
   class=MaterialExpressionStep
   inputs [
    Y {from=MaterialExpressionSubtract_3, output=""}
    X {from=MaterialExpressionFmod_0, output=""}
   ]
   properties {}
  }
  MaterialExpressionSubtract_0 {
   class=MaterialExpressionSubtract
   inputs [
    A {from=MaterialExpressionNamedRerouteUsage_91}
    B {from=MaterialExpressionScalarParameter_0}
   ]
  }
  MaterialExpressionSubtract_13 {
   class=MaterialExpressionSubtract
   inputs [
    A {from=MaterialExpressionNamedRerouteUsage_36}
    B {from=MaterialExpressionLinearInterpolate_67}
   ]
  }
  MaterialExpressionSubtract_2 {
   class=MaterialExpressionSubtract
   inputs [
    A {from=MaterialExpressionMultiply_8}
    B {from=""}
   ]
  }
  MaterialExpressionSubtract_3 {
   class=MaterialExpressionSubtract
   inputs [
    A {from=MaterialExpressionPower_0}
    B {from=""}
   ]
  }
  MaterialExpressionTextureCoordinate_1 {class=MaterialExpressionTextureCoordinate}
  MaterialExpressionTextureCoordinate_2 {class=MaterialExpressionTextureCoordinate}
  MaterialExpressionTextureCoordinate_6 {class=MaterialExpressionTextureCoordinate}
  MaterialExpressionTextureObjectParameter_0 {
   class=MaterialExpressionTextureObjectParameter
   inputs [
    UVs {from="", output=""}
    Apply View MipBias {from="", output=""}
   ]
   properties {
    group=Fur_Shell_02 - Strands
    parameterName=FurPattern
    samplerType=SAMPLERTYPE_Masks
    sortPriority=2
    texture=&T_WhiteMask
   }
  }
  MaterialExpressionTextureObjectParameter_1 {
   class=MaterialExpressionTextureObjectParameter
   inputs [
    UVs {from="", output=""}
    Apply View MipBias {from="", output=""}
   ]
   properties {
    group=Fur_Shell_02 - Strands
    parameterName=FurPatternNormal
    samplerType=SAMPLERTYPE_Normal
    sortPriority=10
    texture=&FlatNormalBlue
   }
  }
  MaterialExpressionTextureSampleParameter2D_20 {
   class=MaterialExpressionTextureSampleParameter2D
   inputs [
    UVs {from="", output=""}
    Apply View MipBias {from="", output=""}
   ]
   properties {
    desc="Map controls:\r\n\r\n- FurStyleBlendMask\r\n- FurLengthMap\r\n- FurClothingMask\r\n- FurNormalMask"
    group=Fur_Shell_01 - Base Settings
    parameterName=FurShellMap
    samplerSource=SSM_Wrap_WorldGroupSettings
    samplerType=SAMPLERTYPE_Masks
    sortPriority=0
    texture=/Game/Global/Textures/Default/Blanks/T_Blank_Mask_Green_R0G1B0A1
   }
  }
  MaterialExpressionTextureSampleParameter2D_21 {
   class=MaterialExpressionTextureSampleParameter2D
   inputs [
    UVs {from="", output=""}
    Apply View MipBias {from="", output=""}
   ]
   properties {
    desc=Flow Map is used to offset the fur UVs and Shell verts to drive the overall fur direction. Also used to inlfuence the distortion controls and as the Tangent for aniso
    group=Fur_Shell_03 - WPO
    parameterName=FurFlowMap
    samplerSource=SSM_Wrap_WorldGroupSettings
    samplerType=SAMPLERTYPE_LinearColor
    sortPriority=0
    texture=/Game/Global/Textures/T_Default_Aniso
   }
  }
  MaterialExpressionTextureSample_0 {
   class=MaterialExpressionTextureSample
   inputs [
    UVs {from=MaterialExpressionMultiply_10}
    Tex {from=""}
    Apply View MipBias {from=""}
   ]
   properties {samplerSource=SSM_Wrap_WorldGroupSettings, texture=/Game/Global/Textures/Default/Noise/Noise}
  }
  MaterialExpressionTextureSample_1 {
   class=MaterialExpressionTextureSample
   inputs [
    UVs {from=MaterialExpressionNamedRerouteUsage_135}
    Tex {from=MaterialExpressionNamedRerouteDeclaration_43}
    Apply View MipBias {from=""}
   ]
   properties {samplerSource=SSM_Wrap_WorldGroupSettings, samplerType=SAMPLERTYPE_Normal}
  }
  MaterialExpressionTextureSample_2 {
   class=MaterialExpressionTextureSample
   inputs [
    UVs {from=MaterialExpressionNamedRerouteDeclaration_86}
    Tex {from=MaterialExpressionNamedRerouteDeclaration_42}
    Bias {from=MaterialExpressionNamedRerouteDeclaration_20}
    Apply View MipBias {from=""}
   ]
   properties {mipValueMode=TMVM_MipBias, samplerSource=SSM_Wrap_WorldGroupSettings}
  }
  MaterialExpressionTextureSample_3 {
   class=MaterialExpressionTextureSample
   inputs [
    UVs {from=MaterialExpressionNamedRerouteDeclaration_37}
    Tex {from=MaterialExpressionNamedRerouteUsage_76}
    Bias {from=MaterialExpressionNamedRerouteUsage_74}
    Apply View MipBias {from=""}
   ]
   properties {mipValueMode=TMVM_MipBias, samplerSource=SSM_Wrap_WorldGroupSettings}
  }
  MaterialExpressionTextureSample_4 {
   class=MaterialExpressionTextureSample
   inputs [
    UVs {from=MaterialExpressionNamedRerouteUsage_59}
    Tex {from=MaterialExpressionNamedRerouteUsage_77}
    Apply View MipBias {from=""}
   ]
   properties {samplerSource=SSM_Wrap_WorldGroupSettings, samplerType=SAMPLERTYPE_Normal}
  }
  MaterialExpressionTransform_0 {
   class=MaterialExpressionTransform
   inputs [None {from=ShellMeshMaterialExpressionLocalShellOffset_0}]
   properties {transformSourceType=TRANSFORMSOURCE_Local}
  }
  MaterialExpressionTransform_1 {class=MaterialExpressionTransform, inputs [None {from=MaterialExpressionNamedRerouteUsage_5}]}
  MaterialExpressionTwoSidedSign_3 {class=MaterialExpressionTwoSidedSign, inputs [], properties {}}
  MaterialExpressionVertexInterpolator_0 {class=MaterialExpressionVertexInterpolator, inputs [VS {from=ShellMeshMaterialExpressionNormalizedShellLayer_0, output=""}], properties {}}
  MaterialExpressionVertexInterpolator_1 {class=MaterialExpressionVertexInterpolator, inputs [VS {from=ShellMeshMaterialExpressionShellLayerIndex_0, output=""}], properties {}}
  MaterialExpressionWorldPosition_0 {class=MaterialExpressionWorldPosition, inputs [], properties {worldPositionShaderOffset=WPT_CameraRelative}}
  ShellMeshMaterialExpressionLocalShellOffset_0 {class=ShellMeshMaterialExpressionLocalShellOffset, inputs [], properties {}}
  ShellMeshMaterialExpressionNormalizedShellLayer_0 {class=ShellMeshMaterialExpressionNormalizedShellLayer, inputs [], properties {}}
  ShellMeshMaterialExpressionShellCount_0 {class=ShellMeshMaterialExpressionShellCount, inputs [], properties {}}
  ShellMeshMaterialExpressionShellLayerIndex_0 {class=ShellMeshMaterialExpressionShellLayerIndex, inputs [], properties {}}
  ShellMeshMaterialExpressionShellLayerIndex_1 {class=ShellMeshMaterialExpressionShellLayerIndex, inputs [], properties {}}
  ShellMeshMaterialExpressionShellLayerIndex_2 {class=ShellMeshMaterialExpressionShellLayerIndex, inputs [], properties {}}
 }}
