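# Material function for the parent character fur material that adds optional subsurface shading to incoming material attributes.
# On the Epic quality tier it sets the Subsurface shading model, with a tint optionally multiplied by diffuse and scaled by an opacity parameter.
# Other tiers keep the shading model and use fixed opacity.
# A static switch gates the effect, and a same-named scalar blends it in, allowing per-variant shading model changes.
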MaterialFunction /Game/Global/Materials/Character_Parent_2023/Fur_Parent_2023/MF_Fur_Subsurface
graph {nodes [
  n1 = NamedRerouteUsage()
  n2 = SetMaterialAttributes!(MaterialAttributes=NamedRerouteUsage(), SubsurfaceColor=NamedRerouteUsage(), Opacity=Constant() {r=0.5}) {attributeSetTypes [#1, #2]}
  n3 = NamedRerouteUsage()
  n4 = NamedRerouteUsage()
  n5 = VectorParameter() {group=Fur_Shell_09 - Subsurface, parameterName=Fur SS Color, sortPriority=-3, defaultValue=<rgba 1 1 1 0>}
  n6 = ScalarParameter() {defaultValue=1, group=Fur_Shell_09 - Subsurface, parameterName=FurSSOpacity, sortPriority=-3}
  n7 = Multiply(
   A=StaticSwitchParameter(True=Multiply(A=@n4, B=@n5.RGB), False=@n5.RGB) {group=Fur_Shell_09 - Subsurface, parameterName=UseDiffuseInFurSSColor, sortPriority=-4}
   B=@n6
  )
  n8 = ShadingModel() {shadingModel=MSM_Subsurface}
  n9 = BlendMaterialAttributes(
   A=@n1
   B=MaterialFunctionCall!(
    Low _ PC Low and Med=@n2
    "High _ XB1, PS4"=@n2
    "Epic _ PC High, XSX, PS5"=SetMaterialAttributes!(MaterialAttributes=@n3, SubsurfaceColor=@n7, Opacity=@n6, ShadingModel=@n8) {attributeSetTypes [#1, #2, #3]}
    --------
    "Mobile Low _ PC Perf, Android"=@n2
    Mobile Med _ Switch=@n2
   ) {materialFunction=/Game/Global/MaterialFunctions/TechArt/MF_Fortnite_QualitySwitch_MaterialAttributes}.Result
   Alpha=NamedRerouteUsage()
  )
  n10 = ScalarParameter() {
   desc=Used for per-variant shading model change via float parameter.
   group=Fur_Shell_09 - Subsurface
   parameterName=UseFurSubsurface
   sortPriority=-5
  }
  FunctionOutput(None=StaticSwitchParameter(True=BlendMaterialAttributes(A=@n1, B=@n9, Alpha=@n10), False=@n1) {group=Fur_Shell_09 - Subsurface, parameterName=UseFurSubsurface, sortPriority=-5})
  n11 = GetMaterialAttributes(None=NamedRerouteDeclaration(None=FunctionInput() {bUsePreviewValueAsDefault=true, inputName=MA, inputType=FunctionInput_MaterialAttributes}) {name=IN_Material, nodeColor=<rgba 0.125 0 1 1>}) {attributeGetTypes [#4, #5, #6]}
  NamedRerouteDeclaration(None=@n11.MaterialAttributes) {name=In_MA, nodeColor=<rgba 1 0 0.5396 1>}
  NamedRerouteDeclaration(None=@n11.BaseColor) {name=In_BaseColor, nodeColor=<rgba 1 0 0.5396 1>}
  NamedRerouteDeclaration(None=@n11.Roughness) {name=In_Roughness, nodeColor=<rgba 1 0 0.5396 1>}
  NamedRerouteDeclaration(None=@n11.Metallic) {name=In_Metallic, nodeColor=<rgba 1 0 0.5396 1>}
  n12 = FunctionInput() {bUsePreviewValueAsDefault=true, inputName=TechArtMask, sortPriority=1, previewValue=<1 1 1 1>}
  NamedRerouteDeclaration(None=ChannelMaskParameter(None=@n12) {group=Fur_Shared, parameterName=ParamAdjustments_FurMaskChannel, sortPriority=-4}.R) {name=In_FurMask, nodeColor=<rgba 1 0 0.5396 1>}
 ]}
