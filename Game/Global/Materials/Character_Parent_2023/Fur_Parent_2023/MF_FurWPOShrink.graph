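# Material function for character fur that adds an optional World Position Offset shrink, pulling vertices inward along their normals.
# The inverted vertex normal is scaled by a shrink amount parameter, with a channel-selectable Mask input and a Fur Length input feeding the effect.
# The UseWPOShrink static switch chooses between material attributes with the adjusted offset and a branch without it.
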MaterialFunction /Game/Global/Materials/Character_Parent_2023/Fur_Parent_2023/MF_FurWPOShrink
graph {nodes [
  n1 = NamedRerouteUsage()
  n2 = Add(A=NamedRerouteUsage(), B=NamedRerouteUsage())
  FunctionOutput(None=StaticSwitchParameter(
    True=SetMaterialAttributes!(MaterialAttributes=@n1, WorldPositionOffset=@n2) {attributeSetTypes [#1]}
    False=NamedRerouteUsage()
   ) {group=Fur_Shared, parameterName=UseWPOShrink, sortPriority=0})
  n3 = Multiply(A=Multiply(A=VertexNormalWS()) {constB=-1}, B=ScalarParameter() {group=Fur_Shared, parameterName=WPO_ShrinkAmount})
  n4 = NamedRerouteUsage()
  NamedRerouteDeclaration(None=Multiply(A=Multiply(A=@n3, B=@n4), B=NamedRerouteUsage())) {name=WPO_ShrinkFur, nodeColor=<rgba 0 1 0.8594 1>}
  NamedRerouteDeclaration(None=ChannelMaskParameter(None=FunctionInput() {bUsePreviewValueAsDefault=true, inputName=Mask, sortPriority=1}) {group=Fur_Shared, parameterName=WPO_ShrinkMaskChannel, sortPriority=1}.R) {name=IN_Mask, nodeColor=<rgba 0.3828 0 1 1>}
  n5 = FunctionInput() {bUsePreviewValueAsDefault=true, inputName=Fur Length, inputType=FunctionInput_Scalar, sortPriority=2, previewValue=<1 0 0 1>}
  NamedRerouteDeclaration(None=@n5) {name=IN_FurLength, nodeColor=<rgba 0 1 0.0625 1>}
  n6 = FunctionInput() {inputName=Material, inputType=FunctionInput_MaterialAttributes}
  NamedRerouteDeclaration(None=GetMaterialAttributes(None=NamedRerouteDeclaration(None=@n6) {name=IN_Material, nodeColor=<rgba 0 0.9766 1 1>}) {attributeGetTypes [#1]}.WorldPositionOffset) {name=IN_WPO, nodeColor=<rgba 1 0 0.2578 1>}
 ]}
