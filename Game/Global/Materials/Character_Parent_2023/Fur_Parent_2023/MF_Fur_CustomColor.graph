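# Material function for the fur parent material that adds optional customizable color tinting to BaseColor, toggled by the UseCustomColors static switch.
# A primary color is layered with three accent colors through Custom Color Mask channels, then blended with a pattern color by a strength-scaled Pattern Mask.
# The tint multiplies the biased, power-adjusted base color, blends in through a selectable TechArtMask channel, and is written back into the material attributes.
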MaterialFunction /Game/Global/Materials/Character_Parent_2023/Fur_Parent_2023/MF_Fur_CustomColor
graph {nodes [
  n1 = NamedRerouteUsage()
  n2 = Saturate(None=Add(A=NamedRerouteUsage(), B=ScalarParameter() {group=CustomizableColor, parameterName=BaseColorValueBias, sortPriority=16}))
  n3 = ScalarParameter() {defaultValue=1, group=CustomizableColor, parameterName=BaseColorValuePow, sortPriority=17}
  n4 = VectorParameter() {group=CustomizableColor, parameterName=PrimaryColor, sortPriority=10, defaultValue=<rgba 1 1 1 1>}
  n5 = VectorParameter() {group=CustomizableColor, parameterName=AccentColor_01, sortPriority=12, defaultValue=<rgba 1 1 1 1>}
  n6 = NamedRerouteUsage()
  n7 = VectorParameter() {group=CustomizableColor, parameterName=AccentColor_02, sortPriority=14, defaultValue=<rgba 1 1 1 1>}
  n8 = VectorParameter() {group=CustomizableColor, parameterName=AccentColor_03, sortPriority=16, defaultValue=<rgba 1 1 1 1>}
  n9 = LinearInterpolate(
   A=LinearInterpolate(A=LinearInterpolate(A=@n4.RGB, B=@n5.RGB, Alpha=@n6) {constA=1}, B=@n7.RGB, Alpha=NamedRerouteUsage())
   B=@n8.RGB
   Alpha=NamedRerouteUsage()
  )
  n10 = VectorParameter() {group=CustomizableColor, parameterName=PatternColor, sortPriority=22, defaultValue=<rgba 1 1 1 1>}
  n11 = TextureSampleParameter2D() {
   group=CustomizableColor
   parameterName=Pattern Mask
   samplerSource=SSM_Wrap_WorldGroupSettings
   samplerType=SAMPLERTYPE_Masks
   sortPriority=20
   texture=&T_BlackMask
  }
  n12 = Multiply(
   A=ChannelMaskParameter(None=@n11.RGB) {group=CustomizableColor, parameterName=PatternMask_Channel, sortPriority=21}.R
   B=ScalarParameter() {group=CustomizableColor, parameterName=Pattern Strength, sortPriority=30}
  )
  n13 = LinearInterpolate(
   A=NamedRerouteUsage()
   B=Multiply(A=Power(Base=@n2, Exp=@n3), B=LinearInterpolate(A=@n9, B=@n10.RGB, Alpha=@n12) {constA=1})
   Alpha=NamedRerouteUsage()
  )
  FunctionOutput(None=StaticSwitchParameter(
    True=SetMaterialAttributes!(MaterialAttributes=@n1, BaseColor=@n13) {attributeSetTypes [#1]}
    False=NamedRerouteUsage()
   ) {group=CustomizableColor, parameterName=UseCustomColors, sortPriority=-1})
  n14 = TextureSampleParameter2D() {
   group=CustomizableColor
   parameterName=Custom Color Mask
   samplerSource=SSM_Wrap_WorldGroupSettings
   samplerType=SAMPLERTYPE_Masks
   sortPriority=0
   texture=&T_BlackMask
  }
  NamedRerouteDeclaration(None=ChannelMaskParameter(None=@n14.RGB) {group=CustomizableColor, parameterName=AccentColor_01_Channel, sortPriority=11}.R) {name=AccentColor_01_Mask, nodeColor=<rgba 1 0.0596 0.1687 1>}
  n15 = ChannelMaskParameter(None=@n14.RGB) {group=CustomizableColor, maskChannel=Green, parameterName=AccentColor_02_Channel, sortPriority=13, defaultValue=<rgba 0 1 0 0>}
  NamedRerouteDeclaration(None=@n15.R) {name=AccentColor_02_Mask, nodeColor=<rgba 0.4929 1 0.2743 1>}
  n16 = ChannelMaskParameter(None=@n14.RGB) {group=CustomizableColor, maskChannel=Blue, parameterName=AccentColor_03_Channel, sortPriority=15, defaultValue=<rgba 0 0 1 0>}
  NamedRerouteDeclaration(None=@n16.R) {name=AccentColor_03_Mask, nodeColor=<rgba 0 0.2243 1 1>}
  n17 = FunctionInput() {bUsePreviewValueAsDefault=true, inputName=TechArtMask, sortPriority=1, previewValue=<1 1 1 1>}
  NamedRerouteDeclaration(None=ChannelMaskParameter(None=@n17) {group=CustomizableColor, parameterName=PrimaryColorMask_Channel, sortPriority=1}.R) {name=IN_Mask, nodeColor=<rgba 0.6172 0 1 1>}
  n18 = FunctionInput() {inputName=Material, inputType=FunctionInput_MaterialAttributes}
  NamedRerouteDeclaration(None=GetMaterialAttributes(None=NamedRerouteDeclaration(None=@n18) {name=IN_Material, nodeColor=<rgba 0.6172 0 1 1>}) {attributeGetTypes [#1]}.BaseColor) {name=IN_BaseColor, nodeColor=<rgba 1 0 0.0234 1>}
 ]}
