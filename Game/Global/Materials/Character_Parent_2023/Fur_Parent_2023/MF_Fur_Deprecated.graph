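# Material function for character fur, named deprecated, that overrides incoming material attributes to shade anisotropic hair.
# Painted hair textures and masks drive roughness, specular, metallic, AO and blended normals, with optional FurID tinting of base color.
# A tangent map, flipped across the U midpoint and offset by "scraggle" noise, shapes anisotropic highlights.
# Sun-vector Fresnel with time-of-day color fakes backlit transmission, and parallax occlusion UVs apply on higher non-mobile quality levels.
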MaterialFunction /Game/Global/Materials/Character_Parent_2023/Fur_Parent_2023/MF_Fur_Deprecated
graph {expressions {
  MaterialExpressionAdd_1 {
   class=MaterialExpressionAdd
   inputs [
    A {from=MaterialExpressionMultiply_29}
    B {from=MaterialExpressionNamedRerouteUsage_9}
   ]
  }
  MaterialExpressionAdd_2 {
   class=MaterialExpressionAdd
   inputs [
    A {from=MaterialExpressionMultiply_9}
    B {from=MaterialExpressionScalarParameter_46}
   ]
  }
  MaterialExpressionAdd_3 {
   class=MaterialExpressionAdd
   inputs [
    A {from=MaterialExpressionNamedRerouteUsage_4}
    B {from=MaterialExpressionMultiply_12}
   ]
  }
  MaterialExpressionAppendVector_0 {
   class=MaterialExpressionAppendVector
   inputs [
    A {from=MaterialExpressionMultiply_18, output=""}
    B {from=MaterialExpressionConstant_10, output=""}
   ]
  }
  MaterialExpressionAtmosphericLightVector_0 {class=MaterialExpressionAtmosphericLightVector, inputs [], properties {}}
  MaterialExpressionAtmosphericLightVector_1 {class=MaterialExpressionAtmosphericLightVector, inputs [], properties {}}
  MaterialExpressionComponentMask_0 {class=MaterialExpressionComponentMask, inputs [None {from=MaterialExpressionTextureCoordinate_4}], properties {r=true}}
  MaterialExpressionComponentMask_11 {class=MaterialExpressionComponentMask, inputs [None {from=MaterialExpressionConvert_1}], properties {g=true, r=true}}
  MaterialExpressionConstant3Vector_1 {class=MaterialExpressionConstant3Vector, properties {constant=<rgba 1 1 1 0>}}
  MaterialExpressionConstant3Vector_2 {class=MaterialExpressionConstant3Vector, properties {constant=<rgba 0 0.0806 0.1875 1>}}
  MaterialExpressionConstant4Vector_2 {class=MaterialExpressionConstant4Vector, inputs [], properties {constant=<rgba 0 1 0 0>}}
  MaterialExpressionConstantBiasScale_0 {
   class=MaterialExpressionConstantBiasScale
   inputs [None {from=MaterialExpressionTextureSampleParameter2D_2, output=RGB}]
   properties {bias=-0.5, scale=2}
  }
  MaterialExpressionConstant_0 {class=MaterialExpressionConstant, properties {desc="Default 1, use Weight map 1", r=1}}
  MaterialExpressionConstant_1 {class=MaterialExpressionConstant}
  MaterialExpressionConstant_10 {class=MaterialExpressionConstant}
  MaterialExpressionConstant_14 {class=MaterialExpressionConstant, properties {r=2}}
  MaterialExpressionConstant_15 {class=MaterialExpressionConstant, properties {r=8}}
  MaterialExpressionConstant_2 {class=MaterialExpressionConstant, properties {r=1}}
  MaterialExpressionConstant_3 {class=MaterialExpressionConstant, properties {r=1}}
  MaterialExpressionConstant_4 {class=MaterialExpressionConstant, properties {r=0.2}}
  MaterialExpressionConstant_5 {class=MaterialExpressionConstant, properties {r=0.7972}}
  MaterialExpressionConstant_6 {class=MaterialExpressionConstant, properties {r=1}}
  MaterialExpressionConstant_7 {class=MaterialExpressionConstant, properties {r=1.5}}
  MaterialExpressionConstant_8 {class=MaterialExpressionConstant, properties {r=0.01}}
  MaterialExpressionConstant_9 {class=MaterialExpressionConstant, properties {r=1}}
  MaterialExpressionConvert_0 {
   class=MaterialExpressionConvert
   inputs [
    None {from=MaterialExpressionScalarParameter_44}
    None {from=MaterialExpressionScalarParameter_32}
    None {from=MaterialExpressionAdd_2}
   ]
   properties {
    nodeName=Make Float3
    convertInputs [
     {
      type=Scalar
      expressionInput {
       expression=&self:MaterialExpressionScalarParameter_44
       inputName=None
       mask=0
       maskA=0
       maskB=0
       maskG=0
       maskR=0
       outputIndex=0
      }
     }
     {
      type=Scalar
      expressionInput {
       expression=&self:MaterialExpressionScalarParameter_32
       inputName=None
       mask=0
       maskA=0
       maskB=0
       maskG=0
       maskR=0
       outputIndex=0
      }
     }
     {
      type=Scalar
      expressionInput {
       expression=&self:MaterialExpressionAdd_2
       inputName=None
       mask=0
       maskA=0
       maskB=0
       maskG=0
       maskR=0
       outputIndex=0
      }
     }
    ]
    convertOutputs [{type=Vector3}]
   }
  }
  MaterialExpressionConvert_1 {
   class=MaterialExpressionConvert
   inputs [
    None {from=MaterialExpressionTextureSampleParameter2D_2, output=R}
    None {from=MaterialExpressionOneMinus_1}
   ]
   properties {
    nodeName=Make Float2
    convertInputs [
     {
      type=Scalar
      expressionInput {
       expression=&self:MaterialExpressionTextureSampleParameter2D_2
       inputName=None
       mask=1
       maskA=0
       maskB=0
       maskG=0
       maskR=1
       outputIndex=1
      }
     }
     {
      type=Scalar
      expressionInput {
       expression=&self:MaterialExpressionOneMinus_1
       inputName=None
       mask=0
       maskA=0
       maskB=0
       maskG=0
       maskR=0
       outputIndex=0
      }
     }
    ]
    convertOutputs [{type=Vector2}]
   }
  }
  MaterialExpressionFresnel_0 {
   class=MaterialExpressionFresnel
   inputs [
    ExponentIn {from=MaterialExpressionConstant_7, output=""}
    BaseReflectFractionIn {from=MaterialExpressionConstant_8, output=""}
    Normal {from="", output=""}
   ]
   properties {baseReflectFraction=0.01, exponent=9}
  }
  MaterialExpressionFunctionInput_0 {
   class=MaterialExpressionFunctionInput
   inputs [Preview {from="", output=""}]
   properties {bUsePreviewValueAsDefault=true, inputName=Mask Input, sortPriority=1}
  }
  MaterialExpressionFunctionInput_2 {
   class=MaterialExpressionFunctionInput
   inputs [Preview {from="", output=""}]
   properties {bUsePreviewValueAsDefault=true, inputType=FunctionInput_MaterialAttributes}
  }
  MaterialExpressionFunctionOutput_0 {
   class=MaterialExpressionFunctionOutput
   inputs [None {from=MaterialExpressionStaticSwitchParameter_0, output=""}]
   properties {outputName=OutPut paint hair texture}
  }
  MaterialExpressionFunctionOutput_1 {class=MaterialExpressionFunctionOutput, inputs [None {from=MaterialExpressionSetMaterialAttributes_0, output=""}]}
  MaterialExpressionFunctionOutput_2 {
   class=MaterialExpressionFunctionOutput
   inputs [None {from=MaterialExpressionNamedRerouteDeclaration_0, output=""}]
   properties {outputName=AnisotropyWeight}
  }
  MaterialExpressionFunctionOutput_4 {class=MaterialExpressionFunctionOutput, inputs [None {from=MaterialExpressionTextureSample_2, output=R}], properties {outputName=HairMask}}
  MaterialExpressionFunctionOutput_5 {
   class=MaterialExpressionFunctionOutput
   inputs [None {from=MaterialExpressionNamedRerouteDeclaration_3, output=""}]
   properties {outputName=OutPutBasecolor}
  }
  MaterialExpressionGetMaterialAttributes_0 {
   class=MaterialExpressionGetMaterialAttributes
   inputs [None {from=MaterialExpressionFunctionInput_2, output=""}]
   properties {attributeGetTypes [
     #1
     #2
     #3
     #4
    ]}
  }
  MaterialExpressionLinearInterpolate_0 {
   class=MaterialExpressionLinearInterpolate
   inputs [
    A {from=MaterialExpressionReroute_3}
    B {from=MaterialExpressionMultiply_1}
    Alpha {from=MaterialExpressionConstant_0}
   ]
  }
  MaterialExpressionLinearInterpolate_1 {
   class=MaterialExpressionLinearInterpolate
   inputs [
    A {from=MaterialExpressionScalarParameter_1}
    B {from=""}
    Alpha {from=MaterialExpressionNamedRerouteUsage_43}
   ]
   properties {constA=0.6}
  }
  MaterialExpressionLinearInterpolate_10 {
   class=MaterialExpressionLinearInterpolate
   inputs [
    A {from=MaterialExpressionLinearInterpolate_2}
    B {from=MaterialExpressionSaturate_3}
    Alpha {from=MaterialExpressionTextureSampleParameter2D_0, output=G}
   ]
  }
  MaterialExpressionLinearInterpolate_11 {
   class=MaterialExpressionLinearInterpolate
   inputs [
    A {from=MaterialExpressionMaterialFunctionCall_8, output=R}
    B {from=MaterialExpressionNamedRerouteUsage_42}
    Alpha {from=""}
   ]
  }
  MaterialExpressionLinearInterpolate_12 {
   class=MaterialExpressionLinearInterpolate
   inputs [
    A {from=MaterialExpressionLinearInterpolate_13}
    B {from=MaterialExpressionMaterialFunctionCall_8, output=G}
    Alpha {from=""}
   ]
  }
  MaterialExpressionLinearInterpolate_13 {
   class=MaterialExpressionLinearInterpolate
   inputs [
    A {from=""}
    B {from=""}
    Alpha {from=MaterialExpressionTextureSample_1, output=G}
   ]
   properties {constA=0.5}
  }
  MaterialExpressionLinearInterpolate_17 {
   class=MaterialExpressionLinearInterpolate
   inputs [
    A {from=""}
    B {from=""}
    Alpha {from=MaterialExpressionNamedRerouteUsage_26}
   ]
   properties {constB=0.25}
  }
  MaterialExpressionLinearInterpolate_18 {
   class=MaterialExpressionLinearInterpolate
   inputs [
    A {from=MaterialExpressionNamedRerouteDeclaration_2}
    B {from=MaterialExpressionNamedRerouteUsage_7}
    Alpha {from=MaterialExpressionScalarParameter_13}
   ]
  }
  MaterialExpressionLinearInterpolate_19 {
   class=MaterialExpressionLinearInterpolate
   inputs [
    A {from=MaterialExpressionOneMinus_0}
    B {from=MaterialExpressionMaterialFunctionCall_2, output=R}
    Alpha {from=MaterialExpressionSaturate_7}
   ]
  }
  MaterialExpressionLinearInterpolate_2 {
   class=MaterialExpressionLinearInterpolate
   inputs [
    A {from=MaterialExpressionNamedRerouteUsage_11}
    B {from=MaterialExpressionSaturate_0}
    Alpha {from=MaterialExpressionTextureSampleParameter2D_0, output=R}
   ]
  }
  MaterialExpressionLinearInterpolate_20 {
   class=MaterialExpressionLinearInterpolate
   inputs [
    A {from=MaterialExpressionMaterialFunctionCall_2, output=G}
    B {from=MaterialExpressionOneMinus_4}
    Alpha {from=MaterialExpressionSaturate_7}
   ]
  }
  MaterialExpressionLinearInterpolate_21 {
   class=MaterialExpressionLinearInterpolate
   inputs [
    A {from=MaterialExpressionTextureCoordinate_2}
    B {from=MaterialExpressionNamedRerouteUsage_21}
    Alpha {from=MaterialExpressionScalarParameter_18}
   ]
  }
  MaterialExpressionLinearInterpolate_22 {
   class=MaterialExpressionLinearInterpolate
   inputs [
    A {from=MaterialExpressionTextureCoordinate_0}
    B {from=MaterialExpressionMaterialFunctionCall_18, output=Parallax UVs}
    Alpha {from=MaterialExpressionPower_3}
   ]
  }
  MaterialExpressionLinearInterpolate_23 {
   class=MaterialExpressionLinearInterpolate
   inputs [
    A {from=MaterialExpressionMaterialFunctionCall_10, output=Result}
    B {from=MaterialExpressionAppendVector_0}
    Alpha {from=MaterialExpressionConstant_1}
   ]
   properties {constAlpha=0}
  }
  MaterialExpressionLinearInterpolate_24 {
   class=MaterialExpressionLinearInterpolate
   inputs [
    A {from=MaterialExpressionMultiply_26}
    B {from=MaterialExpressionMultiply_15}
    Alpha {from=MaterialExpressionNamedRerouteUsage_2}
   ]
  }
  MaterialExpressionLinearInterpolate_3 {
   class=MaterialExpressionLinearInterpolate
   inputs [
    A {from=MaterialExpressionSaturate_8}
    B {from=MaterialExpressionMultiply_23}
    Alpha {from=MaterialExpressionScalarParameter_2}
   ]
  }
  MaterialExpressionLinearInterpolate_4 {
   class=MaterialExpressionLinearInterpolate
   inputs [
    A {from=MaterialExpressionScalarParameter_37}
    B {from=MaterialExpressionScalarParameter_38}
    Alpha {from=MaterialExpressionNamedRerouteUsage_29}
   ]
  }
  MaterialExpressionLinearInterpolate_5 {
   class=MaterialExpressionLinearInterpolate
   inputs [
    A {from=MaterialExpressionScalarParameter_6}
    B {from=""}
    Alpha {from=MaterialExpressionTextureSample_3, output=R}
   ]
   properties {constA=0.5}
  }
  MaterialExpressionLinearInterpolate_6 {
   class=MaterialExpressionLinearInterpolate
   inputs [
    A {from=MaterialExpressionScalarParameter_9}
    B {from=MaterialExpressionScalarParameter_8}
    Alpha {from=MaterialExpressionSaturate_2}
   ]
  }
  MaterialExpressionLinearInterpolate_7 {
   class=MaterialExpressionLinearInterpolate
   inputs [
    A {from=MaterialExpressionScalarParameter_4}
    B {from=MaterialExpressionScalarParameter_0}
    Alpha {from=MaterialExpressionNamedRerouteUsage_5}
   ]
   properties {constB=0.1}
  }
  MaterialExpressionLinearInterpolate_8 {
   class=MaterialExpressionLinearInterpolate
   inputs [
    A {from=MaterialExpressionScalarParameter_62}
    B {from=""}
    Alpha {from=MaterialExpressionMaterialFunctionCall_3, output=Result}
   ]
   properties {constB=0}
  }
  MaterialExpressionLinearInterpolate_9 {
   class=MaterialExpressionLinearInterpolate
   inputs [
    A {from=MaterialExpressionScalarParameter_30}
    B {from=MaterialExpressionScalarParameter_21}
    Alpha {from=MaterialExpressionNamedRerouteUsage_24}
   ]
  }
  MaterialExpressionMaterialFunctionCall_1 {
   class=MaterialExpressionMaterialFunctionCall
   inputs [
    In {from=MaterialExpressionTextureSample_1, output=G}
    Contrast {from=MaterialExpressionScalarParameter_5}
   ]
   properties {materialFunction=&CheapContrast}
  }
  MaterialExpressionMaterialFunctionCall_10 {
   class=MaterialExpressionMaterialFunctionCall
   inputs [
    X {from=MaterialExpressionLinearInterpolate_19}
    Y {from=MaterialExpressionLinearInterpolate_20}
    Z {from=MaterialExpressionMaterialFunctionCall_2, output=B}
   ]
   properties {materialFunction=&MakeFloat3}
  }
  MaterialExpressionMaterialFunctionCall_11 {
   class=MaterialExpressionMaterialFunctionCall
   inputs [
    BaseNormal {from=MaterialExpressionMaterialFunctionCall_13, output=Result}
    AdditionalNormal {from=MaterialExpressionMaterialFunctionCall_12, output=Result}
   ]
   properties {materialFunction=&BlendAngleCorrectedNormals}
  }
  MaterialExpressionMaterialFunctionCall_12 {
   class=MaterialExpressionMaterialFunctionCall
   inputs [
    Normal {from=MaterialExpressionTextureSampleParameter2D_5, output=RGB}
    Flatness {from=MaterialExpressionScalarParameter_16}
   ]
   properties {materialFunction=&FlattenNormal}
  }
  MaterialExpressionMaterialFunctionCall_13 {
   class=MaterialExpressionMaterialFunctionCall
   inputs [
    Normal {from=MaterialExpressionNamedRerouteUsage_14}
    Flatness {from=MaterialExpressionScalarParameter_25}
   ]
   properties {materialFunction=&FlattenNormal}
  }
  MaterialExpressionMaterialFunctionCall_18 {
   class=MaterialExpressionMaterialFunctionCall
   inputs [
    Heightmap Texture {from=MaterialExpressionTextureObjectParameter_3}
    Height Ratio {from=MaterialExpressionLinearInterpolate_8}
    Min Steps {from=MaterialExpressionConstant_14}
    Max Steps {from=MaterialExpressionConstant_15}
    UVs {from=""}
    Heightmap Channel {from=MaterialExpressionConstant4Vector_2, output=RGBA}
    Reference Plane {from=""}
    ----------------------------- {from=""}
    Use World Coordinates {from=""}
    Specify Manual Texture Size {from=""}
    Manual Texture Size {from=""}
    ------------------------------ {from=""}
    "Render Shadows (Occlusion Mapping)" {from=""}
    Light Vector {from=""}
    Shadow Steps {from=""}
    Shadow Penumbra {from=""}
    Transform To VertexNormal {from=""}
   ]
   properties {materialFunction=/Engine/Functions/Engine_MaterialFunctions01/Texturing/ParallaxOcclusionMapping}
  }
  MaterialExpressionMaterialFunctionCall_2 {
   class=MaterialExpressionMaterialFunctionCall
   inputs [Float3 {from=MaterialExpressionConstantBiasScale_0}]
   properties {materialFunction=&BreakOutFloat3Components}
  }
  MaterialExpressionMaterialFunctionCall_3 {
   class=MaterialExpressionMaterialFunctionCall
   inputs [
    In {from=MaterialExpressionTextureSample_2, output=G}
    Contrast {from=MaterialExpressionConstant_4}
   ]
   properties {materialFunction=&CheapContrast}
  }
  MaterialExpressionMaterialFunctionCall_4 {
   class=MaterialExpressionMaterialFunctionCall
   inputs [
    Brightness {from=MaterialExpressionConstant_5}
    Exponent {from=MaterialExpressionConstant_6}
    Camera Vector {from=MaterialExpressionNormalize_2}
    Vertex / Pixel Normal {from=MaterialExpressionPixelNormalWS_1}
   ]
   properties {materialFunction=&MF_Fresnel_FX}
  }
  MaterialExpressionMaterialFunctionCall_5 {
   class=MaterialExpressionMaterialFunctionCall
   inputs [
    Brightness {from=MaterialExpressionScalarParameter_26}
    Exponent {from=MaterialExpressionScalarParameter_27}
    Camera Vector {from=MaterialExpressionNormalize_1}
    Vertex / Pixel Normal {from=MaterialExpressionPixelNormalWS_0}
   ]
   properties {materialFunction=&MF_Fresnel_FX}
  }
  MaterialExpressionMaterialFunctionCall_8 {
   class=MaterialExpressionMaterialFunctionCall
   inputs [Float3 {from=MaterialExpressionFunctionInput_0}]
   properties {desc="R = AO\r\nG = GMap\r\nB = Skin", materialFunction=&BreakOutFloat3Components}
  }
  MaterialExpressionMaterialFunctionCall_9 {
   class=MaterialExpressionMaterialFunctionCall
   inputs [
    Color_Morning {from=MaterialExpressionMultiply_20}
    Color_Day {from=MaterialExpressionConstant3Vector_1}
    Color_Evening {from=MaterialExpressionMultiply_21}
    Color_Night {from=MaterialExpressionConstant3Vector_2}
   ]
   properties {materialFunction=&Color_Time_of_Day}
  }
  MaterialExpressionMultiply_0 {
   class=MaterialExpressionMultiply
   inputs [
    A {from=MaterialExpressionVectorParameter_3, output=RGB}
    B {from=MaterialExpressionNamedRerouteUsage_3}
   ]
  }
  MaterialExpressionMultiply_1 {
   class=MaterialExpressionMultiply
   inputs [
    A {from=MaterialExpressionReroute_3}
    B {from=MaterialExpressionNamedRerouteUsage_8}
   ]
  }
  MaterialExpressionMultiply_10 {
   class=MaterialExpressionMultiply
   inputs [
    A {from=MaterialExpressionTextureCoordinate_1}
    B {from=MaterialExpressionScalarParameter_22}
   ]
  }
  MaterialExpressionMultiply_12 {
   class=MaterialExpressionMultiply
   inputs [
    A {from=MaterialExpressionSubtract_4}
    B {from=MaterialExpressionScalarParameter_57}
   ]
  }
  MaterialExpressionMultiply_14 {
   class=MaterialExpressionMultiply
   inputs [
    A {from=MaterialExpressionVectorParameter_0, output=RGB}
    B {from=MaterialExpressionNamedRerouteUsage_15}
   ]
  }
  MaterialExpressionMultiply_15 {
   class=MaterialExpressionMultiply
   inputs [
    A {from=MaterialExpressionStaticSwitchParameter_0}
    B {from=MaterialExpressionVectorParameter_1, output=RGB}
   ]
  }
  MaterialExpressionMultiply_16 {
   class=MaterialExpressionMultiply
   inputs [
    A {from=MaterialExpressionReroute_10}
    B {from=MaterialExpressionMultiply_19}
   ]
  }
  MaterialExpressionMultiply_17 {
   class=MaterialExpressionMultiply
   inputs [
    A {from=MaterialExpressionLinearInterpolate_6}
    B {from=MaterialExpressionLinearInterpolate_5}
   ]
  }
  MaterialExpressionMultiply_18 {
   class=MaterialExpressionMultiply
   inputs [
    A {from=MaterialExpressionSubtract_8}
    B {from=""}
   ]
   properties {constB=2}
  }
  MaterialExpressionMultiply_19 {
   class=MaterialExpressionMultiply
   inputs [
    A {from=MaterialExpressionMultiply_5}
    B {from=MaterialExpressionMaterialFunctionCall_9, output=Result}
   ]
  }
  MaterialExpressionMultiply_20 {
   class=MaterialExpressionMultiply
   inputs [
    A {from=MaterialExpressionConstant3Vector_1}
    B {from=""}
   ]
   properties {constB=0.5}
  }
  MaterialExpressionMultiply_21 {
   class=MaterialExpressionMultiply
   inputs [
    A {from=MaterialExpressionConstant3Vector_1}
    B {from=""}
   ]
   properties {constB=0.5}
  }
  MaterialExpressionMultiply_23 {
   class=MaterialExpressionMultiply
   inputs [
    A {from=MaterialExpressionSaturate_8}
    B {from=MaterialExpressionNamedRerouteUsage_12}
   ]
  }
  MaterialExpressionMultiply_25 {
   class=MaterialExpressionMultiply
   inputs [
    A {from=MaterialExpressionNamedRerouteUsage_9}
    B {from=MaterialExpressionLinearInterpolate_18}
   ]
  }
  MaterialExpressionMultiply_26 {
   class=MaterialExpressionMultiply
   inputs [
    A {from=MaterialExpressionStaticSwitchParameter_0}
    B {from=MaterialExpressionVectorParameter_2, output=RGB}
   ]
   properties {constB=0.1}
  }
  MaterialExpressionMultiply_29 {
   class=MaterialExpressionMultiply
   inputs [
    A {from=MaterialExpressionNamedRerouteUsage_9}
    B {from=MaterialExpressionLinearInterpolate_18}
   ]
  }
  MaterialExpressionMultiply_3 {
   class=MaterialExpressionMultiply
   inputs [
    A {from=MaterialExpressionAtmosphericLightVector_0}
    B {from=""}
   ]
   properties {constB=-1}
  }
  MaterialExpressionMultiply_4 {
   class=MaterialExpressionMultiply
   inputs [
    A {from=MaterialExpressionNamedRerouteUsage_28}
    B {from=MaterialExpressionScalarParameter_3}
   ]
  }
  MaterialExpressionMultiply_5 {
   class=MaterialExpressionMultiply
   inputs [
    A {from=MaterialExpressionScalarParameter_7}
    B {from=MaterialExpressionPower_5}
   ]
  }
  MaterialExpressionMultiply_6 {
   class=MaterialExpressionMultiply
   inputs [
    A {from=MaterialExpressionSubtract_0}
    B {from=""}
   ]
   properties {constB=5}
  }
  MaterialExpressionMultiply_7 {
   class=MaterialExpressionMultiply
   inputs [
    A {from=MaterialExpressionAtmosphericLightVector_1}
    B {from=""}
   ]
   properties {constB=-1}
  }
  MaterialExpressionMultiply_8 {
   class=MaterialExpressionMultiply
   inputs [
    A {from=MaterialExpressionFresnel_0}
    B {from=MaterialExpressionMaterialFunctionCall_4, output=Fresnel Inverted}
   ]
  }
  MaterialExpressionMultiply_9 {
   class=MaterialExpressionMultiply
   inputs [
    A {from=MaterialExpressionTextureSample_0, output=R}
    B {from=MaterialExpressionScalarParameter_20}
   ]
  }
  MaterialExpressionNamedRerouteDeclaration_0 {
   class=MaterialExpressionNamedRerouteDeclaration
   inputs [None {from=MaterialExpressionTextureSampleParameter2D_2, output=A}]
   properties {name=InPut_AnisotropyWeight, nodeColor=<rgba 1 0 0.8437 1>}
  }
  MaterialExpressionNamedRerouteDeclaration_10 {
   class=MaterialExpressionNamedRerouteDeclaration
   inputs [None {from=MaterialExpressionReroute_31}]
   properties {name=Anisotropy, nodeColor=<rgba 0 0.6953 1 1>}
  }
  MaterialExpressionNamedRerouteDeclaration_11 {
   class=MaterialExpressionNamedRerouteDeclaration
   inputs [None {from=MaterialExpressionPower_1}]
   properties {name=Out specular, nodeColor=<rgba 0 0.2266 1 1>}
  }
  MaterialExpressionNamedRerouteDeclaration_12 {
   class=MaterialExpressionNamedRerouteDeclaration
   inputs [None {from=MaterialExpressionPower_0}]
   properties {name=OutPutRoughness, nodeColor=<rgba 1 0 0.6797 1>}
  }
  MaterialExpressionNamedRerouteDeclaration_13 {
   class=MaterialExpressionNamedRerouteDeclaration
   inputs [None {from=MaterialExpressionMaterialFunctionCall_1, output=Result}]
   properties {name=InPut_Hair_Texture, nodeColor=<rgba 0.1016 0 1 1>}
  }
  MaterialExpressionNamedRerouteDeclaration_14 {
   class=MaterialExpressionNamedRerouteDeclaration
   inputs [None {from=MaterialExpressionReroute_12}]
   properties {name=simulate light transmission, nodeColor=<rgba 0 1 0.8594 1>}
  }
  MaterialExpressionNamedRerouteDeclaration_15 {
   class=MaterialExpressionNamedRerouteDeclaration
   inputs [None {from=MaterialExpressionSaturate_9}]
   properties {name=fresnel_backlight, nodeColor=<rgba 0 1 0.2969 1>}
  }
  MaterialExpressionNamedRerouteDeclaration_16 {
   class=MaterialExpressionNamedRerouteDeclaration
   inputs [None {from=MaterialExpressionNormalize_0}]
   properties {name=Normal, nodeColor=<rgba 0.5 1 0 1>}
  }
  MaterialExpressionNamedRerouteDeclaration_17 {
   class=MaterialExpressionNamedRerouteDeclaration
   inputs [None {from=MaterialExpressionLinearInterpolate_11}]
   properties {name=InputAO, nodeColor=<rgba 1 0.8203 0 1>}
  }
  MaterialExpressionNamedRerouteDeclaration_18 {
   class=MaterialExpressionNamedRerouteDeclaration
   inputs [None {from=MaterialExpressionLinearInterpolate_23}]
   properties {name=InPutTangentMap, nodeColor=<rgba 0.875 1 0 1>}
  }
  MaterialExpressionNamedRerouteDeclaration_19 {
   class=MaterialExpressionNamedRerouteDeclaration
   inputs [None {from=MaterialExpressionLinearInterpolate_3}]
   properties {name=basecolor with fresnel, nodeColor=<rgba 0.8984 0 1 1>}
  }
  MaterialExpressionNamedRerouteDeclaration_2 {
   class=MaterialExpressionNamedRerouteDeclaration
   inputs [None {from=MaterialExpressionMaterialFunctionCall_5, output=Fresnel Inverted}]
   properties {name=fresnel, nodeColor=<rgba 0 1 0.2969 1>}
  }
  MaterialExpressionNamedRerouteDeclaration_20 {
   class=MaterialExpressionNamedRerouteDeclaration
   inputs [None {from=MaterialExpressionLinearInterpolate_12}]
   properties {name=InPutGmap, nodeColor=<rgba 0 1 0.1328 1>}
  }
  MaterialExpressionNamedRerouteDeclaration_3 {
   class=MaterialExpressionNamedRerouteDeclaration
   inputs [None {from=MaterialExpressionTextureSampleParameter2D_1, output=RGB}]
   properties {name=InPutBaseColor, nodeColor=<rgba 0.2422 1 0 1>}
  }
  MaterialExpressionNamedRerouteDeclaration_4 {
   class=MaterialExpressionNamedRerouteDeclaration
   inputs [None {from=MaterialExpressionAdd_3}]
   properties {name=Tangent, nodeColor=<rgba 1 0.9609 0 1>}
  }
  MaterialExpressionNamedRerouteDeclaration_5 {
   class=MaterialExpressionNamedRerouteDeclaration
   inputs [None {from=MaterialExpressionSaturate_4}]
   properties {name=OutPutBaseColor, nodeColor=<rgba 0.2891 1 0 1>}
  }
  MaterialExpressionNamedRerouteDeclaration_6 {
   class=MaterialExpressionNamedRerouteDeclaration
   inputs [None {from=MaterialExpressionGetMaterialAttributes_0, output=Normal}]
   properties {name=InPutNM, nodeColor=<rgba 1 0.5625 0 1>}
  }
  MaterialExpressionNamedRerouteDeclaration_8 {
   class=MaterialExpressionNamedRerouteDeclaration
   inputs [None {from=MaterialExpressionSaturate_10}]
   properties {name=Roughness, nodeColor=<rgba 1 0.4453 0 1>}
  }
  MaterialExpressionNamedRerouteDeclaration_9 {
   class=MaterialExpressionNamedRerouteDeclaration
   inputs [None {from=MaterialExpressionShadingPathSwitch_0}]
   properties {name=ParallaxUV, nodeColor=<rgba 0 1 0.4141 1>}
  }
  MaterialExpressionNamedRerouteUsage_0 {class=MaterialExpressionNamedRerouteUsage}
  MaterialExpressionNamedRerouteUsage_1 {class=MaterialExpressionNamedRerouteUsage}
  MaterialExpressionNamedRerouteUsage_11 {class=MaterialExpressionNamedRerouteUsage}
  MaterialExpressionNamedRerouteUsage_12 {class=MaterialExpressionNamedRerouteUsage}
  MaterialExpressionNamedRerouteUsage_13 {class=MaterialExpressionNamedRerouteUsage}
  MaterialExpressionNamedRerouteUsage_14 {class=MaterialExpressionNamedRerouteUsage}
  MaterialExpressionNamedRerouteUsage_15 {class=MaterialExpressionNamedRerouteUsage}
  MaterialExpressionNamedRerouteUsage_17 {class=MaterialExpressionNamedRerouteUsage}
  MaterialExpressionNamedRerouteUsage_18 {class=MaterialExpressionNamedRerouteUsage}
  MaterialExpressionNamedRerouteUsage_19 {class=MaterialExpressionNamedRerouteUsage}
  MaterialExpressionNamedRerouteUsage_2 {class=MaterialExpressionNamedRerouteUsage}
  MaterialExpressionNamedRerouteUsage_20 {class=MaterialExpressionNamedRerouteUsage}
  MaterialExpressionNamedRerouteUsage_21 {class=MaterialExpressionNamedRerouteUsage}
  MaterialExpressionNamedRerouteUsage_22 {class=MaterialExpressionNamedRerouteUsage}
  MaterialExpressionNamedRerouteUsage_23 {class=MaterialExpressionNamedRerouteUsage}
  MaterialExpressionNamedRerouteUsage_24 {class=MaterialExpressionNamedRerouteUsage}
  MaterialExpressionNamedRerouteUsage_25 {class=MaterialExpressionNamedRerouteUsage}
  MaterialExpressionNamedRerouteUsage_26 {class=MaterialExpressionNamedRerouteUsage}
  MaterialExpressionNamedRerouteUsage_27 {class=MaterialExpressionNamedRerouteUsage}
  MaterialExpressionNamedRerouteUsage_28 {class=MaterialExpressionNamedRerouteUsage}
  MaterialExpressionNamedRerouteUsage_29 {class=MaterialExpressionNamedRerouteUsage}
  MaterialExpressionNamedRerouteUsage_3 {class=MaterialExpressionNamedRerouteUsage}
  MaterialExpressionNamedRerouteUsage_32 {class=MaterialExpressionNamedRerouteUsage}
  MaterialExpressionNamedRerouteUsage_34 {class=MaterialExpressionNamedRerouteUsage}
  MaterialExpressionNamedRerouteUsage_4 {class=MaterialExpressionNamedRerouteUsage}
  MaterialExpressionNamedRerouteUsage_42 {class=MaterialExpressionNamedRerouteUsage}
  MaterialExpressionNamedRerouteUsage_43 {class=MaterialExpressionNamedRerouteUsage}
  MaterialExpressionNamedRerouteUsage_5 {class=MaterialExpressionNamedRerouteUsage}
  MaterialExpressionNamedRerouteUsage_6 {class=MaterialExpressionNamedRerouteUsage}
  MaterialExpressionNamedRerouteUsage_7 {class=MaterialExpressionNamedRerouteUsage}
  MaterialExpressionNamedRerouteUsage_8 {class=MaterialExpressionNamedRerouteUsage}
  MaterialExpressionNamedRerouteUsage_9 {class=MaterialExpressionNamedRerouteUsage}
  MaterialExpressionNormalize_0 {class=MaterialExpressionNormalize, inputs [VectorInput {from=MaterialExpressionMaterialFunctionCall_11, output=Result}]}
  MaterialExpressionNormalize_1 {class=MaterialExpressionNormalize, inputs [VectorInput {from=MaterialExpressionMultiply_3, output=""}]}
  MaterialExpressionNormalize_2 {class=MaterialExpressionNormalize, inputs [VectorInput {from=MaterialExpressionMultiply_7, output=""}]}
  MaterialExpressionOneMinus_0 {class=MaterialExpressionOneMinus, inputs [None {from=MaterialExpressionMaterialFunctionCall_2, output=R}]}
  MaterialExpressionOneMinus_1 {class=MaterialExpressionOneMinus, inputs [None {from=MaterialExpressionTextureSampleParameter2D_2, output=G}]}
  MaterialExpressionOneMinus_3 {class=MaterialExpressionOneMinus, inputs [None {from=MaterialExpressionNamedRerouteUsage_0}]}
  MaterialExpressionOneMinus_4 {class=MaterialExpressionOneMinus, inputs [None {from=MaterialExpressionMaterialFunctionCall_2, output=G}]}
  MaterialExpressionPixelNormalWS_0 {class=MaterialExpressionPixelNormalWS, inputs [], properties {}}
  MaterialExpressionPixelNormalWS_1 {class=MaterialExpressionPixelNormalWS, inputs [], properties {}}
  MaterialExpressionPower_0 {
   class=MaterialExpressionPower
   inputs [
    Base {from=MaterialExpressionOneMinus_3}
    Exp {from=MaterialExpressionConstant_2}
   ]
   properties {constExponent=4}
  }
  MaterialExpressionPower_1 {
   class=MaterialExpressionPower
   inputs [
    Base {from=MaterialExpressionNamedRerouteUsage_1}
    Exp {from=MaterialExpressionConstant_3}
   ]
  }
  MaterialExpressionPower_2 {
   class=MaterialExpressionPower
   inputs [
    Base {from=MaterialExpressionMultiply_8}
    Exp {from=MaterialExpressionConstant_9}
   ]
   properties {constExponent=3}
  }
  MaterialExpressionPower_3 {
   class=MaterialExpressionPower
   inputs [
    Base {from=MaterialExpressionNamedRerouteUsage_27}
    Exp {from=""}
   ]
   properties {constExponent=5}
  }
  MaterialExpressionPower_5 {
   class=MaterialExpressionPower
   inputs [
    Base {from=MaterialExpressionNamedRerouteUsage_20}
    Exp {from=""}
   ]
  }
  MaterialExpressionQualitySwitch_0 {
   class=MaterialExpressionQualitySwitch
   inputs [
    Default {from=MaterialExpressionTextureCoordinate_0, output=""}
    Low {from="", output=""}
    High {from=MaterialExpressionLinearInterpolate_22, output=""}
    Medium {from="", output=""}
    Epic {from=MaterialExpressionLinearInterpolate_22, output=""}
   ]
   properties {}
  }
  MaterialExpressionQualitySwitch_1 {
   class=MaterialExpressionQualitySwitch
   inputs [
    Default {from=MaterialExpressionTextureCoordinate_0, output=""}
    Low {from="", output=""}
    High {from="", output=""}
    Medium {from="", output=""}
    Epic {from="", output=""}
   ]
   properties {}
  }
  MaterialExpressionReroute_10 {class=MaterialExpressionReroute, inputs [None {from=MaterialExpressionMultiply_25}]}
  MaterialExpressionReroute_12 {class=MaterialExpressionReroute, inputs [None {from=MaterialExpressionMultiply_16}]}
  MaterialExpressionReroute_3 {class=MaterialExpressionReroute, inputs [None {from=MaterialExpressionLinearInterpolate_4}]}
  MaterialExpressionReroute_31 {class=MaterialExpressionReroute, inputs [None {from=MaterialExpressionLinearInterpolate_0}]}
  MaterialExpressionReroute_34 {class=MaterialExpressionReroute, inputs [None {from=MaterialExpressionSaturate_1}]}
  MaterialExpressionSaturate_0 {class=MaterialExpressionSaturate, inputs [None {from=MaterialExpressionMultiply_14}]}
  MaterialExpressionSaturate_1 {class=MaterialExpressionSaturate, inputs [None {from=MaterialExpressionLinearInterpolate_9}]}
  MaterialExpressionSaturate_10 {class=MaterialExpressionSaturate, inputs [None {from=MaterialExpressionMultiply_17}]}
  MaterialExpressionSaturate_2 {class=MaterialExpressionSaturate, inputs [None {from=MaterialExpressionNamedRerouteUsage_23}]}
  MaterialExpressionSaturate_3 {class=MaterialExpressionSaturate, inputs [None {from=MaterialExpressionMultiply_0}]}
  MaterialExpressionSaturate_4 {class=MaterialExpressionSaturate, inputs [None {from=MaterialExpressionLinearInterpolate_24}]}
  MaterialExpressionSaturate_7 {class=MaterialExpressionSaturate, inputs [None {from=MaterialExpressionMultiply_6}]}
  MaterialExpressionSaturate_8 {class=MaterialExpressionSaturate, inputs [None {from=MaterialExpressionAdd_1}]}
  MaterialExpressionSaturate_9 {class=MaterialExpressionSaturate, inputs [None {from=MaterialExpressionPower_2}]}
  MaterialExpressionScalarParameter_0 {class=MaterialExpressionScalarParameter, properties {defaultValue=0.2, group=FN_Fur, parameterName=Hair_Metallic_Max}}
  MaterialExpressionScalarParameter_1 {class=MaterialExpressionScalarParameter, properties {defaultValue=0.6, group=FN_Fur_Advanced, parameterName=AmbientOcclusion_Black}}
  MaterialExpressionScalarParameter_13 {class=MaterialExpressionScalarParameter, properties {defaultValue=1, group=FN_Fur_Advanced, parameterName=Simulated backlight}}
  MaterialExpressionScalarParameter_16 {
   class=MaterialExpressionScalarParameter
   properties {defaultValue=0.6, group=FN_Fur, parameterName=Paint_Hair_Normal_Flatness, sortPriority=2}
  }
  MaterialExpressionScalarParameter_18 {class=MaterialExpressionScalarParameter, properties {group=FN_Fur_Parallax, parameterName=Parallax_diffuse_intensity}}
  MaterialExpressionScalarParameter_2 {class=MaterialExpressionScalarParameter, properties {group=FN_Fur_Advanced, parameterName=Gmap_intensity, sortPriority=1}}
  MaterialExpressionScalarParameter_20 {class=MaterialExpressionScalarParameter, properties {defaultValue=10, group=FN_fur_Scraggle, parameterName=Scraggle_NoiseZ_intensity}}
  MaterialExpressionScalarParameter_21 {class=MaterialExpressionScalarParameter, properties {defaultValue=0.5, group=FN_Fur, parameterName=Hair_Specular_MAX}}
  MaterialExpressionScalarParameter_22 {
   class=MaterialExpressionScalarParameter
   properties {defaultValue=30, group=FN_fur_Scraggle, parameterName=Scraggle_NoiseZ_Tiling, sortPriority=5}
  }
  MaterialExpressionScalarParameter_25 {
   class=MaterialExpressionScalarParameter
   properties {defaultValue=0.6, group=FN_Fur, parameterName=Hair_Mesh_Normal_Flatness, sortPriority=2}
  }
  MaterialExpressionScalarParameter_26 {
   class=MaterialExpressionScalarParameter
   properties {defaultValue=0.5, group=FN_Fur_Advanced, parameterName=BaseColor_Fresnel_Brightness, sortPriority=1}
  }
  MaterialExpressionScalarParameter_27 {
   class=MaterialExpressionScalarParameter
   properties {defaultValue=3.5, group=FN_Fur_Advanced, parameterName=Basecolor_Fresnel_Exponent, sortPriority=1}
  }
  MaterialExpressionScalarParameter_3 {
   class=MaterialExpressionScalarParameter
   properties {defaultValue=90, group=FN_Fur_Advanced, parameterName=Roughness_Noise_Tiling, sortPriority=5}
  }
  MaterialExpressionScalarParameter_30 {class=MaterialExpressionScalarParameter, properties {defaultValue=0.3, group=FN_Fur, parameterName=Hair_Specular_MIN}}
  MaterialExpressionScalarParameter_32 {class=MaterialExpressionScalarParameter, properties {defaultValue=0.25, group=FN_fur_Scraggle, parameterName=Scraggle_Noise_Y}}
  MaterialExpressionScalarParameter_37 {class=MaterialExpressionScalarParameter, properties {defaultValue=0.5, group=FN_Fur, parameterName=Hair_Anisotropy0_Min}}
  MaterialExpressionScalarParameter_38 {class=MaterialExpressionScalarParameter, properties {defaultValue=0.75, group=FN_Fur, parameterName=Hair_Anisotropy0_Max}}
  MaterialExpressionScalarParameter_4 {class=MaterialExpressionScalarParameter, properties {defaultValue=0.02, group=FN_Fur, parameterName=Hair_Metallic_Min}}
  MaterialExpressionScalarParameter_44 {class=MaterialExpressionScalarParameter, properties {defaultValue=2.5, group=FN_fur_Scraggle, parameterName=Scraggle_Noise_X}}
  MaterialExpressionScalarParameter_46 {class=MaterialExpressionScalarParameter, properties {defaultValue=7, group=FN_fur_Scraggle, parameterName=Scraggle_Noise_Z}}
  MaterialExpressionScalarParameter_5 {class=MaterialExpressionScalarParameter, properties {group=FN_Fur, parameterName=Paint_Hair_Contrast}}
  MaterialExpressionScalarParameter_57 {class=MaterialExpressionScalarParameter, properties {defaultValue=1, group=FN_fur_Scraggle, parameterName=Scraggle}}
  MaterialExpressionScalarParameter_6 {
   class=MaterialExpressionScalarParameter
   properties {defaultValue=0.75, group=FN_Fur, parameterName=Hair_Noise_Roughness_Min, sortPriority=5}
  }
  MaterialExpressionScalarParameter_62 {class=MaterialExpressionScalarParameter, properties {defaultValue=0.006, group=FN_Fur_Parallax, parameterName=HeightRatio}}
  MaterialExpressionScalarParameter_7 {
   class=MaterialExpressionScalarParameter
   properties {defaultValue=1, group=FN_Fur_Advanced, parameterName=Emissive_Fur_Brightness, sortPriority=1}
  }
  MaterialExpressionScalarParameter_8 {class=MaterialExpressionScalarParameter, properties {defaultValue=0.84, group=FN_Fur, parameterName=Roughness Max, sortPriority=5}}
  MaterialExpressionScalarParameter_9 {class=MaterialExpressionScalarParameter, properties {defaultValue=0.75, group=FN_Fur, parameterName=Roughness Min, sortPriority=5}}
  MaterialExpressionSetMaterialAttributes_0 {
   class=MaterialExpressionSetMaterialAttributes
   inputs [
    MaterialAttributes {from=MaterialExpressionGetMaterialAttributes_0, output=MaterialAttributes}
    Roughness {from=MaterialExpressionNamedRerouteUsage_13, output=""}
    Specular {from=MaterialExpressionReroute_34, output=""}
    Metallic {from=MaterialExpressionLinearInterpolate_7, output=""}
    Anisotropy {from=MaterialExpressionNamedRerouteUsage_17, output=""}
    BaseColor {from=MaterialExpressionNamedRerouteUsage_19, output=""}
    Normal {from=MaterialExpressionNamedRerouteUsage_34, output=""}
    ShadingModel {from=MaterialExpressionShadingModel_2, output=""}
    EmissiveColor {from=MaterialExpressionNamedRerouteUsage_32, output=""}
    Tangent {from=MaterialExpressionNamedRerouteUsage_6, output=""}
    AmbientOcclusion {from=MaterialExpressionLinearInterpolate_1, output=""}
   ]
   properties {attributeSetTypes [
     #5
     #6
     #7
     #4
     #1
     #2
     #8
     #9
     #3
     #10
    ]}
  }
  MaterialExpressionShadingModel_2 {class=MaterialExpressionShadingModel, inputs [], properties {}}
  MaterialExpressionShadingPathSwitch_0 {
   class=MaterialExpressionShadingPathSwitch
   inputs [
    Default {from=MaterialExpressionQualitySwitch_0, output=""}
    Deferred {from="", output=""}
    Forward {from="", output=""}
    Mobile {from=MaterialExpressionQualitySwitch_1, output=""}
   ]
   properties {}
  }
  MaterialExpressionStaticSwitchParameter_0 {
   class=MaterialExpressionStaticSwitchParameter
   inputs [
    True {from=MaterialExpressionLinearInterpolate_10, output=""}
    False {from=MaterialExpressionNamedRerouteUsage_11, output=""}
   ]
   properties {group=FN_Fur, parameterName=UseFurID?}
  }
  MaterialExpressionSubtract_0 {
   class=MaterialExpressionSubtract
   inputs [
    A {from=MaterialExpressionComponentMask_0}
    B {from=""}
   ]
   properties {constB=0.5}
  }
  MaterialExpressionSubtract_4 {
   class=MaterialExpressionSubtract
   inputs [
    A {from=MaterialExpressionLinearInterpolate_17}
    B {from=MaterialExpressionConvert_0}
   ]
  }
  MaterialExpressionSubtract_8 {
   class=MaterialExpressionSubtract
   inputs [
    A {from=MaterialExpressionComponentMask_11}
    B {from=""}
   ]
   properties {constB=0.5}
  }
  MaterialExpressionTextureCoordinate_0 {class=MaterialExpressionTextureCoordinate}
  MaterialExpressionTextureCoordinate_1 {class=MaterialExpressionTextureCoordinate}
  MaterialExpressionTextureCoordinate_2 {class=MaterialExpressionTextureCoordinate}
  MaterialExpressionTextureCoordinate_3 {class=MaterialExpressionTextureCoordinate}
  MaterialExpressionTextureCoordinate_4 {class=MaterialExpressionTextureCoordinate}
  MaterialExpressionTextureObjectParameter_0 {
   class=MaterialExpressionTextureObjectParameter
   inputs [
    UVs {from="", output=""}
    Apply View MipBias {from="", output=""}
   ]
   properties {group=FN_Fur, parameterName=Hair Mask Height, samplerType=SAMPLERTYPE_Masks, texture=&T_BlackMask}
  }
  MaterialExpressionTextureObjectParameter_3 {
   class=MaterialExpressionTextureObjectParameter
   inputs [
    UVs {from="", output=""}
    Apply View MipBias {from="", output=""}
   ]
   properties {
    group=FN_Fur
    parameterName=Hair Mask Height
    samplerType=SAMPLERTYPE_LinearColor
    texture=/Game/Global/Textures/Default/Blanks/T_BlackLinearColor
   }
  }
  MaterialExpressionTextureSampleParameter2D_0 {
   class=MaterialExpressionTextureSampleParameter2D
   inputs [
    UVs {from=MaterialExpressionNamedRerouteUsage_22, output=""}
    Apply View MipBias {from="", output=""}
   ]
   properties {group=FN_Fur, parameterName=FUR_ID, texture=/Game/Global/Textures/Default/Blanks/T_MidgreysRGB}
  }
  MaterialExpressionTextureSampleParameter2D_1 {
   class=MaterialExpressionTextureSampleParameter2D
   inputs [
    UVs {from=MaterialExpressionLinearInterpolate_21, output=""}
    Apply View MipBias {from="", output=""}
   ]
   properties {group=00_Texture Inputs, parameterName=Diffuse, sortPriority=0, texture=&T_MidgreysRGB}
  }
  MaterialExpressionTextureSampleParameter2D_2 {
   class=MaterialExpressionTextureSampleParameter2D
   inputs [
    UVs {from="", output=""}
    Apply View MipBias {from="", output=""}
   ]
   properties {
    group=06_ Anisotropic
    parameterName=AnisotropicTangentWeight
    samplerType=SAMPLERTYPE_LinearColor
    texture=/Game/Global/Textures/T_Default_Aniso
   }
  }
  MaterialExpressionTextureSampleParameter2D_5 {
   class=MaterialExpressionTextureSampleParameter2D
   inputs [
    UVs {from=MaterialExpressionNamedRerouteUsage_25, output=""}
    Apply View MipBias {from="", output=""}
   ]
   properties {
    group=FN_Fur
    parameterName=Strands Normal
    samplerSource=SSM_Wrap_WorldGroupSettings
    samplerType=SAMPLERTYPE_Normal
    texture=&FlatNormalBlue
   }
  }
  MaterialExpressionTextureSample_0 {
   class=MaterialExpressionTextureSample
   inputs [
    UVs {from=MaterialExpressionMultiply_10}
    Tex {from=""}
    Apply View MipBias {from=""}
   ]
   properties {samplerSource=SSM_Wrap_WorldGroupSettings, texture=/Game/Global/Textures/Default/Noise/Noise}
  }
  MaterialExpressionTextureSample_1 {
   class=MaterialExpressionTextureSample
   inputs [
    UVs {from=MaterialExpressionNamedRerouteUsage_18}
    Tex {from=MaterialExpressionTextureObjectParameter_0}
    Apply View MipBias {from=""}
   ]
   properties {samplerType=SAMPLERTYPE_LinearColor}
  }
  MaterialExpressionTextureSample_2 {
   class=MaterialExpressionTextureSample
   inputs [
    UVs {from=MaterialExpressionTextureCoordinate_3}
    Tex {from=MaterialExpressionTextureObjectParameter_3}
    Apply View MipBias {from=""}
   ]
   properties {samplerType=SAMPLERTYPE_LinearColor}
  }
  MaterialExpressionTextureSample_3 {
   class=MaterialExpressionTextureSample
   inputs [
    UVs {from=MaterialExpressionMultiply_4}
    Tex {from=""}
    Apply View MipBias {from=""}
   ]
   properties {samplerSource=SSM_Wrap_WorldGroupSettings, texture=&Noise}
  }
  MaterialExpressionVectorParameter_0 {
   class=MaterialExpressionVectorParameter
   properties {group=FN_Fur, parameterName=Hair_Color_Variation, sortPriority=1, defaultValue=<rgba 1 0.0335 0 1>}
  }
  MaterialExpressionVectorParameter_1 {
   class=MaterialExpressionVectorParameter
   properties {group=FN_Fur, parameterName=Paint_Hair_Color_Brightness, sortPriority=1, defaultValue=<rgba 1 1 1 1>}
  }
  MaterialExpressionVectorParameter_2 {
   class=MaterialExpressionVectorParameter
   properties {group=FN_Fur, parameterName=Paint_Hair_Color_Darkness, sortPriority=1, defaultValue=<rgba 0.5 0.5 0.5 1>}
  }
  MaterialExpressionVectorParameter_3 {
   class=MaterialExpressionVectorParameter
   properties {group=FN_Fur, parameterName=Hair_Color_Variation2, sortPriority=1, defaultValue=<rgba 1 0.0335 0 1>}
  }
 }}
