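# Material function for Fortnite characters that layers painted fur shading over input material attributes, behind the UsePaintedFur static switch and a mask.
# Strand textures and tint parameters shape color, normal, roughness, specular, metallic and anisotropy.
# An anisotropic tangent map, flipped at the UV midline, gets noise-driven scraggle.
# Atmospheric-light Fresnel fakes backlit transmission in emissive, tinted by time of day and scaled by CosmeticsEmissiveMultiplier to tame glow in captures.
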MaterialFunction /Game/Global/Materials/Character_Parent_2023/Fur_Parent_2023/MF_Fur_Painted
graph {expressions {
  MaterialExpressionAdd_1 {
   class=MaterialExpressionAdd
   inputs [
    A {from=MaterialExpressionMultiply_29}
    B {from=MaterialExpressionReroute_6}
   ]
  }
  MaterialExpressionAdd_2 {
   class=MaterialExpressionAdd
   inputs [
    A {from=MaterialExpressionMultiply_9}
    B {from=MaterialExpressionScalarParameter_46}
   ]
  }
  MaterialExpressionAdd_3 {
   class=MaterialExpressionAdd
   inputs [
    A {from=MaterialExpressionNamedRerouteUsage_4}
    B {from=MaterialExpressionMultiply_12}
   ]
  }
  MaterialExpressionAtmosphericLightVector_0 {class=MaterialExpressionAtmosphericLightVector, inputs [], properties {}}
  MaterialExpressionAtmosphericLightVector_1 {class=MaterialExpressionAtmosphericLightVector, inputs [], properties {}}
  MaterialExpressionBlendMaterialAttributes_1 {
   class=MaterialExpressionBlendMaterialAttributes
   inputs [
    A {from=MaterialExpressionNamedRerouteUsage_30, output=""}
    B {from=MaterialExpressionSetMaterialAttributes_0, output=""}
    Alpha {from=MaterialExpressionNamedRerouteUsage_36, output=""}
   ]
   properties {}
  }
  MaterialExpressionChannelMaskParameter_4 {
   class=MaterialExpressionChannelMaskParameter
   inputs [None {from=MaterialExpressionFunctionInput_7, output=""}]
   properties {group=Fur_Painted_01 - Base Settings, parameterName=PaintedFur_MaskChannel, sortPriority=0}
  }
  MaterialExpressionCollectionParameter_0 {
   class=MaterialExpressionCollectionParameter
   inputs []
   properties {
    collection=&FortniteMaterialParameters
    desc="CAPTURE EMISSIVE-GLOW GATE  —  FORT-1106795\r\n\r\nWHY: Capture's controlled/dark lighting makes this glow blow out on dark hair && fur.This lets them dim/kill it for a shot without touching the authored look.\r\n\r\nDRIVEN BY: console cvar  Fort.Cosmetics.EmissiveMultiplier"
    parameterName=CosmeticsEmissiveMultiplier
   }
  }
  MaterialExpressionComponentMask_0 {class=MaterialExpressionComponentMask, inputs [None {from=MaterialExpressionTextureCoordinate_4}], properties {r=true}}
  MaterialExpressionConstant3Vector_1 {class=MaterialExpressionConstant3Vector, properties {constant=<rgba 1 1 1 0>}}
  MaterialExpressionConstant3Vector_2 {class=MaterialExpressionConstant3Vector, properties {constant=<rgba 0 0.0806 0.1875 1>}}
  MaterialExpressionConstantBiasScale_0 {
   class=MaterialExpressionConstantBiasScale
   inputs [None {from=MaterialExpressionTextureSampleParameter2D_2, output=RGB}]
   properties {bias=-0.5, scale=2}
  }
  MaterialExpressionConstant_2 {class=MaterialExpressionConstant, properties {r=1}}
  MaterialExpressionConstant_38 {class=MaterialExpressionConstant, properties {r=1}}
  MaterialExpressionConstant_5 {class=MaterialExpressionConstant, properties {r=0.7972}}
  MaterialExpressionConstant_6 {class=MaterialExpressionConstant, properties {r=1}}
  MaterialExpressionConstant_7 {class=MaterialExpressionConstant, properties {r=1.5}}
  MaterialExpressionConstant_8 {class=MaterialExpressionConstant, properties {r=0.01}}
  MaterialExpressionConstant_9 {class=MaterialExpressionConstant, properties {r=1}}
  MaterialExpressionConvert_0 {
   class=MaterialExpressionConvert
   inputs [
    None {from=MaterialExpressionScalarParameter_44}
    None {from=MaterialExpressionScalarParameter_32}
    None {from=MaterialExpressionAdd_2}
   ]
   properties {
    nodeName=Make Float3
    convertInputs [
     {
      type=Scalar
      expressionInput {
       expression=&self:MaterialExpressionScalarParameter_44
       inputName=None
       mask=0
       maskA=0
       maskB=0
       maskG=0
       maskR=0
       outputIndex=0
      }
     }
     {
      type=Scalar
      expressionInput {
       expression=&self:MaterialExpressionScalarParameter_32
       inputName=None
       mask=0
       maskA=0
       maskB=0
       maskG=0
       maskR=0
       outputIndex=0
      }
     }
     {
      type=Scalar
      expressionInput {
       expression=&self:MaterialExpressionAdd_2
       inputName=None
       mask=0
       maskA=0
       maskB=0
       maskG=0
       maskR=0
       outputIndex=0
      }
     }
    ]
    convertOutputs [{type=Vector3}]
   }
  }
  MaterialExpressionFresnel_0 {
   class=MaterialExpressionFresnel
   inputs [
    ExponentIn {from=MaterialExpressionConstant_7, output=""}
    BaseReflectFractionIn {from=MaterialExpressionConstant_8, output=""}
    Normal {from="", output=""}
   ]
   properties {baseReflectFraction=0.01, exponent=9}
  }
  MaterialExpressionFunctionInput_0 {
   class=MaterialExpressionFunctionInput
   inputs [Preview {from="", output=""}]
   properties {bUsePreviewValueAsDefault=true, inputName=Mask Input, sortPriority=1}
  }
  MaterialExpressionFunctionInput_2 {
   class=MaterialExpressionFunctionInput
   inputs [Preview {from="", output=""}]
   properties {bUsePreviewValueAsDefault=true, inputType=FunctionInput_MaterialAttributes}
  }
  MaterialExpressionFunctionInput_7 {
   class=MaterialExpressionFunctionInput
   inputs [Preview {from="", output=""}]
   properties {bUsePreviewValueAsDefault=true, inputName=TechArtMask, sortPriority=1, previewValue=<1 1 1 1>}
  }
  MaterialExpressionFunctionOutput_1 {class=MaterialExpressionFunctionOutput, inputs [None {from=MaterialExpressionStaticSwitchParameter_1, output=""}]}
  MaterialExpressionGetMaterialAttributes_0 {
   class=MaterialExpressionGetMaterialAttributes
   inputs [None {from=MaterialExpressionFunctionInput_2, output=""}]
   properties {attributeGetTypes [#1, #2]}
  }
  MaterialExpressionLinearInterpolate_1 {
   class=MaterialExpressionLinearInterpolate
   inputs [
    A {from=MaterialExpressionScalarParameter_1}
    B {from=""}
    Alpha {from=MaterialExpressionNamedRerouteUsage_43}
   ]
   properties {constA=0.6}
  }
  MaterialExpressionLinearInterpolate_11 {
   class=MaterialExpressionLinearInterpolate
   inputs [
    A {from=MaterialExpressionReroute_4}
    B {from=MaterialExpressionNamedRerouteUsage_42}
    Alpha {from=""}
   ]
  }
  MaterialExpressionLinearInterpolate_12 {
   class=MaterialExpressionLinearInterpolate
   inputs [
    A {from=MaterialExpressionLinearInterpolate_13}
    B {from=MaterialExpressionMaterialFunctionCall_8, output=G}
    Alpha {from=""}
   ]
  }
  MaterialExpressionLinearInterpolate_13 {
   class=MaterialExpressionLinearInterpolate
   inputs [
    A {from=""}
    B {from=""}
    Alpha {from=MaterialExpressionTextureSampleParameter2D_6, output=G}
   ]
   properties {constA=0.5}
  }
  MaterialExpressionLinearInterpolate_14 {
   class=MaterialExpressionLinearInterpolate
   inputs [
    A {from=MaterialExpressionScalarParameter_11}
    B {from=MaterialExpressionScalarParameter_10}
    Alpha {from=MaterialExpressionPower_11}
   ]
  }
  MaterialExpressionLinearInterpolate_17 {
   class=MaterialExpressionLinearInterpolate
   inputs [
    A {from=""}
    B {from=""}
    Alpha {from=MaterialExpressionNamedRerouteUsage_26}
   ]
   properties {constB=0.25}
  }
  MaterialExpressionLinearInterpolate_18 {
   class=MaterialExpressionLinearInterpolate
   inputs [
    A {from=MaterialExpressionNamedRerouteDeclaration_2}
    B {from=MaterialExpressionNamedRerouteUsage_7}
    Alpha {from=MaterialExpressionScalarParameter_13}
   ]
  }
  MaterialExpressionLinearInterpolate_19 {
   class=MaterialExpressionLinearInterpolate
   inputs [
    A {from=MaterialExpressionOneMinus_0}
    B {from=MaterialExpressionMaterialFunctionCall_2, output=R}
    Alpha {from=MaterialExpressionSaturate_7}
   ]
  }
  MaterialExpressionLinearInterpolate_20 {
   class=MaterialExpressionLinearInterpolate
   inputs [
    A {from=MaterialExpressionMaterialFunctionCall_2, output=G}
    B {from=MaterialExpressionOneMinus_4}
    Alpha {from=MaterialExpressionSaturate_7}
   ]
  }
  MaterialExpressionLinearInterpolate_24 {
   class=MaterialExpressionLinearInterpolate
   inputs [
    A {from=MaterialExpressionMultiply_26}
    B {from=MaterialExpressionMultiply_15}
    Alpha {from=MaterialExpressionNamedRerouteUsage_2}
   ]
  }
  MaterialExpressionLinearInterpolate_3 {
   class=MaterialExpressionLinearInterpolate
   inputs [
    A {from=MaterialExpressionSaturate_8}
    B {from=MaterialExpressionMultiply_23}
    Alpha {from=MaterialExpressionScalarParameter_2}
   ]
  }
  MaterialExpressionLinearInterpolate_4 {
   class=MaterialExpressionLinearInterpolate
   inputs [
    A {from=MaterialExpressionScalarParameter_37}
    B {from=MaterialExpressionScalarParameter_38}
    Alpha {from=MaterialExpressionNamedRerouteUsage_29}
   ]
  }
  MaterialExpressionLinearInterpolate_5 {
   class=MaterialExpressionLinearInterpolate
   inputs [
    A {from=MaterialExpressionScalarParameter_6}
    B {from=""}
    Alpha {from=MaterialExpressionTextureSample_3, output=R}
   ]
   properties {constA=0.5}
  }
  MaterialExpressionLinearInterpolate_6 {
   class=MaterialExpressionLinearInterpolate
   inputs [
    A {from=MaterialExpressionScalarParameter_9}
    B {from=MaterialExpressionScalarParameter_8}
    Alpha {from=MaterialExpressionSaturate_2}
   ]
  }
  MaterialExpressionLinearInterpolate_7 {
   class=MaterialExpressionLinearInterpolate
   inputs [
    A {from=MaterialExpressionScalarParameter_4}
    B {from=MaterialExpressionScalarParameter_0}
    Alpha {from=MaterialExpressionNamedRerouteUsage_5}
   ]
   properties {constB=0.1}
  }
  MaterialExpressionMaterialFunctionCall_1 {
   class=MaterialExpressionMaterialFunctionCall
   inputs [
    In {from=MaterialExpressionTextureSampleParameter2D_6, output=G}
    Contrast {from=MaterialExpressionScalarParameter_5}
   ]
   properties {materialFunction=&CheapContrast}
  }
  MaterialExpressionMaterialFunctionCall_10 {
   class=MaterialExpressionMaterialFunctionCall
   inputs [
    X {from=MaterialExpressionLinearInterpolate_19}
    Y {from=MaterialExpressionLinearInterpolate_20}
    Z {from=MaterialExpressionReroute_2}
   ]
   properties {materialFunction=&MakeFloat3}
  }
  MaterialExpressionMaterialFunctionCall_11 {
   class=MaterialExpressionMaterialFunctionCall
   inputs [
    BaseNormal {from=MaterialExpressionMaterialFunctionCall_13, output=Result}
    AdditionalNormal {from=MaterialExpressionMaterialFunctionCall_12, output=Result}
   ]
   properties {materialFunction=&BlendAngleCorrectedNormals}
  }
  MaterialExpressionMaterialFunctionCall_12 {
   class=MaterialExpressionMaterialFunctionCall
   inputs [
    Normal {from=MaterialExpressionTextureSampleParameter2D_5, output=RGB}
    Flatness {from=MaterialExpressionScalarParameter_16}
   ]
   properties {materialFunction=&FlattenNormal}
  }
  MaterialExpressionMaterialFunctionCall_13 {
   class=MaterialExpressionMaterialFunctionCall
   inputs [
    Normal {from=MaterialExpressionNamedRerouteUsage_14}
    Flatness {from=MaterialExpressionScalarParameter_25}
   ]
   properties {materialFunction=&FlattenNormal}
  }
  MaterialExpressionMaterialFunctionCall_2 {
   class=MaterialExpressionMaterialFunctionCall
   inputs [Float3 {from=MaterialExpressionConstantBiasScale_0}]
   properties {materialFunction=&BreakOutFloat3Components}
  }
  MaterialExpressionMaterialFunctionCall_4 {
   class=MaterialExpressionMaterialFunctionCall
   inputs [
    Brightness {from=MaterialExpressionConstant_5}
    Exponent {from=MaterialExpressionConstant_6}
    Camera Vector {from=MaterialExpressionNormalize_2}
    Vertex / Pixel Normal {from=MaterialExpressionPixelNormalWS_1}
   ]
   properties {materialFunction=&MF_Fresnel_FX}
  }
  MaterialExpressionMaterialFunctionCall_5 {
   class=MaterialExpressionMaterialFunctionCall
   inputs [
    Brightness {from=MaterialExpressionScalarParameter_26}
    Exponent {from=MaterialExpressionScalarParameter_27}
    Camera Vector {from=MaterialExpressionNormalize_1}
    Vertex / Pixel Normal {from=MaterialExpressionPixelNormalWS_0}
   ]
   properties {materialFunction=&MF_Fresnel_FX}
  }
  MaterialExpressionMaterialFunctionCall_8 {
   class=MaterialExpressionMaterialFunctionCall
   inputs [Float3 {from=MaterialExpressionNamedRerouteUsage_35}]
   properties {desc="R = AO\r\nG = GMap\r\nB = Skin", materialFunction=&BreakOutFloat3Components}
  }
  MaterialExpressionMaterialFunctionCall_9 {
   class=MaterialExpressionMaterialFunctionCall
   inputs [
    Color_Morning {from=MaterialExpressionMultiply_20}
    Color_Day {from=MaterialExpressionConstant3Vector_1}
    Color_Evening {from=MaterialExpressionMultiply_21}
    Color_Night {from=MaterialExpressionConstant3Vector_2}
   ]
   properties {materialFunction=&Color_Time_of_Day}
  }
  MaterialExpressionMultiply_1 {
   class=MaterialExpressionMultiply
   inputs [
    A {from=MaterialExpressionLinearInterpolate_4}
    B {from=MaterialExpressionNamedRerouteUsage_8}
   ]
  }
  MaterialExpressionMultiply_10 {
   class=MaterialExpressionMultiply
   inputs [
    A {from=MaterialExpressionTextureCoordinate_1}
    B {from=MaterialExpressionScalarParameter_22}
   ]
  }
  MaterialExpressionMultiply_12 {
   class=MaterialExpressionMultiply
   inputs [
    A {from=MaterialExpressionSubtract_4}
    B {from=MaterialExpressionScalarParameter_57}
   ]
  }
  MaterialExpressionMultiply_15 {
   class=MaterialExpressionMultiply
   inputs [
    A {from=MaterialExpressionNamedRerouteUsage_11}
    B {from=MaterialExpressionVectorParameter_1, output=RGB}
   ]
  }
  MaterialExpressionMultiply_16 {
   class=MaterialExpressionMultiply
   inputs [
    A {from=MaterialExpressionMultiply_25}
    B {from=MaterialExpressionMultiply_19}
   ]
  }
  MaterialExpressionMultiply_17 {
   class=MaterialExpressionMultiply
   inputs [
    A {from=MaterialExpressionLinearInterpolate_6}
    B {from=MaterialExpressionLinearInterpolate_5}
   ]
  }
  MaterialExpressionMultiply_19 {
   class=MaterialExpressionMultiply
   inputs [
    A {from=MaterialExpressionMultiply_5}
    B {from=MaterialExpressionMaterialFunctionCall_9, output=Result}
   ]
  }
  MaterialExpressionMultiply_2 {
   class=MaterialExpressionMultiply
   inputs [
    A {from=MaterialExpressionMultiply_16}
    B {from=MaterialExpressionCollectionParameter_0}
   ]
  }
  MaterialExpressionMultiply_20 {
   class=MaterialExpressionMultiply
   inputs [
    A {from=MaterialExpressionConstant3Vector_1}
    B {from=""}
   ]
   properties {constB=0.5}
  }
  MaterialExpressionMultiply_21 {
   class=MaterialExpressionMultiply
   inputs [
    A {from=MaterialExpressionConstant3Vector_1}
    B {from=""}
   ]
   properties {constB=0.5}
  }
  MaterialExpressionMultiply_23 {
   class=MaterialExpressionMultiply
   inputs [
    A {from=MaterialExpressionSaturate_8}
    B {from=MaterialExpressionNamedRerouteUsage_12}
   ]
  }
  MaterialExpressionMultiply_25 {
   class=MaterialExpressionMultiply
   inputs [
    A {from=MaterialExpressionReroute_5}
    B {from=MaterialExpressionLinearInterpolate_18}
   ]
  }
  MaterialExpressionMultiply_26 {
   class=MaterialExpressionMultiply
   inputs [
    A {from=MaterialExpressionReroute_1}
    B {from=MaterialExpressionVectorParameter_2, output=RGB}
   ]
   properties {constB=0.1}
  }
  MaterialExpressionMultiply_29 {
   class=MaterialExpressionMultiply
   inputs [
    A {from=MaterialExpressionSaturate_4}
    B {from=MaterialExpressionLinearInterpolate_18}
   ]
  }
  MaterialExpressionMultiply_3 {
   class=MaterialExpressionMultiply
   inputs [
    A {from=MaterialExpressionAtmosphericLightVector_0}
    B {from=""}
   ]
   properties {constB=-1}
  }
  MaterialExpressionMultiply_4 {
   class=MaterialExpressionMultiply
   inputs [
    A {from=MaterialExpressionTextureCoordinate_2}
    B {from=MaterialExpressionScalarParameter_3}
   ]
  }
  MaterialExpressionMultiply_5 {
   class=MaterialExpressionMultiply
   inputs [
    A {from=MaterialExpressionScalarParameter_7}
    B {from=MaterialExpressionPower_5}
   ]
  }
  MaterialExpressionMultiply_6 {
   class=MaterialExpressionMultiply
   inputs [
    A {from=MaterialExpressionSubtract_0}
    B {from=""}
   ]
   properties {constB=5}
  }
  MaterialExpressionMultiply_7 {
   class=MaterialExpressionMultiply
   inputs [
    A {from=MaterialExpressionAtmosphericLightVector_1}
    B {from=""}
   ]
   properties {constB=-1}
  }
  MaterialExpressionMultiply_8 {
   class=MaterialExpressionMultiply
   inputs [
    A {from=MaterialExpressionFresnel_0}
    B {from=MaterialExpressionMaterialFunctionCall_4, output=Fresnel Inverted}
   ]
  }
  MaterialExpressionMultiply_9 {
   class=MaterialExpressionMultiply
   inputs [
    A {from=MaterialExpressionTextureSample_0, output=R}
    B {from=MaterialExpressionScalarParameter_20}
   ]
  }
  MaterialExpressionNamedRerouteDeclaration_0 {
   class=MaterialExpressionNamedRerouteDeclaration
   inputs [None {from=MaterialExpressionTextureSampleParameter2D_2, output=A}]
   properties {name=AnisotropyWeight, nodeColor=<rgba 1 0 0.8437 1>}
  }
  MaterialExpressionNamedRerouteDeclaration_10 {
   class=MaterialExpressionNamedRerouteDeclaration
   inputs [None {from=MaterialExpressionMultiply_1}]
   properties {name=Anisotropy, nodeColor=<rgba 0 0.6953 1 1>}
  }
  MaterialExpressionNamedRerouteDeclaration_13 {
   class=MaterialExpressionNamedRerouteDeclaration
   inputs [None {from=MaterialExpressionMaterialFunctionCall_1, output=Result}]
   properties {name=Fur_Strand_Texture, nodeColor=<rgba 0.1016 0 1 1>}
  }
  MaterialExpressionNamedRerouteDeclaration_14 {
   class=MaterialExpressionNamedRerouteDeclaration
   inputs [None {from=MaterialExpressionMultiply_2}]
   properties {name=simulate light transmission, nodeColor=<rgba 0 1 0.8594 1>}
  }
  MaterialExpressionNamedRerouteDeclaration_15 {
   class=MaterialExpressionNamedRerouteDeclaration
   inputs [None {from=MaterialExpressionSaturate_9}]
   properties {name=fresnel_backlight, nodeColor=<rgba 0 1 0.2969 1>}
  }
  MaterialExpressionNamedRerouteDeclaration_16 {
   class=MaterialExpressionNamedRerouteDeclaration
   inputs [None {from=MaterialExpressionNormalize_0}]
   properties {name=Normal, nodeColor=<rgba 0.5 1 0 1>}
  }
  MaterialExpressionNamedRerouteDeclaration_17 {
   class=MaterialExpressionNamedRerouteDeclaration
   inputs [None {from=MaterialExpressionLinearInterpolate_11}]
   properties {name=InputAO, nodeColor=<rgba 1 0.8203 0 1>}
  }
  MaterialExpressionNamedRerouteDeclaration_18 {
   class=MaterialExpressionNamedRerouteDeclaration
   inputs [None {from=MaterialExpressionMaterialFunctionCall_10, output=Result}]
   properties {name=InPutTangentMap, nodeColor=<rgba 0.875 1 0 1>}
  }
  MaterialExpressionNamedRerouteDeclaration_19 {
   class=MaterialExpressionNamedRerouteDeclaration
   inputs [None {from=MaterialExpressionLinearInterpolate_3}]
   properties {name=basecolor with fresnel, nodeColor=<rgba 0.8984 0 1 1>}
  }
  MaterialExpressionNamedRerouteDeclaration_2 {
   class=MaterialExpressionNamedRerouteDeclaration
   inputs [None {from=MaterialExpressionMaterialFunctionCall_5, output=Fresnel Inverted}]
   properties {name=fresnel, nodeColor=<rgba 0 1 0.2969 1>}
  }
  MaterialExpressionNamedRerouteDeclaration_20 {
   class=MaterialExpressionNamedRerouteDeclaration
   inputs [None {from=MaterialExpressionLinearInterpolate_12}]
   properties {name=Fur_Gmap, nodeColor=<rgba 0 1 0.1328 1>}
  }
  MaterialExpressionNamedRerouteDeclaration_22 {
   class=MaterialExpressionNamedRerouteDeclaration
   inputs [None {from=MaterialExpressionFunctionInput_0}]
   properties {name=In_Mask, nodeColor=<rgba 1 0 0.5396 1>}
  }
  MaterialExpressionNamedRerouteDeclaration_25 {
   class=MaterialExpressionNamedRerouteDeclaration
   inputs [None {from=MaterialExpressionChannelMaskParameter_4, output=R}]
   properties {name=In_FurMask, nodeColor=<rgba 1 0 0.5396 1>}
  }
  MaterialExpressionNamedRerouteDeclaration_3 {
   class=MaterialExpressionNamedRerouteDeclaration
   inputs [None {from=MaterialExpressionGetMaterialAttributes_0, output=BaseColor}]
   properties {name=In_BaseColor, nodeColor=<rgba 1 0 0.5396 1>}
  }
  MaterialExpressionNamedRerouteDeclaration_4 {
   class=MaterialExpressionNamedRerouteDeclaration
   inputs [None {from=MaterialExpressionAdd_3}]
   properties {name=Tangent, nodeColor=<rgba 1 0.9609 0 1>}
  }
  MaterialExpressionNamedRerouteDeclaration_6 {
   class=MaterialExpressionNamedRerouteDeclaration
   inputs [None {from=MaterialExpressionGetMaterialAttributes_0, output=Normal}]
   properties {name=In_NM, nodeColor=<rgba 1 0 0.5396 1>}
  }
  MaterialExpressionNamedRerouteDeclaration_7 {
   class=MaterialExpressionNamedRerouteDeclaration
   inputs [None {from=MaterialExpressionSaturate_5}]
   properties {name=Specular, nodeColor=<rgba 0.2294 1 0.9718 1>}
  }
  MaterialExpressionNamedRerouteDeclaration_8 {
   class=MaterialExpressionNamedRerouteDeclaration
   inputs [None {from=MaterialExpressionSaturate_10}]
   properties {name=Roughness, nodeColor=<rgba 1 0.4453 0 1>}
  }
  MaterialExpressionNamedRerouteDeclaration_96 {
   class=MaterialExpressionNamedRerouteDeclaration
   inputs [None {from=MaterialExpressionGetMaterialAttributes_0, output=MaterialAttributes}]
   properties {name=In_MA, nodeColor=<rgba 1 0 0.5396 1>}
  }
  MaterialExpressionNamedRerouteUsage_0 {class=MaterialExpressionNamedRerouteUsage}
  MaterialExpressionNamedRerouteUsage_11 {class=MaterialExpressionNamedRerouteUsage}
  MaterialExpressionNamedRerouteUsage_12 {class=MaterialExpressionNamedRerouteUsage}
  MaterialExpressionNamedRerouteUsage_13 {class=MaterialExpressionNamedRerouteUsage}
  MaterialExpressionNamedRerouteUsage_14 {class=MaterialExpressionNamedRerouteUsage}
  MaterialExpressionNamedRerouteUsage_16 {class=MaterialExpressionNamedRerouteUsage}
  MaterialExpressionNamedRerouteUsage_17 {class=MaterialExpressionNamedRerouteUsage}
  MaterialExpressionNamedRerouteUsage_19 {class=MaterialExpressionNamedRerouteUsage}
  MaterialExpressionNamedRerouteUsage_2 {class=MaterialExpressionNamedRerouteUsage}
  MaterialExpressionNamedRerouteUsage_20 {class=MaterialExpressionNamedRerouteUsage}
  MaterialExpressionNamedRerouteUsage_26 {class=MaterialExpressionNamedRerouteUsage}
  MaterialExpressionNamedRerouteUsage_29 {class=MaterialExpressionNamedRerouteUsage}
  MaterialExpressionNamedRerouteUsage_30 {class=MaterialExpressionNamedRerouteUsage}
  MaterialExpressionNamedRerouteUsage_31 {class=MaterialExpressionNamedRerouteUsage}
  MaterialExpressionNamedRerouteUsage_32 {class=MaterialExpressionNamedRerouteUsage}
  MaterialExpressionNamedRerouteUsage_33 {class=MaterialExpressionNamedRerouteUsage}
  MaterialExpressionNamedRerouteUsage_34 {class=MaterialExpressionNamedRerouteUsage}
  MaterialExpressionNamedRerouteUsage_35 {class=MaterialExpressionNamedRerouteUsage}
  MaterialExpressionNamedRerouteUsage_36 {class=MaterialExpressionNamedRerouteUsage}
  MaterialExpressionNamedRerouteUsage_4 {class=MaterialExpressionNamedRerouteUsage}
  MaterialExpressionNamedRerouteUsage_42 {class=MaterialExpressionNamedRerouteUsage}
  MaterialExpressionNamedRerouteUsage_43 {class=MaterialExpressionNamedRerouteUsage}
  MaterialExpressionNamedRerouteUsage_5 {class=MaterialExpressionNamedRerouteUsage}
  MaterialExpressionNamedRerouteUsage_6 {class=MaterialExpressionNamedRerouteUsage}
  MaterialExpressionNamedRerouteUsage_7 {class=MaterialExpressionNamedRerouteUsage}
  MaterialExpressionNamedRerouteUsage_8 {class=MaterialExpressionNamedRerouteUsage}
  MaterialExpressionNamedRerouteUsage_80 {class=MaterialExpressionNamedRerouteUsage}
  MaterialExpressionNormalize_0 {class=MaterialExpressionNormalize, inputs [VectorInput {from=MaterialExpressionMaterialFunctionCall_11, output=Result}]}
  MaterialExpressionNormalize_1 {class=MaterialExpressionNormalize, inputs [VectorInput {from=MaterialExpressionMultiply_3, output=""}]}
  MaterialExpressionNormalize_2 {class=MaterialExpressionNormalize, inputs [VectorInput {from=MaterialExpressionMultiply_7, output=""}]}
  MaterialExpressionOneMinus_0 {class=MaterialExpressionOneMinus, inputs [None {from=MaterialExpressionMaterialFunctionCall_2, output=R}]}
  MaterialExpressionOneMinus_3 {class=MaterialExpressionOneMinus, inputs [None {from=MaterialExpressionNamedRerouteUsage_0}]}
  MaterialExpressionOneMinus_4 {class=MaterialExpressionOneMinus, inputs [None {from=MaterialExpressionMaterialFunctionCall_2, output=G}]}
  MaterialExpressionPixelNormalWS_0 {class=MaterialExpressionPixelNormalWS, inputs [], properties {}}
  MaterialExpressionPixelNormalWS_1 {class=MaterialExpressionPixelNormalWS, inputs [], properties {}}
  MaterialExpressionPower_0 {
   class=MaterialExpressionPower
   inputs [
    Base {from=MaterialExpressionOneMinus_3}
    Exp {from=MaterialExpressionConstant_2}
   ]
   properties {constExponent=4}
  }
  MaterialExpressionPower_11 {
   class=MaterialExpressionPower
   inputs [
    Base {from=MaterialExpressionNamedRerouteUsage_80}
    Exp {from=MaterialExpressionConstant_38}
   ]
  }
  MaterialExpressionPower_2 {
   class=MaterialExpressionPower
   inputs [
    Base {from=MaterialExpressionMultiply_8}
    Exp {from=MaterialExpressionConstant_9}
   ]
   properties {constExponent=3}
  }
  MaterialExpressionPower_5 {
   class=MaterialExpressionPower
   inputs [
    Base {from=MaterialExpressionNamedRerouteUsage_20}
    Exp {from=""}
   ]
  }
  MaterialExpressionReroute_1 {class=MaterialExpressionReroute, inputs [None {from=MaterialExpressionNamedRerouteUsage_11}]}
  MaterialExpressionReroute_2 {class=MaterialExpressionReroute, inputs [None {from=MaterialExpressionMaterialFunctionCall_2, output=B}]}
  MaterialExpressionReroute_4 {class=MaterialExpressionReroute, inputs [None {from=MaterialExpressionMaterialFunctionCall_8, output=R}]}
  MaterialExpressionReroute_5 {class=MaterialExpressionReroute, inputs [None {from=MaterialExpressionSaturate_4}]}
  MaterialExpressionReroute_6 {class=MaterialExpressionReroute, inputs [None {from=MaterialExpressionReroute_7}]}
  MaterialExpressionReroute_7 {class=MaterialExpressionReroute, inputs [None {from=MaterialExpressionSaturate_4}]}
  MaterialExpressionSaturate_10 {class=MaterialExpressionSaturate, inputs [None {from=MaterialExpressionMultiply_17}]}
  MaterialExpressionSaturate_2 {class=MaterialExpressionSaturate, inputs [None {from=MaterialExpressionPower_0}]}
  MaterialExpressionSaturate_4 {class=MaterialExpressionSaturate, inputs [None {from=MaterialExpressionLinearInterpolate_24}]}
  MaterialExpressionSaturate_5 {class=MaterialExpressionSaturate, inputs [None {from=MaterialExpressionLinearInterpolate_14}]}
  MaterialExpressionSaturate_7 {class=MaterialExpressionSaturate, inputs [None {from=MaterialExpressionMultiply_6}]}
  MaterialExpressionSaturate_8 {class=MaterialExpressionSaturate, inputs [None {from=MaterialExpressionAdd_1}]}
  MaterialExpressionSaturate_9 {class=MaterialExpressionSaturate, inputs [None {from=MaterialExpressionPower_2}]}
  MaterialExpressionScalarParameter_0 {class=MaterialExpressionScalarParameter, properties {defaultValue=0.2, group=Fur_Painted_01 - Base Settings, parameterName=Metallic_Max}}
  MaterialExpressionScalarParameter_1 {
   class=MaterialExpressionScalarParameter
   properties {defaultValue=0.6, group=Fur_Painted_02 - Advanced Settings, parameterName=AmbientOcclusion_Black}
  }
  MaterialExpressionScalarParameter_10 {class=MaterialExpressionScalarParameter, properties {defaultValue=0.5, group=Fur_Painted_01 - Base Settings, parameterName=Specular_Max}}
  MaterialExpressionScalarParameter_11 {class=MaterialExpressionScalarParameter, properties {defaultValue=0.3, group=Fur_Painted_01 - Base Settings, parameterName=Specular_Min}}
  MaterialExpressionScalarParameter_13 {
   class=MaterialExpressionScalarParameter
   properties {defaultValue=1, group=Fur_Painted_02 - Advanced Settings, parameterName=Simulated_Backlight, sortPriority=31}
  }
  MaterialExpressionScalarParameter_16 {
   class=MaterialExpressionScalarParameter
   properties {defaultValue=0.6, group=Fur_Painted_01 - Base Settings, parameterName=Fur_Normal_Flatness, sortPriority=2}
  }
  MaterialExpressionScalarParameter_2 {class=MaterialExpressionScalarParameter, properties {group=Fur_Painted_02 - Advanced Settings, parameterName=Gmap_Intensity, sortPriority=1}}
  MaterialExpressionScalarParameter_20 {
   class=MaterialExpressionScalarParameter
   properties {defaultValue=10, group=Fur_Painted_02 - Advanced Settings, parameterName=Scraggle_Noise_Z_intensity, sortPriority=33}
  }
  MaterialExpressionScalarParameter_22 {
   class=MaterialExpressionScalarParameter
   properties {defaultValue=30, group=Fur_Painted_02 - Advanced Settings, parameterName=Scraggle_Noise_Z_Tiling, sortPriority=33}
  }
  MaterialExpressionScalarParameter_25 {
   class=MaterialExpressionScalarParameter
   properties {defaultValue=0.6, group=Fur_Painted_01 - Base Settings, parameterName=Mesh_Normal_Flatness, sortPriority=2}
  }
  MaterialExpressionScalarParameter_26 {
   class=MaterialExpressionScalarParameter
   properties {defaultValue=0.5, group=Fur_Painted_02 - Advanced Settings, parameterName=BaseColor_Fresnel_Brightness, sortPriority=1}
  }
  MaterialExpressionScalarParameter_27 {
   class=MaterialExpressionScalarParameter
   properties {defaultValue=3.5, group=Fur_Painted_02 - Advanced Settings, parameterName=Basecolor_Fresnel_Exponent, sortPriority=1}
  }
  MaterialExpressionScalarParameter_3 {
   class=MaterialExpressionScalarParameter
   properties {defaultValue=90, group=Fur_Painted_01 - Base Settings, parameterName=Roughness_Noise_Tiling, sortPriority=5}
  }
  MaterialExpressionScalarParameter_32 {
   class=MaterialExpressionScalarParameter
   properties {defaultValue=0.25, group=Fur_Painted_02 - Advanced Settings, parameterName=Scraggle_Noise_Y, sortPriority=33}
  }
  MaterialExpressionScalarParameter_37 {
   class=MaterialExpressionScalarParameter
   properties {defaultValue=0.5, group=Fur_Painted_01 - Base Settings, parameterName=Fur_Anisotropy_Min}
  }
  MaterialExpressionScalarParameter_38 {
   class=MaterialExpressionScalarParameter
   properties {defaultValue=0.75, group=Fur_Painted_01 - Base Settings, parameterName=Fur_Anisotropy_Max}
  }
  MaterialExpressionScalarParameter_4 {class=MaterialExpressionScalarParameter, properties {defaultValue=0.02, group=Fur_Painted_01 - Base Settings, parameterName=Metallic_Min}}
  MaterialExpressionScalarParameter_44 {
   class=MaterialExpressionScalarParameter
   properties {defaultValue=2.5, group=Fur_Painted_02 - Advanced Settings, parameterName=Scraggle_Noise_X, sortPriority=33}
  }
  MaterialExpressionScalarParameter_46 {
   class=MaterialExpressionScalarParameter
   properties {defaultValue=7, group=Fur_Painted_02 - Advanced Settings, parameterName=Scraggle_Noise_Z, sortPriority=33}
  }
  MaterialExpressionScalarParameter_5 {class=MaterialExpressionScalarParameter, properties {group=Fur_Painted_01 - Base Settings, parameterName=Fur_Contrast, sortPriority=3}}
  MaterialExpressionScalarParameter_57 {
   class=MaterialExpressionScalarParameter
   properties {defaultValue=1, group=Fur_Painted_02 - Advanced Settings, parameterName=Scraggle_Strength}
  }
  MaterialExpressionScalarParameter_6 {
   class=MaterialExpressionScalarParameter
   properties {defaultValue=0.75, group=Fur_Painted_01 - Base Settings, parameterName=Roughness_Noise_Min, sortPriority=5}
  }
  MaterialExpressionScalarParameter_7 {
   class=MaterialExpressionScalarParameter
   properties {defaultValue=1, group=Fur_Painted_02 - Advanced Settings, parameterName=Emissive_Fur_Brightness, sortPriority=1}
  }
  MaterialExpressionScalarParameter_8 {
   class=MaterialExpressionScalarParameter
   properties {defaultValue=0.84, group=Fur_Painted_01 - Base Settings, parameterName=Roughness_Max, sortPriority=5}
  }
  MaterialExpressionScalarParameter_9 {
   class=MaterialExpressionScalarParameter
   properties {defaultValue=0.75, group=Fur_Painted_01 - Base Settings, parameterName=Roughness_Min, sortPriority=5}
  }
  MaterialExpressionSetMaterialAttributes_0 {
   class=MaterialExpressionSetMaterialAttributes
   inputs [
    MaterialAttributes {from=MaterialExpressionNamedRerouteUsage_33, output=""}
    Roughness {from=MaterialExpressionNamedRerouteUsage_13, output=""}
    Specular {from=MaterialExpressionNamedRerouteUsage_16, output=""}
    Metallic {from=MaterialExpressionLinearInterpolate_7, output=""}
    Anisotropy {from=MaterialExpressionNamedRerouteUsage_17, output=""}
    BaseColor {from=MaterialExpressionNamedRerouteUsage_19, output=""}
    Normal {from=MaterialExpressionNamedRerouteUsage_34, output=""}
    ShadingModel {from=MaterialExpressionShadingModel_2, output=""}
    EmissiveColor {from=MaterialExpressionNamedRerouteUsage_32, output=""}
    Tangent {from=MaterialExpressionNamedRerouteUsage_6, output=""}
    AmbientOcclusion {from=MaterialExpressionLinearInterpolate_1, output=""}
   ]
   properties {attributeSetTypes [
     #3
     #4
     #5
     #6
     #1
     #2
     #7
     #8
     #9
     #10
    ]}
  }
  MaterialExpressionShadingModel_2 {class=MaterialExpressionShadingModel, inputs [], properties {}}
  MaterialExpressionStaticSwitchParameter_1 {
   class=MaterialExpressionStaticSwitchParameter
   inputs [
    True {from=MaterialExpressionBlendMaterialAttributes_1, output=""}
    False {from=MaterialExpressionNamedRerouteUsage_31, output=""}
   ]
   properties {group=Fur_Painted_01 - Base Settings, parameterName=UsePaintedFur, sortPriority=-1}
  }
  MaterialExpressionSubtract_0 {
   class=MaterialExpressionSubtract
   inputs [
    A {from=MaterialExpressionComponentMask_0}
    B {from=""}
   ]
   properties {constB=0.5}
  }
  MaterialExpressionSubtract_4 {
   class=MaterialExpressionSubtract
   inputs [
    A {from=MaterialExpressionLinearInterpolate_17}
    B {from=MaterialExpressionConvert_0}
   ]
  }
  MaterialExpressionTextureCoordinate_1 {class=MaterialExpressionTextureCoordinate}
  MaterialExpressionTextureCoordinate_2 {class=MaterialExpressionTextureCoordinate}
  MaterialExpressionTextureCoordinate_4 {class=MaterialExpressionTextureCoordinate}
  MaterialExpressionTextureSampleParameter2D_2 {
   class=MaterialExpressionTextureSampleParameter2D
   inputs [
    UVs {from="", output=""}
    Apply View MipBias {from="", output=""}
   ]
   properties {
    group=06_ Anisotropic
    parameterName=AnisotropicTangentWeight
    samplerSource=SSM_Wrap_WorldGroupSettings
    samplerType=SAMPLERTYPE_LinearColor
    texture=/Game/Global/Textures/T_Default_Aniso
   }
  }
  MaterialExpressionTextureSampleParameter2D_5 {
   class=MaterialExpressionTextureSampleParameter2D
   inputs [
    UVs {from="", output=""}
    Apply View MipBias {from="", output=""}
   ]
   properties {
    group=Fur_Painted_01 - Base Settings
    parameterName=Strands Normal
    samplerSource=SSM_Wrap_WorldGroupSettings
    samplerType=SAMPLERTYPE_Normal
    sortPriority=1
    texture=&FlatNormalBlue
   }
  }
  MaterialExpressionTextureSampleParameter2D_6 {
   class=MaterialExpressionTextureSampleParameter2D
   inputs [
    UVs {from="", output=""}
    Apply View MipBias {from="", output=""}
   ]
   properties {
    group=Fur_Painted_01 - Base Settings
    parameterName=Strand Map
    samplerSource=SSM_Wrap_WorldGroupSettings
    samplerType=SAMPLERTYPE_Masks
    sortPriority=1
    texture=/Game/Global/Textures/Default/Blanks/T_MidgreyMask
   }
  }
  MaterialExpressionTextureSample_0 {
   class=MaterialExpressionTextureSample
   inputs [
    UVs {from=MaterialExpressionMultiply_10}
    Tex {from=""}
    Apply View MipBias {from=""}
   ]
   properties {samplerSource=SSM_Wrap_WorldGroupSettings, texture=/Game/Global/Textures/Default/Noise/Noise}
  }
  MaterialExpressionTextureSample_3 {
   class=MaterialExpressionTextureSample
   inputs [
    UVs {from=MaterialExpressionMultiply_4}
    Tex {from=""}
    Apply View MipBias {from=""}
   ]
   properties {samplerSource=SSM_Wrap_WorldGroupSettings, texture=&Noise}
  }
  MaterialExpressionVectorParameter_1 {
   class=MaterialExpressionVectorParameter
   properties {group=Fur_Painted_01 - Base Settings, parameterName=Fur_Color_Brightness, sortPriority=2, defaultValue=<rgba 1 1 1 1>}
  }
  MaterialExpressionVectorParameter_2 {
   class=MaterialExpressionVectorParameter
   properties {group=Fur_Painted_01 - Base Settings, parameterName=Fur_Color_Darkness, sortPriority=2, defaultValue=<rgba 0.5 0.5 0.5 1>}
  }
 }}
