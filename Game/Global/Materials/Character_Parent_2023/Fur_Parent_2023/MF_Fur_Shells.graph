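# Material function that builds shell-based fur for characters, taking material attributes and a tech-art mask and also outputting fur length.
# Texture channels drive length, normal, clothing and secondary-pattern masks, shaping the per-shell pattern and normals.
# FOV-adjusted pixel distance blends to far-fur textures and thins shell layers through log2/fmod LOD masks.
# A custom NaN output culls shells, and a sine hash supplies per-strand variation.
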MaterialFunction /Game/Global/Materials/Character_Parent_2023/Fur_Parent_2023/MF_Fur_Shells
graph {expressions {
  MaterialExpressionAdd_10 {
   class=MaterialExpressionAdd
   inputs [
    A {from=MaterialExpressionScalarParameter_77}
    B {from=MaterialExpressionScalarParameter_78}
   ]
  }
  MaterialExpressionAdd_13 {
   class=MaterialExpressionAdd
   inputs [
    A {from=MaterialExpressionScalarParameter_82}
    B {from=MaterialExpressionScalarParameter_83}
   ]
  }
  MaterialExpressionAdd_14 {
   class=MaterialExpressionAdd
   inputs [
    A {from=MaterialExpressionNamedRerouteUsage_127}
    B {from=MaterialExpressionMultiply_50}
   ]
  }
  MaterialExpressionAdd_15 {
   class=MaterialExpressionAdd
   inputs [
    A {from=MaterialExpressionScalarParameter_85}
    B {from=MaterialExpressionScalarParameter_86}
   ]
  }
  MaterialExpressionAdd_16 {
   class=MaterialExpressionAdd
   inputs [
    A {from=MaterialExpressionStaticSwitchParameter_7}
    B {from=MaterialExpressionMultiply_56}
   ]
  }
  MaterialExpressionAdd_17 {
   class=MaterialExpressionAdd
   inputs [
    A {from=MaterialExpressionAppendVector_7}
    B {from=MaterialExpressionTextureCoordinate_0}
   ]
  }
  MaterialExpressionAdd_18 {
   class=MaterialExpressionAdd
   inputs [
    A {from=MaterialExpressionMultiply_60}
    B {from=MaterialExpressionNamedRerouteDeclaration_118}
   ]
  }
  MaterialExpressionAdd_19 {
   class=MaterialExpressionAdd
   inputs [
    A {from=MaterialExpressionAdd_4}
    B {from=MaterialExpressionNamedRerouteUsage_179}
   ]
  }
  MaterialExpressionAdd_3 {
   class=MaterialExpressionAdd
   inputs [
    A {from=MaterialExpressionAppendVector_4}
    B {from=MaterialExpressionTextureCoordinate_0}
   ]
  }
  MaterialExpressionAdd_4 {
   class=MaterialExpressionAdd
   inputs [
    A {from=MaterialExpressionNamedRerouteUsage_78}
    B {from=MaterialExpressionNamedRerouteUsage_79}
   ]
  }
  MaterialExpressionAdd_44 {
   class=MaterialExpressionAdd
   inputs [
    A {from=MaterialExpressionScalarParameter_147}
    B {from=MaterialExpressionScalarParameter_148}
   ]
  }
  MaterialExpressionAdd_5 {
   class=MaterialExpressionAdd
   inputs [
    A {from=MaterialExpressionNamedRerouteUsage_34}
    B {from=MaterialExpressionMultiply_38}
   ]
  }
  MaterialExpressionAdd_7 {
   class=MaterialExpressionAdd
   inputs [
    A {from=MaterialExpressionNamedRerouteUsage_118}
    B {from=""}
   ]
   properties {constB=0.1}
  }
  MaterialExpressionAppendVector_10 {
   class=MaterialExpressionAppendVector
   inputs [
    A {from=MaterialExpressionScalarParameter_104, output=""}
    B {from=MaterialExpressionScalarParameter_104, output=""}
   ]
  }
  MaterialExpressionAppendVector_11 {
   class=MaterialExpressionAppendVector
   inputs [
    A {from=MaterialExpressionAppendVector_10, output=""}
    B {from=MaterialExpressionConstant_18, output=""}
   ]
  }
  MaterialExpressionAppendVector_2 {
   class=MaterialExpressionAppendVector
   inputs [
    A {from=MaterialExpressionScalarParameter_11, output=""}
    B {from=MaterialExpressionScalarParameter_11, output=""}
   ]
  }
  MaterialExpressionAppendVector_3 {
   class=MaterialExpressionAppendVector
   inputs [
    A {from=MaterialExpressionAppendVector_2, output=""}
    B {from=MaterialExpressionConstant_3, output=""}
   ]
  }
  MaterialExpressionAppendVector_4 {
   class=MaterialExpressionAppendVector
   inputs [
    A {from=MaterialExpressionScalarParameter_51, output=""}
    B {from=MaterialExpressionScalarParameter_52, output=""}
   ]
  }
  MaterialExpressionAppendVector_5 {
   class=MaterialExpressionAppendVector
   inputs [
    A {from=MaterialExpressionScalarParameter_90, output=""}
    B {from=MaterialExpressionScalarParameter_90, output=""}
   ]
  }
  MaterialExpressionAppendVector_6 {
   class=MaterialExpressionAppendVector
   inputs [
    A {from=MaterialExpressionAppendVector_5, output=""}
    B {from=MaterialExpressionConstant_15, output=""}
   ]
  }
  MaterialExpressionAppendVector_7 {
   class=MaterialExpressionAppendVector
   inputs [
    A {from=MaterialExpressionScalarParameter_99, output=""}
    B {from=MaterialExpressionScalarParameter_100, output=""}
   ]
  }
  MaterialExpressionAppendVector_8 {
   class=MaterialExpressionAppendVector
   inputs [
    A {from=MaterialExpressionScalarParameter_103, output=""}
    B {from=MaterialExpressionScalarParameter_103, output=""}
   ]
  }
  MaterialExpressionAppendVector_9 {
   class=MaterialExpressionAppendVector
   inputs [
    A {from=MaterialExpressionAppendVector_8, output=""}
    B {from=MaterialExpressionConstant_17, output=""}
   ]
  }
  MaterialExpressionAtmosphericLightVector_0 {class=MaterialExpressionAtmosphericLightVector, inputs [], properties {}}
  MaterialExpressionCeil_1 {class=MaterialExpressionCeil, inputs [None {from=MaterialExpressionLogarithm2_1, output=""}], properties {}}
  MaterialExpressionChannelMaskParameter_0 {
   class=MaterialExpressionChannelMaskParameter
   inputs [None {from=MaterialExpressionTextureSampleParameter2D_0, output=RGBA}]
   properties {group=Fur_Shell_01 - Base Settings, maskChannel=Green, parameterName=FurLegnthChannel, sortPriority=1, defaultValue=<rgba 0 1 0 0>}
  }
  MaterialExpressionChannelMaskParameter_5 {
   class=MaterialExpressionChannelMaskParameter
   inputs [None {from=MaterialExpressionTextureSampleParameter2D_0, output=RGBA}]
   properties {group=Fur_Shell_01 - Base Settings, parameterName=FurNormalMaskChannel, sortPriority=2}
  }
  MaterialExpressionChannelMaskParameter_6 {
   class=MaterialExpressionChannelMaskParameter
   inputs [None {from=MaterialExpressionTextureSampleParameter2D_0, output=RGBA}]
   properties {group=Fur_Shell_01 - Base Settings, maskChannel=Blue, parameterName=FurClothingMaskChannel, sortPriority=2, defaultValue=<rgba 0 0 1 0>}
  }
  MaterialExpressionChannelMaskParameter_7 {
   class=MaterialExpressionChannelMaskParameter
   inputs [None {from=MaterialExpressionFunctionInput_7, output=""}]
   properties {
    desc=Coming from the TA mask. Determines what area of the mesh should have fur.
    group=Fur_Shell_01 - Base Settings
    parameterName=ShellFurMaskChannel
    sortPriority=-1
   }
  }
  MaterialExpressionChannelMaskParameter_8 {
   class=MaterialExpressionChannelMaskParameter
   inputs [None {from=MaterialExpressionTextureSampleParameter2D_0, output=RGBA}]
   properties {
    group=Fur_Shell_01 - Base Settings
    maskChannel=Alpha
    parameterName=SecondaryFurBlendMapChannel
    sortPriority=2
    defaultValue=<rgba 0 0 0 1>
   }
  }
  MaterialExpressionClamp_1 {
   class=MaterialExpressionClamp
   inputs [
    None {from=MaterialExpressionNamedRerouteDeclaration_70, output=""}
    Min {from="", output=""}
    Max {from="", output=""}
   ]
   properties {maxDefault=0.975}
  }
  MaterialExpressionComponentMask_1 {
   class=MaterialExpressionComponentMask
   inputs [None {from=MaterialExpressionTextureSampleParameter2D_1, output=RGB}]
   properties {g=true, r=true}
  }
  MaterialExpressionConstant2Vector_3 {class=MaterialExpressionConstant2Vector, properties {g=78.233, r=12.9898}}
  MaterialExpressionConstant3Vector_3 {class=MaterialExpressionConstant3Vector, properties {constant=<rgba 0 0 1 0>}}
  MaterialExpressionConstant3Vector_4 {class=MaterialExpressionConstant3Vector, properties {constant=<rgba 0 1 0 0>}}
  MaterialExpressionConstant3Vector_5 {class=MaterialExpressionConstant3Vector, properties {constant=<rgba 1 0 0 0>}}
  MaterialExpressionConstant3Vector_6 {class=MaterialExpressionConstant3Vector, properties {constant=<rgba 0 0 1 0>}}
  MaterialExpressionConstant_10 {class=MaterialExpressionConstant, properties {r=1}}
  MaterialExpressionConstant_11 {class=MaterialExpressionConstant}
  MaterialExpressionConstant_13 {class=MaterialExpressionConstant, properties {r=1}}
  MaterialExpressionConstant_14 {class=MaterialExpressionConstant}
  MaterialExpressionConstant_15 {class=MaterialExpressionConstant, properties {r=1}}
  MaterialExpressionConstant_16 {class=MaterialExpressionConstant, properties {r=2}}
  MaterialExpressionConstant_17 {class=MaterialExpressionConstant, properties {r=1}}
  MaterialExpressionConstant_18 {class=MaterialExpressionConstant, properties {r=1}}
  MaterialExpressionConstant_19 {class=MaterialExpressionConstant, properties {r=2}}
  MaterialExpressionConstant_3 {class=MaterialExpressionConstant, properties {r=1}}
  MaterialExpressionConstant_6 {class=MaterialExpressionConstant, properties {r=0.7972}}
  MaterialExpressionConstant_7 {class=MaterialExpressionConstant, properties {r=1}}
  MaterialExpressionConstant_8 {class=MaterialExpressionConstant, properties {r=1.5}}
  MaterialExpressionConstant_9 {class=MaterialExpressionConstant, properties {r=0.01}}
  MaterialExpressionConvert_0 {
   class=MaterialExpressionConvert
   inputs [None {from=MaterialExpressionNamedRerouteUsage_11}]
   properties {
    nodeName=Break Float3
    convertInputs [{
      type=Vector3
      expressionInput {
       expression=&self:MaterialExpressionNamedRerouteUsage_11
       inputName=None
       mask=0
       maskA=0
       maskB=0
       maskG=0
       maskR=0
       outputIndex=0
      }
     }]
    convertOutputs [{type=Scalar}, {type=Scalar}, {type=Scalar}]
   }
  }
  MaterialExpressionConvert_1 {
   class=MaterialExpressionConvert
   inputs [
    None {from=MaterialExpressionMultiply_30}
    None {from=MaterialExpressionConvert_0}
    None {from=MaterialExpressionConvert_0}
   ]
   properties {
    nodeName=Make Float3
    convertInputs [
     {
      type=Scalar
      expressionInput {
       expression=&self:MaterialExpressionMultiply_30
       inputName=None
       mask=0
       maskA=0
       maskB=0
       maskG=0
       maskR=0
       outputIndex=0
      }
     }
     {
      type=Scalar
      expressionInput {
       expression=&self:MaterialExpressionConvert_0
       inputName=None
       mask=0
       maskA=0
       maskB=0
       maskG=0
       maskR=0
       outputIndex=1
      }
     }
     {
      type=Scalar
      expressionInput {
       expression=&self:MaterialExpressionConvert_0
       inputName=None
       mask=0
       maskA=0
       maskB=0
       maskG=0
       maskR=0
       outputIndex=2
      }
     }
    ]
    convertOutputs [{type=Vector3}]
   }
  }
  MaterialExpressionCrossProduct_3 {
   class=MaterialExpressionCrossProduct
   inputs [
    A {from=MaterialExpressionVertexNormalWS_3, output=""}
    B {from=MaterialExpressionTransform_19, output=""}
   ]
   properties {}
  }
  MaterialExpressionCustom_1 {class=MaterialExpressionCustom, inputs [], properties {code="const float fNaN = 0.0f / 0.0f;\r\nreturn fNaN;", showCode=true, inputs []}}
  MaterialExpressionDeriveNormalZ_0 {class=MaterialExpressionDeriveNormalZ, inputs [InXY {from=MaterialExpressionSubtract_5, output=""}], properties {}}
  MaterialExpressionDivide_1 {
   class=MaterialExpressionDivide
   inputs [
    A {from=MaterialExpressionSubtract_4}
    B {from=MaterialExpressionOneMinus_12}
   ]
  }
  MaterialExpressionDivide_2 {
   class=MaterialExpressionDivide
   inputs [
    A {from=MaterialExpressionNamedRerouteUsage_115}
    B {from=MaterialExpressionNamedRerouteUsage_114}
   ]
  }
  MaterialExpressionDotProduct_3 {
   class=MaterialExpressionDotProduct
   inputs [
    A {from=MaterialExpressionMultiply_144, output=""}
    B {from=MaterialExpressionConstant2Vector_3, output=G}
   ]
   properties {}
  }
  MaterialExpressionFmod_1 {
   class=MaterialExpressionFmod
   inputs [
    A {from=ShellMeshMaterialExpressionShellLayerIndex_5, output=""}
    B {from=MaterialExpressionPower_16, output=""}
   ]
   properties {}
  }
  MaterialExpressionFrac_3 {class=MaterialExpressionFrac, inputs [None {from=MaterialExpressionMultiply_141, output=""}], properties {}}
  MaterialExpressionFresnel_0 {
   class=MaterialExpressionFresnel
   inputs [
    ExponentIn {from=MaterialExpressionConstant_8, output=""}
    BaseReflectFractionIn {from=MaterialExpressionConstant_9, output=""}
    Normal {from="", output=""}
   ]
   properties {baseReflectFraction=0.01, exponent=9}
  }
  MaterialExpressionFunctionInput_6 {class=MaterialExpressionFunctionInput, inputs [Preview {from="", output=""}], properties {inputType=FunctionInput_MaterialAttributes}}
  MaterialExpressionFunctionInput_7 {
   class=MaterialExpressionFunctionInput
   inputs [Preview {from="", output=""}]
   properties {bUsePreviewValueAsDefault=true, inputName=TechArtMask, sortPriority=1, previewValue=<1 1 1 1>}
  }
  MaterialExpressionFunctionOutput_0 {
   class=MaterialExpressionFunctionOutput
   inputs [None {from=MaterialExpressionNamedRerouteDeclaration_7, output=""}]
   properties {outputName=Fur Length, sortPriority=1}
  }
  MaterialExpressionFunctionOutput_3 {class=MaterialExpressionFunctionOutput, inputs [None {from=MaterialExpressionStaticSwitchParameter_6, output=""}]}
  MaterialExpressionGetMaterialAttributes_0 {
   class=MaterialExpressionGetMaterialAttributes
   inputs [None {from=MaterialExpressionNamedRerouteDeclaration_74, output=""}]
   properties {attributeGetTypes [
     #1
     #2
     #3
     #4
    ]}
  }
  MaterialExpressionIf_3 {
   class=MaterialExpressionIf
   inputs [
    A {from=MaterialExpressionVertexInterpolator_3, output=PS}
    B {from=MaterialExpressionScalarParameter_74, output=""}
    "A > B" {from=MaterialExpressionConstant_11, output=""}
    "A == B" {from=MaterialExpressionConstant_11, output=""}
    "A < B" {from=MaterialExpressionNamedRerouteUsage_105, output=""}
   ]
   properties {}
  }
  MaterialExpressionIf_4 {
   class=MaterialExpressionIf
   inputs [
    A {from=MaterialExpressionNamedRerouteUsage_116, output=""}
    B {from="", output=""}
    "A > B" {from=MaterialExpressionCustom_1, output=""}
    "A == B" {from=MaterialExpressionMultiply_33, output=""}
    "A < B" {from=MaterialExpressionMultiply_33, output=""}
   ]
   properties {constB=0.5}
  }
  MaterialExpressionIf_5 {
   class=MaterialExpressionIf
   inputs [
    A {from=ShellMeshMaterialExpressionShellCount_1, output=""}
    B {from=MaterialExpressionScalarParameter_88, output=""}
    "A > B" {from=MaterialExpressionMultiply_55, output=""}
    "A == B" {from=MaterialExpressionConstant_14, output=""}
    "A < B" {from=MaterialExpressionConstant_14, output=""}
   ]
   properties {}
  }
  MaterialExpressionInverseLinearInterpolate_4 {
   class=MaterialExpressionInverseLinearInterpolate
   inputs [
    A {from=MaterialExpressionScalarParameter_77, output=""}
    B {from=MaterialExpressionAdd_10, output=""}
    Value {from=MaterialExpressionNamedRerouteDeclaration_101, output=""}
   ]
   properties {constB=0}
  }
  MaterialExpressionInverseLinearInterpolate_5 {
   class=MaterialExpressionInverseLinearInterpolate
   inputs [
    A {from=MaterialExpressionScalarParameter_82, output=""}
    B {from=MaterialExpressionAdd_13, output=""}
    Value {from=MaterialExpressionNamedRerouteUsage_125, output=""}
   ]
   properties {constB=0}
  }
  MaterialExpressionInverseLinearInterpolate_6 {
   class=MaterialExpressionInverseLinearInterpolate
   inputs [
    A {from=MaterialExpressionScalarParameter_85, output=""}
    B {from=MaterialExpressionAdd_15, output=""}
    Value {from=MaterialExpressionNamedRerouteUsage_129, output=""}
   ]
   properties {constB=0}
  }
  MaterialExpressionInverseLinearInterpolate_7 {
   class=MaterialExpressionInverseLinearInterpolate
   inputs [
    A {from="", output=""}
    B {from="", output=""}
    Value {from=MaterialExpressionMaterialFunctionCall_8, output=FOV angle in degrees}
   ]
   properties {constA=45, constB=90}
  }
  MaterialExpressionLength_1 {class=MaterialExpressionLength, inputs [None {from=MaterialExpressionWorldPosition_1, output=XYZ}], properties {}}
  MaterialExpressionLinearInterpolate_0 {
   class=MaterialExpressionLinearInterpolate
   inputs [
    A {from=MaterialExpressionOneMinus_1}
    B {from=MaterialExpressionOneMinus_0}
    Alpha {from=MaterialExpressionPower_0}
   ]
  }
  MaterialExpressionLinearInterpolate_1 {
   class=MaterialExpressionLinearInterpolate
   inputs [
    A {from=MaterialExpressionScalarParameter_0}
    B {from=MaterialExpressionScalarParameter_3}
    Alpha {from=MaterialExpressionNamedRerouteUsage_4}
   ]
  }
  MaterialExpressionLinearInterpolate_2 {
   class=MaterialExpressionLinearInterpolate
   inputs [
    A {from=MaterialExpressionScalarParameter_1}
    B {from=MaterialExpressionScalarParameter_4}
    Alpha {from=MaterialExpressionNamedRerouteUsage_3}
   ]
  }
  MaterialExpressionLinearInterpolate_21 {
   class=MaterialExpressionLinearInterpolate
   inputs [
    A {from=MaterialExpressionMultiply_29}
    B {from=MaterialExpressionMaterialFunctionCall_2, output=Result}
    Alpha {from=MaterialExpressionTwoSidedSign_0}
   ]
  }
  MaterialExpressionLinearInterpolate_22 {
   class=MaterialExpressionLinearInterpolate
   inputs [
    A {from=MaterialExpressionNamedRerouteDeclaration_91}
    B {from=MaterialExpressionNamedRerouteDeclaration_92}
    Alpha {from=MaterialExpressionNamedRerouteUsage_10}
   ]
  }
  MaterialExpressionLinearInterpolate_23 {
   class=MaterialExpressionLinearInterpolate
   inputs [
    A {from=MaterialExpressionScalarParameter_56}
    B {from=MaterialExpressionLinearInterpolate_4}
    Alpha {from=MaterialExpressionPower_9}
   ]
  }
  MaterialExpressionLinearInterpolate_24 {
   class=MaterialExpressionLinearInterpolate
   inputs [
    A {from=MaterialExpressionConstant3Vector_3}
    B {from=MaterialExpressionNormalize_1}
    Alpha {from=MaterialExpressionPower_10}
   ]
  }
  MaterialExpressionLinearInterpolate_25 {
   class=MaterialExpressionLinearInterpolate
   inputs [
    A {from=MaterialExpressionNamedRerouteUsage_80}
    B {from=MaterialExpressionMultiply_39}
    Alpha {from=MaterialExpressionPower_12}
   ]
  }
  MaterialExpressionLinearInterpolate_26 {
   class=MaterialExpressionLinearInterpolate
   inputs [
    A {from=MaterialExpressionNamedRerouteUsage_81}
    B {from=MaterialExpressionMultiply_40}
    Alpha {from=MaterialExpressionPower_14}
   ]
  }
  MaterialExpressionLinearInterpolate_27 {
   class=MaterialExpressionLinearInterpolate
   inputs [
    A {from=MaterialExpressionScalarParameter_70}
    B {from=MaterialExpressionScalarParameter_71}
    Alpha {from=MaterialExpressionPower_15}
   ]
  }
  MaterialExpressionLinearInterpolate_28 {
   class=MaterialExpressionLinearInterpolate
   inputs [
    A {from=MaterialExpressionConstant3Vector_4}
    B {from=MaterialExpressionConstant3Vector_5}
    Alpha {from=MaterialExpressionNamedRerouteUsage_103}
   ]
  }
  MaterialExpressionLinearInterpolate_29 {
   class=MaterialExpressionLinearInterpolate
   inputs [
    A {from=MaterialExpressionNamedRerouteUsage_104}
    B {from=MaterialExpressionLinearInterpolate_28}
    Alpha {from=MaterialExpressionScalarParameter_75}
   ]
  }
  MaterialExpressionLinearInterpolate_3 {
   class=MaterialExpressionLinearInterpolate
   inputs [
    A {from=MaterialExpressionLinearInterpolate_7}
    B {from=MaterialExpressionLinearInterpolate_0}
    Alpha {from=MaterialExpressionNamedRerouteUsage_8}
   ]
  }
  MaterialExpressionLinearInterpolate_30 {
   class=MaterialExpressionLinearInterpolate
   inputs [
    A {from=MaterialExpressionNamedRerouteUsage_122}
    B {from=MaterialExpressionNamedRerouteUsage_113}
    Alpha {from=MaterialExpressionSaturate_20}
   ]
  }
  MaterialExpressionLinearInterpolate_31 {
   class=MaterialExpressionLinearInterpolate
   inputs [
    A {from=MaterialExpressionNamedRerouteDeclaration_97}
    B {from=MaterialExpressionSaturate_16}
    Alpha {from=MaterialExpressionScalarParameter_79}
   ]
  }
  MaterialExpressionLinearInterpolate_32 {
   class=MaterialExpressionLinearInterpolate
   inputs [
    A {from=""}
    B {from=MaterialExpressionNamedRerouteUsage_100}
    Alpha {from=MaterialExpressionScalarParameter_72}
   ]
   properties {constA=1}
  }
  MaterialExpressionLinearInterpolate_33 {
   class=MaterialExpressionLinearInterpolate
   inputs [
    A {from=MaterialExpressionLinearInterpolate_22}
    B {from=MaterialExpressionMultiply_49}
    Alpha {from=MaterialExpressionClamp_1}
   ]
  }
  MaterialExpressionLinearInterpolate_34 {
   class=MaterialExpressionLinearInterpolate
   inputs [
    A {from=MaterialExpressionScalarParameter_87}
    B {from=MaterialExpressionMax_1}
    Alpha {from=MaterialExpressionNamedRerouteUsage_130}
   ]
  }
  MaterialExpressionLinearInterpolate_35 {
   class=MaterialExpressionLinearInterpolate
   inputs [
    A {from=MaterialExpressionTextureSample_7, output=RGB}
    B {from=MaterialExpressionNamedRerouteUsage_133}
    Alpha {from=MaterialExpressionNamedRerouteUsage_134}
   ]
  }
  MaterialExpressionLinearInterpolate_36 {
   class=MaterialExpressionLinearInterpolate
   inputs [
    A {from=MaterialExpressionNormalize_6}
    B {from=MaterialExpressionNamedRerouteUsage_136}
    Alpha {from=MaterialExpressionNamedRerouteUsage_137}
   ]
  }
  MaterialExpressionLinearInterpolate_37 {
   class=MaterialExpressionLinearInterpolate
   inputs [
    A {from=MaterialExpressionScalarParameter_13}
    B {from=MaterialExpressionScalarParameter_91}
    Alpha {from=MaterialExpressionNamedRerouteUsage_138}
   ]
  }
  MaterialExpressionLinearInterpolate_38 {
   class=MaterialExpressionLinearInterpolate
   inputs [
    A {from=MaterialExpressionScalarParameter_24}
    B {from=MaterialExpressionScalarParameter_92}
    Alpha {from=MaterialExpressionNamedRerouteUsage_126}
   ]
  }
  MaterialExpressionLinearInterpolate_39 {
   class=MaterialExpressionLinearInterpolate
   inputs [
    A {from=""}
    B {from=MaterialExpressionScalarParameter_93}
    Alpha {from=MaterialExpressionNamedRerouteUsage_152}
   ]
   properties {constA=1}
  }
  MaterialExpressionLinearInterpolate_4 {
   class=MaterialExpressionLinearInterpolate
   inputs [
    A {from=MaterialExpressionScalarParameter_5}
    B {from=MaterialExpressionScalarParameter_58}
    Alpha {from=MaterialExpressionLinearInterpolate_5}
   ]
  }
  MaterialExpressionLinearInterpolate_40 {
   class=MaterialExpressionLinearInterpolate
   inputs [
    A {from=MaterialExpressionScalarParameter_55}
    B {from=MaterialExpressionScalarParameter_95}
    Alpha {from=MaterialExpressionNamedRerouteUsage_153}
   ]
   properties {constA=1}
  }
  MaterialExpressionLinearInterpolate_41 {
   class=MaterialExpressionLinearInterpolate
   inputs [
    A {from=MaterialExpressionScalarParameter_97}
    B {from=""}
    Alpha {from=MaterialExpressionSaturate_22}
   ]
   properties {constA=1}
  }
  MaterialExpressionLinearInterpolate_42 {
   class=MaterialExpressionLinearInterpolate
   inputs [
    A {from=MaterialExpressionTextureSample_10, output=RGB}
    B {from=MaterialExpressionNamedRerouteUsage_161}
    Alpha {from=MaterialExpressionNamedRerouteUsage_159}
   ]
  }
  MaterialExpressionLinearInterpolate_43 {
   class=MaterialExpressionLinearInterpolate
   inputs [
    A {from=MaterialExpressionNamedRerouteDeclaration_68}
    B {from=MaterialExpressionLinearInterpolate_44}
    Alpha {from=MaterialExpressionNamedRerouteUsage_165}
   ]
  }
  MaterialExpressionLinearInterpolate_44 {
   class=MaterialExpressionLinearInterpolate
   inputs [
    A {from=MaterialExpressionNamedRerouteDeclaration_118}
    B {from=MaterialExpressionAdd_18}
    Alpha {from=MaterialExpressionScalarParameter_102}
   ]
  }
  MaterialExpressionLinearInterpolate_45 {
   class=MaterialExpressionLinearInterpolate
   inputs [
    A {from=MaterialExpressionNamedRerouteDeclaration_69}
    B {from=MaterialExpressionNamedRerouteUsage_183}
    Alpha {from=MaterialExpressionNamedRerouteUsage_166}
   ]
  }
  MaterialExpressionLinearInterpolate_46 {
   class=MaterialExpressionLinearInterpolate
   inputs [
    A {from=MaterialExpressionNormalize_9}
    B {from=MaterialExpressionNamedRerouteUsage_172}
    Alpha {from=MaterialExpressionNamedRerouteUsage_167}
   ]
  }
  MaterialExpressionLinearInterpolate_47 {
   class=MaterialExpressionLinearInterpolate
   inputs [
    A {from=MaterialExpressionScalarParameter_105}
    B {from=MaterialExpressionScalarParameter_106}
    Alpha {from=MaterialExpressionNamedRerouteUsage_180}
   ]
  }
  MaterialExpressionLinearInterpolate_5 {
   class=MaterialExpressionLinearInterpolate
   inputs [
    A {from=""}
    B {from=MaterialExpressionNamedRerouteUsage_0}
    Alpha {from=MaterialExpressionScalarParameter_6}
   ]
   properties {constA=1}
  }
  MaterialExpressionLinearInterpolate_6 {
   class=MaterialExpressionLinearInterpolate
   inputs [
    A {from=MaterialExpressionLinearInterpolate_33}
    B {from=""}
    Alpha {from=MaterialExpressionStep_0}
   ]
   properties {constB=0}
  }
  MaterialExpressionLinearInterpolate_7 {
   class=MaterialExpressionLinearInterpolate
   inputs [
    A {from=MaterialExpressionOneMinus_5}
    B {from=MaterialExpressionOneMinus_3}
    Alpha {from=MaterialExpressionPower_6}
   ]
  }
  MaterialExpressionLogarithm2_1 {class=MaterialExpressionLogarithm2, inputs [X {from=MaterialExpressionLinearInterpolate_34, output=""}], properties {}}
  MaterialExpressionMaterialFunctionCall_2 {
   class=MaterialExpressionMaterialFunctionCall
   inputs [
    BaseNormal {from=MaterialExpressionMaterialFunctionCall_5, output=Result}
    AdditionalNormal {from=MaterialExpressionMaterialFunctionCall_6, output=Result}
   ]
   properties {materialFunction=&BlendAngleCorrectedNormals}
  }
  MaterialExpressionMaterialFunctionCall_4 {
   class=MaterialExpressionMaterialFunctionCall
   inputs [
    Brightness {from=MaterialExpressionConstant_6}
    Exponent {from=MaterialExpressionConstant_7}
    Camera Vector {from=MaterialExpressionNormalize_5}
    Vertex / Pixel Normal {from=MaterialExpressionPixelNormalWS_0}
   ]
   properties {materialFunction=&MF_Fresnel_FX}
  }
  MaterialExpressionMaterialFunctionCall_5 {
   class=MaterialExpressionMaterialFunctionCall
   inputs [
    Normal {from=MaterialExpressionNamedRerouteUsage_15}
    Flatness {from=MaterialExpressionScalarParameter_61}
   ]
   properties {materialFunction=&FlattenNormal}
  }
  MaterialExpressionMaterialFunctionCall_6 {
   class=MaterialExpressionMaterialFunctionCall
   inputs [
    Normal {from=MaterialExpressionStaticSwitchParameter_5}
    Flatness {from=MaterialExpressionNamedRerouteDeclaration_113}
   ]
   properties {materialFunction=&FlattenNormal}
  }
  MaterialExpressionMaterialFunctionCall_8 {
   class=MaterialExpressionMaterialFunctionCall
   inputs []
   properties {materialFunction=/Engine/Functions/Engine_MaterialFunctions02/Utility/FOV}
  }
  MaterialExpressionMaterialFunctionCall_9 {
   class=MaterialExpressionMaterialFunctionCall
   inputs [Multiplier {from=MaterialExpressionConstant_19}]
   properties {materialFunction=&Pi}
  }
  MaterialExpressionMax_1 {
   class=MaterialExpressionMax
   inputs [
    A {from=MaterialExpressionScalarParameter_84}
    B {from=""}
   ]
   properties {constB=2}
  }
  MaterialExpressionMultiply_126 {
   class=MaterialExpressionMultiply
   inputs [
    A {from=MaterialExpressionNamedRerouteUsage_178}
    B {from=MaterialExpressionAdd_44}
   ]
  }
  MaterialExpressionMultiply_127 {
   class=MaterialExpressionMultiply
   inputs [
    A {from=MaterialExpressionMultiply_126}
    B {from=MaterialExpressionMaterialFunctionCall_9, output=Result}
   ]
  }
  MaterialExpressionMultiply_128 {
   class=MaterialExpressionMultiply
   inputs [
    A {from=MaterialExpressionMultiply_140}
    B {from=MaterialExpressionSine_10}
   ]
  }
  MaterialExpressionMultiply_129 {
   class=MaterialExpressionMultiply
   inputs [
    A {from=MaterialExpressionNamedRerouteUsage_184}
    B {from=MaterialExpressionMultiply_128}
   ]
  }
  MaterialExpressionMultiply_140 {
   class=MaterialExpressionMultiply
   inputs [
    A {from=MaterialExpressionMultiply_145}
    B {from=MaterialExpressionScalarParameter_147}
   ]
  }
  MaterialExpressionMultiply_141 {
   class=MaterialExpressionMultiply
   inputs [
    A {from=MaterialExpressionSine_11}
    B {from=""}
   ]
   properties {constB=43758.5469}
  }
  MaterialExpressionMultiply_142 {
   class=MaterialExpressionMultiply
   inputs [
    A {from=MaterialExpressionNamedRerouteUsage_182}
    B {from=MaterialExpressionScalarParameter_169}
   ]
   properties {constB=2}
  }
  MaterialExpressionMultiply_143 {
   class=MaterialExpressionMultiply
   inputs [
    A {from=MaterialExpressionFrac_3}
    B {from=""}
   ]
   properties {constB=2}
  }
  MaterialExpressionMultiply_144 {
   class=MaterialExpressionMultiply
   inputs [
    A {from=MaterialExpressionTextureCoordinate_7}
    B {from=MaterialExpressionScalarParameter_170}
   ]
  }
  MaterialExpressionMultiply_145 {
   class=MaterialExpressionMultiply
   inputs [
    A {from=MaterialExpressionMultiply_64}
    B {from=MaterialExpressionScalarParameter_168}
   ]
  }
  MaterialExpressionMultiply_21 {
   class=MaterialExpressionMultiply
   inputs [
    A {from=MaterialExpressionTextureSample_6, output=RGB}
    B {from=MaterialExpressionAppendVector_3}
   ]
  }
  MaterialExpressionMultiply_26 {
   class=MaterialExpressionMultiply
   inputs [
    A {from=MaterialExpressionScalarParameter_49}
    B {from=MaterialExpressionAdd_3}
   ]
  }
  MaterialExpressionMultiply_28 {
   class=MaterialExpressionMultiply
   inputs [
    A {from=MaterialExpressionSubtract_1}
    B {from=MaterialExpressionNamedRerouteUsage_88}
   ]
  }
  MaterialExpressionMultiply_29 {
   class=MaterialExpressionMultiply
   inputs [
    A {from=MaterialExpressionMaterialFunctionCall_2, output=Result}
    B {from=""}
   ]
   properties {constB=-1}
  }
  MaterialExpressionMultiply_30 {
   class=MaterialExpressionMultiply
   inputs [
    A {from=MaterialExpressionConvert_0}
    B {from=""}
   ]
   properties {constB=-1}
  }
  MaterialExpressionMultiply_31 {
   class=MaterialExpressionMultiply
   inputs [
    A {from=MaterialExpressionMultiply_32}
    B {from=MaterialExpressionPower_7}
   ]
  }
  MaterialExpressionMultiply_32 {
   class=MaterialExpressionMultiply
   inputs [
    A {from=MaterialExpressionStaticSwitch_20}
    B {from=MaterialExpressionNamedRerouteDeclaration_99}
   ]
  }
  MaterialExpressionMultiply_33 {
   class=MaterialExpressionMultiply
   inputs [
    A {from=MaterialExpressionStaticSwitchParameter_10}
    B {from=MaterialExpressionNamedRerouteUsage_47}
   ]
  }
  MaterialExpressionMultiply_34 {
   class=MaterialExpressionMultiply
   inputs [
    A {from=MaterialExpressionTransform_2}
    B {from=MaterialExpressionLinearInterpolate_23}
   ]
  }
  MaterialExpressionMultiply_35 {
   class=MaterialExpressionMultiply
   inputs [
    A {from=MaterialExpressionAtmosphericLightVector_0}
    B {from=""}
   ]
   properties {constB=-1}
  }
  MaterialExpressionMultiply_36 {
   class=MaterialExpressionMultiply
   inputs [
    A {from=MaterialExpressionFresnel_0}
    B {from=MaterialExpressionMaterialFunctionCall_4, output=Fresnel Inverted}
   ]
  }
  MaterialExpressionMultiply_37 {
   class=MaterialExpressionMultiply
   inputs [
    A {from=MaterialExpressionNamedRerouteUsage_49}
    B {from=MaterialExpressionScalarParameter_60}
   ]
  }
  MaterialExpressionMultiply_38 {
   class=MaterialExpressionMultiply
   inputs [
    A {from=MaterialExpressionNamedRerouteUsage_34}
    B {from=MaterialExpressionMultiply_37}
   ]
  }
  MaterialExpressionMultiply_39 {
   class=MaterialExpressionMultiply
   inputs [
    A {from=MaterialExpressionNamedRerouteUsage_80}
    B {from=MaterialExpressionScalarParameter_62}
   ]
  }
  MaterialExpressionMultiply_40 {
   class=MaterialExpressionMultiply
   inputs [
    A {from=MaterialExpressionNamedRerouteUsage_81}
    B {from=MaterialExpressionScalarParameter_64}
   ]
  }
  MaterialExpressionMultiply_41 {
   class=MaterialExpressionMultiply
   inputs [
    A {from=MaterialExpressionNamedRerouteUsage_86}
    B {from=MaterialExpressionOneMinus_9}
   ]
  }
  MaterialExpressionMultiply_42 {
   class=MaterialExpressionMultiply
   inputs [
    A {from=MaterialExpressionMultiply_31}
    B {from=MaterialExpressionNamedRerouteUsage_92}
   ]
  }
  MaterialExpressionMultiply_43 {
   class=MaterialExpressionMultiply
   inputs [
    A {from=MaterialExpressionTextureCoordinate_3}
    B {from=MaterialExpressionScalarParameter_69}
   ]
  }
  MaterialExpressionMultiply_44 {
   class=MaterialExpressionMultiply
   inputs [
    A {from=MaterialExpressionComponentMask_1}
    B {from=""}
   ]
   properties {constB=2}
  }
  MaterialExpressionMultiply_45 {
   class=MaterialExpressionMultiply
   inputs [
    A {from=MaterialExpressionStaticSwitchParameter_8}
    B {from=MaterialExpressionLinearInterpolate_32}
   ]
  }
  MaterialExpressionMultiply_46 {
   class=MaterialExpressionMultiply
   inputs [
    A {from=MaterialExpressionTextureSample_5, output=RGB}
    B {from=MaterialExpressionScalarParameter_68}
   ]
   properties {constB=2}
  }
  MaterialExpressionMultiply_47 {
   class=MaterialExpressionMultiply
   inputs [
    A {from=MaterialExpressionLinearInterpolate_27}
    B {from=MaterialExpressionNamedRerouteUsage_102}
   ]
  }
  MaterialExpressionMultiply_48 {
   class=MaterialExpressionMultiply
   inputs [
    A {from=MaterialExpressionNamedRerouteUsage_121}
    B {from=MaterialExpressionNamedRerouteUsage_117}
   ]
  }
  MaterialExpressionMultiply_49 {
   class=MaterialExpressionMultiply
   inputs [
    A {from=MaterialExpressionLinearInterpolate_22}
    B {from=MaterialExpressionNamedRerouteUsage_124}
   ]
  }
  MaterialExpressionMultiply_50 {
   class=MaterialExpressionMultiply
   inputs [
    A {from=MaterialExpressionStep_4}
    B {from=MaterialExpressionStep_5}
   ]
  }
  MaterialExpressionMultiply_51 {
   class=MaterialExpressionMultiply
   inputs [
    A {from=MaterialExpressionScalarParameter_89}
    B {from=MaterialExpressionTextureCoordinate_4}
   ]
  }
  MaterialExpressionMultiply_52 {
   class=MaterialExpressionMultiply
   inputs [
    A {from=MaterialExpressionTextureSample_9, output=RGB}
    B {from=MaterialExpressionAppendVector_6}
   ]
  }
  MaterialExpressionMultiply_53 {
   class=MaterialExpressionMultiply
   inputs [
    A {from=MaterialExpressionScalarParameter_12}
    B {from=MaterialExpressionLinearInterpolate_39}
   ]
  }
  MaterialExpressionMultiply_54 {
   class=MaterialExpressionMultiply
   inputs [
    A {from=MaterialExpressionLinearInterpolate_39}
    B {from=MaterialExpressionScalarParameter_53}
   ]
  }
  MaterialExpressionMultiply_55 {
   class=MaterialExpressionMultiply
   inputs [
    A {from=MaterialExpressionStep_6}
    B {from=MaterialExpressionStep_7}
   ]
  }
  MaterialExpressionMultiply_56 {
   class=MaterialExpressionMultiply
   inputs [
    A {from=MaterialExpressionScalarParameter_96}
    B {from=MaterialExpressionNamedRerouteUsage_155}
   ]
  }
  MaterialExpressionMultiply_57 {
   class=MaterialExpressionMultiply
   inputs [
    A {from=MaterialExpressionLength_1}
    B {from=MaterialExpressionSaturate_21}
   ]
  }
  MaterialExpressionMultiply_58 {
   class=MaterialExpressionMultiply
   inputs [
    A {from=MaterialExpressionScalarParameter_98}
    B {from=MaterialExpressionAdd_17}
   ]
  }
  MaterialExpressionMultiply_59 {
   class=MaterialExpressionMultiply
   inputs [
    A {from=MaterialExpressionScalarParameter_101}
    B {from=MaterialExpressionTextureCoordinate_5}
   ]
  }
  MaterialExpressionMultiply_60 {
   class=MaterialExpressionMultiply
   inputs [
    A {from=MaterialExpressionNamedRerouteDeclaration_68}
    B {from=MaterialExpressionOneMinus_14}
   ]
  }
  MaterialExpressionMultiply_61 {
   class=MaterialExpressionMultiply
   inputs [
    A {from=MaterialExpressionTextureSample_12, output=RGB}
    B {from=MaterialExpressionAppendVector_9}
   ]
  }
  MaterialExpressionMultiply_62 {
   class=MaterialExpressionMultiply
   inputs [
    A {from=MaterialExpressionTextureSample_13, output=RGB}
    B {from=MaterialExpressionAppendVector_11}
   ]
  }
  MaterialExpressionMultiply_63 {
   class=MaterialExpressionMultiply
   inputs [
    A {from=MaterialExpressionStaticSwitch_24}
    B {from=MaterialExpressionMultiply_129}
   ]
  }
  MaterialExpressionMultiply_64 {
   class=MaterialExpressionMultiply
   inputs [
    A {from=MaterialExpressionNamedRerouteUsage_176}
    B {from=MaterialExpressionSign_7}
   ]
  }
  MaterialExpressionNamedRerouteDeclaration_100 {
   class=MaterialExpressionNamedRerouteDeclaration
   inputs [None {from=MaterialExpressionScalarParameter_80}]
   properties {name=FurTextureMipBias, nodeColor=<rgba 1 0 0.2344 1>}
  }
  MaterialExpressionNamedRerouteDeclaration_101 {
   class=MaterialExpressionNamedRerouteDeclaration
   inputs [None {from=MaterialExpressionMultiply_57}]
   properties {name=PixelDistance, nodeColor=<rgba 1 0 0.2109 1>}
  }
  MaterialExpressionNamedRerouteDeclaration_102 {
   class=MaterialExpressionNamedRerouteDeclaration
   inputs [None {from=MaterialExpressionSaturate_18}]
   properties {name=FarFurBlend, nodeColor=<rgba 0 0.5312 1 1>}
  }
  MaterialExpressionNamedRerouteDeclaration_103 {
   class=MaterialExpressionNamedRerouteDeclaration
   inputs [None {from=MaterialExpressionIf_5}]
   properties {name=ShellLODCullMask, nodeColor=<rgba 0.8281 0 1 1>}
  }
  MaterialExpressionNamedRerouteDeclaration_105 {
   class=MaterialExpressionNamedRerouteDeclaration
   inputs [None {from=MaterialExpressionSaturate_19}]
   properties {name=ShellLODCullBlend, nodeColor=<rgba 0 0.5312 1 1>}
  }
  MaterialExpressionNamedRerouteDeclaration_106 {
   class=MaterialExpressionNamedRerouteDeclaration
   inputs [None {from=MaterialExpressionTextureSample_8, output=RGB}]
   properties {name=FurPattern_Distance, nodeColor=<rgba 0.047 0.6685 1 1>}
  }
  MaterialExpressionNamedRerouteDeclaration_107 {
   class=MaterialExpressionNamedRerouteDeclaration
   inputs [None {from=MaterialExpressionMultiply_51}]
   properties {name=FurDistanceUVs, nodeColor=<rgba 1 0.5859 0 1>}
  }
  MaterialExpressionNamedRerouteDeclaration_108 {
   class=MaterialExpressionNamedRerouteDeclaration
   inputs [None {from=MaterialExpressionNormalize_7}]
   properties {name=FurNormalDistance, nodeColor=<rgba 0.2656 0 1 1>}
  }
  MaterialExpressionNamedRerouteDeclaration_109 {
   class=MaterialExpressionNamedRerouteDeclaration
   inputs [None {from=MaterialExpressionStaticBoolParameter_5}]
   properties {name=UseFarFur, nodeColor=<rgba 0.0547 1 0 1>}
  }
  MaterialExpressionNamedRerouteDeclaration_110 {
   class=MaterialExpressionNamedRerouteDeclaration
   inputs [None {from=MaterialExpressionInverseLinearInterpolate_6}]
   properties {name=ShellLODCullBlendUnclamped, nodeColor=<rgba 0 1 0.1797 1>}
  }
  MaterialExpressionNamedRerouteDeclaration_111 {
   class=MaterialExpressionNamedRerouteDeclaration
   inputs [None {from=MaterialExpressionTextureObjectParameter_2}]
   properties {name=FurPatternTexture, nodeColor=<rgba 0.9922 1 0 1>}
  }
  MaterialExpressionNamedRerouteDeclaration_112 {
   class=MaterialExpressionNamedRerouteDeclaration
   inputs [None {from=MaterialExpressionTextureObjectParameter_3}]
   properties {name=FurPatternNormalTexture, nodeColor=<rgba 0.8516 1 0 1>}
  }
  MaterialExpressionNamedRerouteDeclaration_113 {
   class=MaterialExpressionNamedRerouteDeclaration
   inputs [None {from=MaterialExpressionMultiply_41}]
   properties {name=ClothingMaskInverse, nodeColor=<rgba 0 1 1 1>}
  }
  MaterialExpressionNamedRerouteDeclaration_114 {
   class=MaterialExpressionNamedRerouteDeclaration
   inputs [None {from=MaterialExpressionMultiply_58}]
   properties {name=Secondary_FurPatternUVs, nodeColor=<rgba 0.9122 0.4794 1 1>}
  }
  MaterialExpressionNamedRerouteDeclaration_115 {
   class=MaterialExpressionNamedRerouteDeclaration
   inputs [None {from=MaterialExpressionTextureObjectParameter_4}]
   properties {name=Secondary_FurPatternTexture, nodeColor=<rgba 0.9922 1 0 1>}
  }
  MaterialExpressionNamedRerouteDeclaration_116 {
   class=MaterialExpressionNamedRerouteDeclaration
   inputs [None {from=MaterialExpressionTextureSample_11, output=RGB}]
   properties {name=Secondary_FurPattern_Distance, nodeColor=<rgba 0.047 0.6685 1 1>}
  }
  MaterialExpressionNamedRerouteDeclaration_117 {
   class=MaterialExpressionNamedRerouteDeclaration
   inputs [None {from=MaterialExpressionMultiply_59}]
   properties {name=Secondary_FurDistanceUVs, nodeColor=<rgba 1 0.5859 0 1>}
  }
  MaterialExpressionNamedRerouteDeclaration_118 {
   class=MaterialExpressionNamedRerouteDeclaration
   inputs [None {from=MaterialExpressionStaticSwitch_21}]
   properties {name=Secondary_FurPattern, nodeColor=<rgba 0.3125 1 0 1>}
  }
  MaterialExpressionNamedRerouteDeclaration_119 {
   class=MaterialExpressionNamedRerouteDeclaration
   inputs [None {from=MaterialExpressionStaticBoolParameter_6}]
   properties {name=UseSecondaryFurStyle, nodeColor=<rgba 0.0547 1 0 1>}
  }
  MaterialExpressionNamedRerouteDeclaration_120 {
   class=MaterialExpressionNamedRerouteDeclaration
   inputs [None {from=MaterialExpressionStaticSwitch_22}]
   properties {name=FurPatternBlended, nodeColor=<rgba 0.3029 0.128 1 1>}
  }
  MaterialExpressionNamedRerouteDeclaration_121 {
   class=MaterialExpressionNamedRerouteDeclaration
   inputs [None {from=MaterialExpressionChannelMaskParameter_8, output=R}]
   properties {name=SecondaryFurBlendMap, nodeColor=<rgba 1 0.2848 0 1>}
  }
  MaterialExpressionNamedRerouteDeclaration_122 {
   class=MaterialExpressionNamedRerouteDeclaration
   inputs [None {from=MaterialExpressionTextureObjectParameter_5}]
   properties {name=Secondary_FurPatternNormalTexture, nodeColor=<rgba 0.8516 1 0 1>}
  }
  MaterialExpressionNamedRerouteDeclaration_123 {
   class=MaterialExpressionNamedRerouteDeclaration
   inputs [None {from=MaterialExpressionNormalize_10}]
   properties {name=Secondary_FurNormalDistance, nodeColor=<rgba 0.2656 0 1 1>}
  }
  MaterialExpressionNamedRerouteDeclaration_124 {
   class=MaterialExpressionNamedRerouteDeclaration
   inputs [None {from=MaterialExpressionStaticSwitchParameter_12}]
   properties {name=Secondary_FurNormal, nodeColor=<rgba 0.0312 1 0 1>}
  }
  MaterialExpressionNamedRerouteDeclaration_125 {
   class=MaterialExpressionNamedRerouteDeclaration
   inputs [None {from=MaterialExpressionStaticSwitch_27}]
   properties {name=FurNormalBlended, nodeColor=<rgba 0.0312 0 1 1>}
  }
  MaterialExpressionNamedRerouteDeclaration_126 {
   class=MaterialExpressionNamedRerouteDeclaration
   inputs [None {from=MaterialExpressionNormalize_13}]
   properties {name=DistortionFlowMap, nodeColor=<rgba 1 0 0.5787 1>}
  }
  MaterialExpressionNamedRerouteDeclaration_127 {
   class=MaterialExpressionNamedRerouteDeclaration
   inputs [None {from=MaterialExpressionMultiply_63}]
   properties {name=WPO_Distortion, nodeColor=<rgba 1 0 0.8186 1>}
  }
  MaterialExpressionNamedRerouteDeclaration_128 {
   class=MaterialExpressionNamedRerouteDeclaration
   inputs [None {from=MaterialExpressionOneMinus_30}]
   properties {name=RandomSeed, nodeColor=<rgba 0.0034 0.8376 1 1>}
  }
  MaterialExpressionNamedRerouteDeclaration_21 {
   class=MaterialExpressionNamedRerouteDeclaration
   inputs [None {from=MaterialExpressionMultiply_26}]
   properties {name=FurPattern_UVs, nodeColor=<rgba 0.9735 0 1 1>}
  }
  MaterialExpressionNamedRerouteDeclaration_23 {
   class=MaterialExpressionNamedRerouteDeclaration
   inputs [None {from=MaterialExpressionMultiply_28}]
   properties {name=Fur_Opacity, nodeColor=<rgba 0.3906 0.2103 0.0483 1>}
  }
  MaterialExpressionNamedRerouteDeclaration_44 {
   class=MaterialExpressionNamedRerouteDeclaration
   inputs [None {from=MaterialExpressionGetMaterialAttributes_0, output=BaseColor}]
   properties {name=In_BaseColor, nodeColor=<rgba 1 0 0.5396 1>}
  }
  MaterialExpressionNamedRerouteDeclaration_45 {
   class=MaterialExpressionNamedRerouteDeclaration
   inputs [None {from=MaterialExpressionGetMaterialAttributes_0, output=Normal}]
   properties {name=In_Normal, nodeColor=<rgba 1 0 0.5396 1>}
  }
  MaterialExpressionNamedRerouteDeclaration_47 {
   class=MaterialExpressionNamedRerouteDeclaration
   inputs [None {from=MaterialExpressionGetMaterialAttributes_0, output=Tangent}]
   properties {name=In_Tangent, nodeColor=<rgba 1 0 0.5396 1>}
  }
  MaterialExpressionNamedRerouteDeclaration_48 {
   class=MaterialExpressionNamedRerouteDeclaration
   inputs [None {from=MaterialExpressionGetMaterialAttributes_0, output=Anisotropy}]
   properties {name=In_Aniso, nodeColor=<rgba 1 0 0.5396 1>}
  }
  MaterialExpressionNamedRerouteDeclaration_54 {
   class=MaterialExpressionNamedRerouteDeclaration
   inputs [None {from=MaterialExpressionGetMaterialAttributes_0, output=MaterialAttributes}]
   properties {name=In_MA, nodeColor=<rgba 1 0 0.5396 1>}
  }
  MaterialExpressionNamedRerouteDeclaration_55 {
   class=MaterialExpressionNamedRerouteDeclaration
   inputs [None {from=MaterialExpressionAdd_5}]
   properties {name=FurColor_Backlit, nodeColor=<rgba 1 0.6959 0 1>}
  }
  MaterialExpressionNamedRerouteDeclaration_56 {
   class=MaterialExpressionNamedRerouteDeclaration
   inputs [None {from=MaterialExpressionChannelMaskParameter_6, output=R}]
   properties {name=FurClothingMask, nodeColor=<rgba 0.8281 0 1 1>}
  }
  MaterialExpressionNamedRerouteDeclaration_57 {
   class=MaterialExpressionNamedRerouteDeclaration
   inputs [None {from=MaterialExpressionIf_4}]
   properties {name=FurWPO, nodeColor=<rgba 1 0.8672 0 1>}
  }
  MaterialExpressionNamedRerouteDeclaration_58 {
   class=MaterialExpressionNamedRerouteDeclaration
   inputs [None {from=MaterialExpressionNormalize_4}]
   properties {name=FlowMap_Swizzled, nodeColor=<rgba 1 0.9375 0 1>}
  }
  MaterialExpressionNamedRerouteDeclaration_59 {
   class=MaterialExpressionNamedRerouteDeclaration
   inputs [None {from=MaterialExpressionLinearInterpolate_30}]
   properties {name=WPO_FurLength, nodeColor=<rgba 0.8281 0 1 1>}
  }
  MaterialExpressionNamedRerouteDeclaration_6 {
   class=MaterialExpressionNamedRerouteDeclaration
   inputs [None {from=MaterialExpressionLinearInterpolate_21}]
   properties {name=Fur_Normal, nodeColor=<rgba 0.4247 0.3851 1 1>}
  }
  MaterialExpressionNamedRerouteDeclaration_60 {
   class=MaterialExpressionNamedRerouteDeclaration
   inputs [None {from=MaterialExpressionSaturate_1}]
   properties {name=fresnel_backlight, nodeColor=<rgba 0 1 0.2969 1>}
  }
  MaterialExpressionNamedRerouteDeclaration_61 {
   class=MaterialExpressionNamedRerouteDeclaration
   inputs [None {from=MaterialExpressionChannelMaskParameter_5, output=R}]
   properties {name=FurNormalMask, nodeColor=<rgba 0.7109 1 0 1>}
  }
  MaterialExpressionNamedRerouteDeclaration_63 {
   class=MaterialExpressionNamedRerouteDeclaration
   inputs [None {from=MaterialExpressionMultiply_34}]
   properties {name=WPO_NormalInfluence, nodeColor=<rgba 1 0 0.7266 1>}
  }
  MaterialExpressionNamedRerouteDeclaration_64 {
   class=MaterialExpressionNamedRerouteDeclaration
   inputs [None {from=MaterialExpressionMultiply_42}]
   properties {name=WPO_FlowMap, nodeColor=<rgba 1 0 0.8203 1>}
  }
  MaterialExpressionNamedRerouteDeclaration_65 {
   class=MaterialExpressionNamedRerouteDeclaration
   inputs [None {from=MaterialExpressionLinearInterpolate_26}]
   properties {name=Fur_Color, nodeColor=<rgba 0.2422 1 0 1>}
  }
  MaterialExpressionNamedRerouteDeclaration_66 {
   class=MaterialExpressionNamedRerouteDeclaration
   inputs [None {from=MaterialExpressionLinearInterpolate_25}]
   properties {name=FurColor_Roots, nodeColor=<rgba 0.7344 1 0 1>}
  }
  MaterialExpressionNamedRerouteDeclaration_67 {
   class=MaterialExpressionNamedRerouteDeclaration
   inputs [None {from=MaterialExpressionLinearInterpolate_31}]
   properties {name=RenormalizedShellLayer, nodeColor=<rgba 0 1 0.1328 1>}
  }
  MaterialExpressionNamedRerouteDeclaration_68 {
   class=MaterialExpressionNamedRerouteDeclaration
   inputs [None {from=MaterialExpressionStaticSwitch_14}]
   properties {name=FurPattern, nodeColor=<rgba 0.3125 1 0 1>}
  }
  MaterialExpressionNamedRerouteDeclaration_69 {
   class=MaterialExpressionNamedRerouteDeclaration
   inputs [None {from=MaterialExpressionStaticSwitch_15}]
   properties {name=FurNormalNormal, nodeColor=<rgba 0.0312 1 0 1>}
  }
  MaterialExpressionNamedRerouteDeclaration_7 {
   class=MaterialExpressionNamedRerouteDeclaration
   inputs [None {from=MaterialExpressionChannelMaskParameter_0, output=R}]
   properties {name=FurLengthMap, nodeColor=<rgba 0 1 0.2002 1>}
  }
  MaterialExpressionNamedRerouteDeclaration_70 {
   class=MaterialExpressionNamedRerouteDeclaration
   inputs [None {from=MaterialExpressionScalarParameter_66}]
   properties {name=FurClothing, nodeColor=<rgba 0.7812 1 0 1>}
  }
  MaterialExpressionNamedRerouteDeclaration_71 {
   class=MaterialExpressionNamedRerouteDeclaration
   inputs [None {from=MaterialExpressionChannelMaskParameter_7, output=R}]
   properties {name=ShellFurMask, nodeColor=<rgba 1 0.8203 0 1>}
  }
  MaterialExpressionNamedRerouteDeclaration_72 {
   class=MaterialExpressionNamedRerouteDeclaration
   inputs [None {from=MaterialExpressionSaturate_2}]
   properties {name=FlowMask, nodeColor=<rgba 1 0.9375 0 1>}
  }
  MaterialExpressionNamedRerouteDeclaration_73 {
   class=MaterialExpressionNamedRerouteDeclaration
   inputs [None {from=MaterialExpressionSetMaterialAttributes_0}]
   properties {name=FurShellMaterial, nodeColor=<rgba 1 0.4453 0 1>}
  }
  MaterialExpressionNamedRerouteDeclaration_74 {
   class=MaterialExpressionNamedRerouteDeclaration
   inputs [None {from=MaterialExpressionFunctionInput_6}]
   properties {name=IN_Material, nodeColor=<rgba 0.125 0 1 1>}
  }
  MaterialExpressionNamedRerouteDeclaration_75 {
   class=MaterialExpressionNamedRerouteDeclaration
   inputs [None {from=MaterialExpressionNamedRerouteDeclaration_58}]
   properties {name=Tangent, nodeColor=<rgba 0 1 0.7891 1>}
  }
  MaterialExpressionNamedRerouteDeclaration_76 {
   class=MaterialExpressionNamedRerouteDeclaration
   inputs [None {from=MaterialExpressionOneMinus_10}]
   properties {name=AnisoBreakup, nodeColor=<rgba 0 1 0.5547 1>}
  }
  MaterialExpressionNamedRerouteDeclaration_77 {
   class=MaterialExpressionNamedRerouteDeclaration
   inputs [None {from=MaterialExpressionSaturate_3}]
   properties {name=Aniso, nodeColor=<rgba 0.0605 1 0.8716 1>}
  }
  MaterialExpressionNamedRerouteDeclaration_78 {
   class=MaterialExpressionNamedRerouteDeclaration
   inputs [None {from=MaterialExpressionTextureSampleParameter2D_1, output=B}]
   properties {name=AnisoMask, nodeColor=<rgba 0.8281 0 1 1>}
  }
  MaterialExpressionNamedRerouteDeclaration_79 {
   class=MaterialExpressionNamedRerouteDeclaration
   inputs [None {from=MaterialExpressionLinearInterpolate_29}]
   properties {name=FurDebugColor, nodeColor=<rgba 0.0312 1 0 1>}
  }
  MaterialExpressionNamedRerouteDeclaration_8 {
   class=MaterialExpressionNamedRerouteDeclaration
   inputs [None {from=MaterialExpressionDeriveNormalZ_0}]
   properties {name=FlowMap, nodeColor=<rgba 1 0.4525 0 1>}
  }
  MaterialExpressionNamedRerouteDeclaration_81 {
   class=MaterialExpressionNamedRerouteDeclaration
   inputs [None {from=MaterialExpressionStaticSwitch_13}]
   properties {name=FurDebugOpacity, nodeColor=<rgba 0.0547 0 1 1>}
  }
  MaterialExpressionNamedRerouteDeclaration_83 {
   class=MaterialExpressionNamedRerouteDeclaration
   inputs [None {from=MaterialExpressionStaticBoolParameter_3}]
   properties {name=UseFurDebug, nodeColor=<rgba 0 0.2266 1 1>}
  }
  MaterialExpressionNamedRerouteDeclaration_84 {
   class=MaterialExpressionNamedRerouteDeclaration
   inputs [None {from=MaterialExpressionLinearInterpolate_6}]
   properties {name=FurLength, nodeColor=<rgba 0.6875 0 1 1>}
  }
  MaterialExpressionNamedRerouteDeclaration_87 {
   class=MaterialExpressionNamedRerouteDeclaration
   inputs [None {from=MaterialExpressionMultiply_48}]
   properties {name=ScaledShellLayer, nodeColor=<rgba 0.2188 1 0 1>}
  }
  MaterialExpressionNamedRerouteDeclaration_89 {
   class=MaterialExpressionNamedRerouteDeclaration
   inputs [None {from=MaterialExpressionStaticSwitchParameter_9}]
   properties {name=ShellCullingMask, nodeColor=<rgba 0.125 1 0 1>}
  }
  MaterialExpressionNamedRerouteDeclaration_91 {
   class=MaterialExpressionNamedRerouteDeclaration
   inputs [None {from=MaterialExpressionStaticSwitch_18}]
   properties {name=FurMin, nodeColor=<rgba 0.3828 0 1 1>}
  }
  MaterialExpressionNamedRerouteDeclaration_92 {
   class=MaterialExpressionNamedRerouteDeclaration
   inputs [None {from=MaterialExpressionStaticSwitch_19}]
   properties {name=Fur_Max, nodeColor=<rgba 0 1 0.0625 1>}
  }
  MaterialExpressionNamedRerouteDeclaration_97 {
   class=MaterialExpressionNamedRerouteDeclaration
   inputs [None {from=MaterialExpressionShaderStageSwitch_0}]
   properties {name=NormalizedShellLayer, nodeColor=<rgba 0.2656 0 1 1>}
  }
  MaterialExpressionNamedRerouteDeclaration_98 {
   class=MaterialExpressionNamedRerouteDeclaration
   inputs [None {from=MaterialExpressionStaticBoolParameter_4}]
   properties {name=UseShellCullingDebug, nodeColor=<rgba 0 0.2266 1 1>}
  }
  MaterialExpressionNamedRerouteDeclaration_99 {
   class=MaterialExpressionNamedRerouteDeclaration
   inputs [None {from=MaterialExpressionTransform_3}]
   properties {name=FlowMap_WS, nodeColor=<rgba 0.4062 1 0 1>}
  }
  MaterialExpressionNamedRerouteUsage_0 {class=MaterialExpressionNamedRerouteUsage}
  MaterialExpressionNamedRerouteUsage_10 {class=MaterialExpressionNamedRerouteUsage}
  MaterialExpressionNamedRerouteUsage_100 {class=MaterialExpressionNamedRerouteUsage}
  MaterialExpressionNamedRerouteUsage_101 {class=MaterialExpressionNamedRerouteUsage}
  MaterialExpressionNamedRerouteUsage_102 {class=MaterialExpressionNamedRerouteUsage}
  MaterialExpressionNamedRerouteUsage_103 {class=MaterialExpressionNamedRerouteUsage}
  MaterialExpressionNamedRerouteUsage_104 {class=MaterialExpressionNamedRerouteUsage}
  MaterialExpressionNamedRerouteUsage_105 {class=MaterialExpressionNamedRerouteUsage}
  MaterialExpressionNamedRerouteUsage_106 {class=MaterialExpressionNamedRerouteUsage}
  MaterialExpressionNamedRerouteUsage_107 {class=MaterialExpressionNamedRerouteUsage}
  MaterialExpressionNamedRerouteUsage_108 {class=MaterialExpressionNamedRerouteUsage}
  MaterialExpressionNamedRerouteUsage_109 {class=MaterialExpressionNamedRerouteUsage}
  MaterialExpressionNamedRerouteUsage_11 {class=MaterialExpressionNamedRerouteUsage}
  MaterialExpressionNamedRerouteUsage_110 {class=MaterialExpressionNamedRerouteUsage}
  MaterialExpressionNamedRerouteUsage_111 {class=MaterialExpressionNamedRerouteUsage}
  MaterialExpressionNamedRerouteUsage_112 {class=MaterialExpressionNamedRerouteUsage}
  MaterialExpressionNamedRerouteUsage_113 {class=MaterialExpressionNamedRerouteUsage}
  MaterialExpressionNamedRerouteUsage_114 {class=MaterialExpressionNamedRerouteUsage}
  MaterialExpressionNamedRerouteUsage_115 {class=MaterialExpressionNamedRerouteUsage}
  MaterialExpressionNamedRerouteUsage_116 {class=MaterialExpressionNamedRerouteUsage}
  MaterialExpressionNamedRerouteUsage_117 {class=MaterialExpressionNamedRerouteUsage}
  MaterialExpressionNamedRerouteUsage_118 {class=MaterialExpressionNamedRerouteUsage}
  MaterialExpressionNamedRerouteUsage_119 {class=MaterialExpressionNamedRerouteUsage}
  MaterialExpressionNamedRerouteUsage_120 {class=MaterialExpressionNamedRerouteUsage}
  MaterialExpressionNamedRerouteUsage_121 {class=MaterialExpressionNamedRerouteUsage}
  MaterialExpressionNamedRerouteUsage_122 {class=MaterialExpressionNamedRerouteUsage}
  MaterialExpressionNamedRerouteUsage_123 {class=MaterialExpressionNamedRerouteUsage}
  MaterialExpressionNamedRerouteUsage_124 {class=MaterialExpressionNamedRerouteUsage}
  MaterialExpressionNamedRerouteUsage_125 {class=MaterialExpressionNamedRerouteUsage}
  MaterialExpressionNamedRerouteUsage_126 {class=MaterialExpressionNamedRerouteUsage}
  MaterialExpressionNamedRerouteUsage_127 {class=MaterialExpressionNamedRerouteUsage}
  MaterialExpressionNamedRerouteUsage_129 {class=MaterialExpressionNamedRerouteUsage}
  MaterialExpressionNamedRerouteUsage_13 {class=MaterialExpressionNamedRerouteUsage}
  MaterialExpressionNamedRerouteUsage_130 {class=MaterialExpressionNamedRerouteUsage}
  MaterialExpressionNamedRerouteUsage_132 {class=MaterialExpressionNamedRerouteUsage}
  MaterialExpressionNamedRerouteUsage_133 {class=MaterialExpressionNamedRerouteUsage}
  MaterialExpressionNamedRerouteUsage_134 {class=MaterialExpressionNamedRerouteUsage}
  MaterialExpressionNamedRerouteUsage_136 {class=MaterialExpressionNamedRerouteUsage}
  MaterialExpressionNamedRerouteUsage_137 {class=MaterialExpressionNamedRerouteUsage}
  MaterialExpressionNamedRerouteUsage_138 {class=MaterialExpressionNamedRerouteUsage}
  MaterialExpressionNamedRerouteUsage_139 {class=MaterialExpressionNamedRerouteUsage}
  MaterialExpressionNamedRerouteUsage_14 {class=MaterialExpressionNamedRerouteUsage}
  MaterialExpressionNamedRerouteUsage_140 {class=MaterialExpressionNamedRerouteUsage}
  MaterialExpressionNamedRerouteUsage_141 {class=MaterialExpressionNamedRerouteUsage}
  MaterialExpressionNamedRerouteUsage_142 {class=MaterialExpressionNamedRerouteUsage}
  MaterialExpressionNamedRerouteUsage_143 {class=MaterialExpressionNamedRerouteUsage}
  MaterialExpressionNamedRerouteUsage_144 {class=MaterialExpressionNamedRerouteUsage}
  MaterialExpressionNamedRerouteUsage_145 {class=MaterialExpressionNamedRerouteUsage}
  MaterialExpressionNamedRerouteUsage_146 {class=MaterialExpressionNamedRerouteUsage}
  MaterialExpressionNamedRerouteUsage_147 {class=MaterialExpressionNamedRerouteUsage}
  MaterialExpressionNamedRerouteUsage_148 {class=MaterialExpressionNamedRerouteUsage}
  MaterialExpressionNamedRerouteUsage_15 {class=MaterialExpressionNamedRerouteUsage}
  MaterialExpressionNamedRerouteUsage_152 {class=MaterialExpressionNamedRerouteUsage}
  MaterialExpressionNamedRerouteUsage_153 {class=MaterialExpressionNamedRerouteUsage}
  MaterialExpressionNamedRerouteUsage_154 {class=MaterialExpressionNamedRerouteUsage}
  MaterialExpressionNamedRerouteUsage_155 {class=MaterialExpressionNamedRerouteUsage}
  MaterialExpressionNamedRerouteUsage_156 {class=MaterialExpressionNamedRerouteUsage}
  MaterialExpressionNamedRerouteUsage_157 {class=MaterialExpressionNamedRerouteUsage}
  MaterialExpressionNamedRerouteUsage_158 {class=MaterialExpressionNamedRerouteUsage}
  MaterialExpressionNamedRerouteUsage_159 {class=MaterialExpressionNamedRerouteUsage}
  MaterialExpressionNamedRerouteUsage_16 {class=MaterialExpressionNamedRerouteUsage}
  MaterialExpressionNamedRerouteUsage_160 {class=MaterialExpressionNamedRerouteUsage}
  MaterialExpressionNamedRerouteUsage_161 {class=MaterialExpressionNamedRerouteUsage}
  MaterialExpressionNamedRerouteUsage_162 {class=MaterialExpressionNamedRerouteUsage}
  MaterialExpressionNamedRerouteUsage_163 {class=MaterialExpressionNamedRerouteUsage}
  MaterialExpressionNamedRerouteUsage_164 {class=MaterialExpressionNamedRerouteUsage}
  MaterialExpressionNamedRerouteUsage_165 {class=MaterialExpressionNamedRerouteUsage}
  MaterialExpressionNamedRerouteUsage_166 {class=MaterialExpressionNamedRerouteUsage}
  MaterialExpressionNamedRerouteUsage_167 {class=MaterialExpressionNamedRerouteUsage}
  MaterialExpressionNamedRerouteUsage_168 {class=MaterialExpressionNamedRerouteUsage}
  MaterialExpressionNamedRerouteUsage_169 {class=MaterialExpressionNamedRerouteUsage}
  MaterialExpressionNamedRerouteUsage_170 {class=MaterialExpressionNamedRerouteUsage}
  MaterialExpressionNamedRerouteUsage_171 {class=MaterialExpressionNamedRerouteUsage}
  MaterialExpressionNamedRerouteUsage_172 {class=MaterialExpressionNamedRerouteUsage}
  MaterialExpressionNamedRerouteUsage_173 {class=MaterialExpressionNamedRerouteUsage}
  MaterialExpressionNamedRerouteUsage_174 {class=MaterialExpressionNamedRerouteUsage}
  MaterialExpressionNamedRerouteUsage_176 {class=MaterialExpressionNamedRerouteUsage}
  MaterialExpressionNamedRerouteUsage_177 {class=MaterialExpressionNamedRerouteUsage}
  MaterialExpressionNamedRerouteUsage_178 {class=MaterialExpressionNamedRerouteUsage}
  MaterialExpressionNamedRerouteUsage_179 {class=MaterialExpressionNamedRerouteUsage}
  MaterialExpressionNamedRerouteUsage_180 {class=MaterialExpressionNamedRerouteUsage}
  MaterialExpressionNamedRerouteUsage_181 {class=MaterialExpressionNamedRerouteUsage}
  MaterialExpressionNamedRerouteUsage_182 {class=MaterialExpressionNamedRerouteUsage}
  MaterialExpressionNamedRerouteUsage_183 {class=MaterialExpressionNamedRerouteUsage}
  MaterialExpressionNamedRerouteUsage_184 {class=MaterialExpressionNamedRerouteUsage}
  MaterialExpressionNamedRerouteUsage_2 {class=MaterialExpressionNamedRerouteUsage}
  MaterialExpressionNamedRerouteUsage_3 {class=MaterialExpressionNamedRerouteUsage}
  MaterialExpressionNamedRerouteUsage_34 {class=MaterialExpressionNamedRerouteUsage}
  MaterialExpressionNamedRerouteUsage_39 {class=MaterialExpressionNamedRerouteUsage}
  MaterialExpressionNamedRerouteUsage_4 {class=MaterialExpressionNamedRerouteUsage}
  MaterialExpressionNamedRerouteUsage_41 {class=MaterialExpressionNamedRerouteUsage}
  MaterialExpressionNamedRerouteUsage_47 {class=MaterialExpressionNamedRerouteUsage}
  MaterialExpressionNamedRerouteUsage_49 {class=MaterialExpressionNamedRerouteUsage}
  MaterialExpressionNamedRerouteUsage_5 {class=MaterialExpressionNamedRerouteUsage}
  MaterialExpressionNamedRerouteUsage_50 {class=MaterialExpressionNamedRerouteUsage}
  MaterialExpressionNamedRerouteUsage_6 {class=MaterialExpressionNamedRerouteUsage}
  MaterialExpressionNamedRerouteUsage_7 {class=MaterialExpressionNamedRerouteUsage}
  MaterialExpressionNamedRerouteUsage_78 {class=MaterialExpressionNamedRerouteUsage}
  MaterialExpressionNamedRerouteUsage_79 {class=MaterialExpressionNamedRerouteUsage}
  MaterialExpressionNamedRerouteUsage_8 {class=MaterialExpressionNamedRerouteUsage}
  MaterialExpressionNamedRerouteUsage_80 {class=MaterialExpressionNamedRerouteUsage}
  MaterialExpressionNamedRerouteUsage_81 {class=MaterialExpressionNamedRerouteUsage}
  MaterialExpressionNamedRerouteUsage_82 {class=MaterialExpressionNamedRerouteUsage}
  MaterialExpressionNamedRerouteUsage_83 {class=MaterialExpressionNamedRerouteUsage}
  MaterialExpressionNamedRerouteUsage_85 {class=MaterialExpressionNamedRerouteUsage}
  MaterialExpressionNamedRerouteUsage_86 {class=MaterialExpressionNamedRerouteUsage}
  MaterialExpressionNamedRerouteUsage_87 {class=MaterialExpressionNamedRerouteUsage}
  MaterialExpressionNamedRerouteUsage_88 {class=MaterialExpressionNamedRerouteUsage}
  MaterialExpressionNamedRerouteUsage_9 {class=MaterialExpressionNamedRerouteUsage}
  MaterialExpressionNamedRerouteUsage_92 {class=MaterialExpressionNamedRerouteUsage}
  MaterialExpressionNamedRerouteUsage_96 {class=MaterialExpressionNamedRerouteUsage}
  MaterialExpressionNamedRerouteUsage_97 {class=MaterialExpressionNamedRerouteUsage}
  MaterialExpressionNamedRerouteUsage_98 {class=MaterialExpressionNamedRerouteUsage}
  MaterialExpressionNamedRerouteUsage_99 {class=MaterialExpressionNamedRerouteUsage}
  MaterialExpressionNormalize_1 {class=MaterialExpressionNormalize, inputs [VectorInput {from=MaterialExpressionNamedRerouteUsage_174, output=""}]}
  MaterialExpressionNormalize_10 {class=MaterialExpressionNormalize, inputs [VectorInput {from=MaterialExpressionMultiply_62, output=""}]}
  MaterialExpressionNormalize_13 {class=MaterialExpressionNormalize, inputs [VectorInput {from=MaterialExpressionCrossProduct_3, output=""}]}
  MaterialExpressionNormalize_14 {class=MaterialExpressionNormalize, inputs [VectorInput {from=MaterialExpressionNamedRerouteUsage_177, output=""}]}
  MaterialExpressionNormalize_4 {class=MaterialExpressionNormalize, inputs [VectorInput {from=MaterialExpressionConvert_1, output=""}]}
  MaterialExpressionNormalize_5 {class=MaterialExpressionNormalize, inputs [VectorInput {from=MaterialExpressionMultiply_35, output=""}]}
  MaterialExpressionNormalize_6 {class=MaterialExpressionNormalize, inputs [VectorInput {from=MaterialExpressionMultiply_21, output=""}]}
  MaterialExpressionNormalize_7 {class=MaterialExpressionNormalize, inputs [VectorInput {from=MaterialExpressionMultiply_52, output=""}]}
  MaterialExpressionNormalize_9 {class=MaterialExpressionNormalize, inputs [VectorInput {from=MaterialExpressionMultiply_61, output=""}]}
  MaterialExpressionOneMinus_0 {class=MaterialExpressionOneMinus, inputs [None {from=MaterialExpressionStaticSwitch_1}]}
  MaterialExpressionOneMinus_1 {class=MaterialExpressionOneMinus, inputs [None {from=MaterialExpressionStaticSwitch_0}]}
  MaterialExpressionOneMinus_10 {class=MaterialExpressionOneMinus, inputs [None {from=MaterialExpressionSaturate_12}]}
  MaterialExpressionOneMinus_11 {class=MaterialExpressionOneMinus, inputs [None {from=MaterialExpressionSaturate_15}]}
  MaterialExpressionOneMinus_12 {class=MaterialExpressionOneMinus, inputs [None {from=MaterialExpressionScalarParameter_67}]}
  MaterialExpressionOneMinus_13 {class=MaterialExpressionOneMinus, inputs [None {from=MaterialExpressionSaturate_17}]}
  MaterialExpressionOneMinus_14 {class=MaterialExpressionOneMinus, inputs [None {from=MaterialExpressionNamedRerouteDeclaration_118}]}
  MaterialExpressionOneMinus_3 {class=MaterialExpressionOneMinus, inputs [None {from=MaterialExpressionStaticSwitch_17}]}
  MaterialExpressionOneMinus_30 {class=MaterialExpressionOneMinus, inputs [None {from=MaterialExpressionMultiply_143}]}
  MaterialExpressionOneMinus_5 {class=MaterialExpressionOneMinus, inputs [None {from=MaterialExpressionStaticSwitch_16}]}
  MaterialExpressionOneMinus_8 {class=MaterialExpressionOneMinus, inputs [None {from=MaterialExpressionNamedRerouteUsage_83}]}
  MaterialExpressionOneMinus_9 {class=MaterialExpressionOneMinus, inputs [None {from=MaterialExpressionNamedRerouteUsage_87}]}
  MaterialExpressionPixelNormalWS_0 {class=MaterialExpressionPixelNormalWS, inputs [], properties {}}
  MaterialExpressionPower_0 {
   class=MaterialExpressionPower
   inputs [
    Base {from=MaterialExpressionNamedRerouteUsage_2}
    Exp {from=MaterialExpressionScalarParameter_2}
   ]
  }
  MaterialExpressionPower_10 {
   class=MaterialExpressionPower
   inputs [
    Base {from=MaterialExpressionNamedRerouteUsage_50}
    Exp {from=MaterialExpressionScalarParameter_59}
   ]
  }
  MaterialExpressionPower_11 {
   class=MaterialExpressionPower
   inputs [
    Base {from=MaterialExpressionMultiply_36}
    Exp {from=MaterialExpressionConstant_10}
   ]
   properties {constExponent=3}
  }
  MaterialExpressionPower_12 {
   class=MaterialExpressionPower
   inputs [
    Base {from=MaterialExpressionOneMinus_8}
    Exp {from=MaterialExpressionScalarParameter_63}
   ]
  }
  MaterialExpressionPower_14 {
   class=MaterialExpressionPower
   inputs [
    Base {from=MaterialExpressionNamedRerouteUsage_82}
    Exp {from=MaterialExpressionScalarParameter_65}
   ]
  }
  MaterialExpressionPower_15 {
   class=MaterialExpressionPower
   inputs [
    Base {from=MaterialExpressionNamedRerouteUsage_101}
    Exp {from=MaterialExpressionScalarParameter_73}
   ]
  }
  MaterialExpressionPower_16 {
   class=MaterialExpressionPower
   inputs [
    Base {from=MaterialExpressionConstant_16}
    Exp {from=MaterialExpressionCeil_1}
   ]
  }
  MaterialExpressionPower_6 {
   class=MaterialExpressionPower
   inputs [
    Base {from=MaterialExpressionNamedRerouteUsage_85}
    Exp {from=MaterialExpressionScalarParameter_33}
   ]
  }
  MaterialExpressionPower_7 {
   class=MaterialExpressionPower
   inputs [
    Base {from=MaterialExpressionAdd_7}
    Exp {from=MaterialExpressionScalarParameter_54}
   ]
  }
  MaterialExpressionPower_9 {
   class=MaterialExpressionPower
   inputs [
    Base {from=MaterialExpressionNamedRerouteUsage_119}
    Exp {from=MaterialExpressionScalarParameter_57}
   ]
  }
  MaterialExpressionSaturate_1 {class=MaterialExpressionSaturate, inputs [None {from=MaterialExpressionPower_11}]}
  MaterialExpressionSaturate_12 {class=MaterialExpressionSaturate, inputs [None {from=MaterialExpressionMultiply_46}]}
  MaterialExpressionSaturate_15 {class=MaterialExpressionSaturate, inputs [None {from=MaterialExpressionScalarParameter_76}]}
  MaterialExpressionSaturate_16 {class=MaterialExpressionSaturate, inputs [None {from=MaterialExpressionDivide_2}]}
  MaterialExpressionSaturate_17 {class=MaterialExpressionSaturate, inputs [None {from=MaterialExpressionInverseLinearInterpolate_4}]}
  MaterialExpressionSaturate_18 {class=MaterialExpressionSaturate, inputs [None {from=MaterialExpressionInverseLinearInterpolate_5}]}
  MaterialExpressionSaturate_19 {class=MaterialExpressionSaturate, inputs [None {from=MaterialExpressionNamedRerouteDeclaration_110}]}
  MaterialExpressionSaturate_2 {class=MaterialExpressionSaturate, inputs [None {from=MaterialExpressionDivide_1}]}
  MaterialExpressionSaturate_20 {class=MaterialExpressionSaturate, inputs [None {from=MaterialExpressionAdd_16}]}
  MaterialExpressionSaturate_21 {class=MaterialExpressionSaturate, inputs [None {from=MaterialExpressionLinearInterpolate_41}]}
  MaterialExpressionSaturate_22 {class=MaterialExpressionSaturate, inputs [None {from=MaterialExpressionInverseLinearInterpolate_7}]}
  MaterialExpressionSaturate_3 {class=MaterialExpressionSaturate, inputs [None {from=MaterialExpressionMultiply_45}]}
  MaterialExpressionScalarParameter_0 {
   class=MaterialExpressionScalarParameter
   properties {
    defaultValue=1.5
    desc=How thick is the fur strand at the base/root.
    group=Fur_Shell_02 - Strands
    parameterName=Secondary_FurThicknessRoot
    sortPriority=55
   }
  }
  MaterialExpressionScalarParameter_1 {
   class=MaterialExpressionScalarParameter
   properties {
    defaultValue=0.5
    desc=How thick is the fur strand at the tip.
    group=Fur_Shell_02 - Strands
    parameterName=Secondary_FurThicknessTip
    sortPriority=56
   }
  }
  MaterialExpressionScalarParameter_100 {
   class=MaterialExpressionScalarParameter
   properties {group=Fur_Shell_02 - Strands, parameterName=Secondary_FurPatternUVOffsetY, sortPriority=6}
  }
  MaterialExpressionScalarParameter_101 {
   class=MaterialExpressionScalarParameter
   properties {defaultValue=10, group=Fur_Shell_07 - Far Fur, parameterName=FarFurPattern_02_UVTile, sortPriority=10}
  }
  MaterialExpressionScalarParameter_102 {
   class=MaterialExpressionScalarParameter
   properties {
    defaultValue=0.5
    desc=Blends Fur pattern 1 into the black areas of the Fur pattern 2 texture
    group=Fur_Shell_02 - Strands
    parameterName=FurPatternBlend
    sortPriority=7
   }
  }
  MaterialExpressionScalarParameter_103 {
   class=MaterialExpressionScalarParameter
   properties {
    defaultValue=1
    desc=Strength of the normal map.
    group=Fur_Shell_02 - Strands
    parameterName=Secondary_FurPatternNormalStrength
    sortPriority=12
   }
  }
  MaterialExpressionScalarParameter_104 {
   class=MaterialExpressionScalarParameter
   properties {
    defaultValue=1
    desc=Strength of the normal map.
    group=Fur_Shell_07 - Far Fur
    parameterName=FarFurPattern_02_NormalStrength
    sortPriority=20
   }
  }
  MaterialExpressionScalarParameter_105 {
   class=MaterialExpressionScalarParameter
   properties {defaultValue=5, group=Fur_Shell_08 - Distortion, parameterName=DistortionStrength, sortPriority=2}
  }
  MaterialExpressionScalarParameter_106 {
   class=MaterialExpressionScalarParameter
   properties {defaultValue=5, group=Fur_Shell_08 - Distortion, parameterName=Secondary_DistortionStrength, sortPriority=2}
  }
  MaterialExpressionScalarParameter_11 {
   class=MaterialExpressionScalarParameter
   properties {defaultValue=1, desc=Strength of the normal map., group=Fur_Shell_02 - Strands, parameterName=FurPatternNormalStrength, sortPriority=12}
  }
  MaterialExpressionScalarParameter_12 {
   class=MaterialExpressionScalarParameter
   properties {defaultValue=0.2, group=Fur_Shell_01 - Base Settings, parameterName=FurLengthMin, sortPriority=8}
  }
  MaterialExpressionScalarParameter_13 {
   class=MaterialExpressionScalarParameter
   properties {
    defaultValue=1.5
    desc=How thick is the fur strand at the base/root.
    group=Fur_Shell_02 - Strands
    parameterName=FurThicknessRoot
    sortPriority=55
   }
  }
  MaterialExpressionScalarParameter_147 {class=MaterialExpressionScalarParameter, properties {defaultValue=0.1, group=Fur_Shell_08 - Distortion, parameterName=Wave, sortPriority=1}}
  MaterialExpressionScalarParameter_148 {
   class=MaterialExpressionScalarParameter
   properties {defaultValue=0.1, group=Fur_Shell_08 - Distortion, parameterName=WaveFrequency, sortPriority=3}
  }
  MaterialExpressionScalarParameter_168 {
   class=MaterialExpressionScalarParameter
   properties {defaultValue=5, group=Fur_Shell_08 - Distortion, parameterName=MaxWaveAmp, sortPriority=2}
  }
  MaterialExpressionScalarParameter_169 {class=MaterialExpressionScalarParameter, properties {group=Fur_Shell_08 - Distortion, parameterName=RandomDIstortion, sortPriority=4}}
  MaterialExpressionScalarParameter_170 {
   class=MaterialExpressionScalarParameter
   properties {defaultValue=1, group=Fur_Shell_08 - Distortion, parameterName=RandomSeed, sortPriority=4}
  }
  MaterialExpressionScalarParameter_2 {
   class=MaterialExpressionScalarParameter
   properties {
    defaultValue=1
    desc=How is the transition between root and tip thickness distributed along the strand.
    group=Fur_Shell_02 - Strands
    parameterName=Secondary_FurThicknessRootTipDistribution
    sortPriority=57
   }
  }
  MaterialExpressionScalarParameter_24 {
   class=MaterialExpressionScalarParameter
   properties {defaultValue=0.5, desc=How thick is the fur strand at the tip., group=Fur_Shell_02 - Strands, parameterName=FurThicknessTip, sortPriority=56}
  }
  MaterialExpressionScalarParameter_3 {
   class=MaterialExpressionScalarParameter
   properties {
    defaultValue=1.5
    desc=How thick is the fur strand at the base/root.
    group=Fur_Shell_07 - Far Fur
    parameterName=Secondary_FarFurThicknessRoot
    sortPriority=11
   }
  }
  MaterialExpressionScalarParameter_33 {
   class=MaterialExpressionScalarParameter
   properties {
    defaultValue=1
    desc=How is the transition between root and tip thickness distributed along the strand.
    group=Fur_Shell_02 - Strands
    parameterName=FurThicknessRootTipDistribution
    sortPriority=57
   }
  }
  MaterialExpressionScalarParameter_37 {class=MaterialExpressionScalarParameter, properties {group=DeRez, parameterName=NewDissolveGradient}}
  MaterialExpressionScalarParameter_4 {
   class=MaterialExpressionScalarParameter
   properties {
    defaultValue=0.5
    desc=How thick is the fur strand at the tip.
    group=Fur_Shell_07 - Far Fur
    parameterName=Secondary_FarFurThicknessTip
    sortPriority=12
   }
  }
  MaterialExpressionScalarParameter_49 {
   class=MaterialExpressionScalarParameter
   properties {defaultValue=10, group=Fur_Shell_02 - Strands, parameterName=FurPatternUVTile, sortPriority=3}
  }
  MaterialExpressionScalarParameter_5 {
   class=MaterialExpressionScalarParameter
   properties {
    defaultValue=0.4
    desc="Flatten the tip of the modeleded clumps "
    group=Fur_Shell_03 - WPO
    parameterName="NormalVSTipStrength_ModeledClumps (Controlled by length map)"
    sortPriority=5
   }
  }
  MaterialExpressionScalarParameter_51 {class=MaterialExpressionScalarParameter, properties {group=Fur_Shell_02 - Strands, parameterName=FurPatternUVOffsetX, sortPriority=3}}
  MaterialExpressionScalarParameter_52 {class=MaterialExpressionScalarParameter, properties {group=Fur_Shell_02 - Strands, parameterName=FurPatternUVOffsetY, sortPriority=3}}
  MaterialExpressionScalarParameter_53 {
   class=MaterialExpressionScalarParameter
   properties {defaultValue=1, group=Fur_Shell_01 - Base Settings, parameterName=FurLengthMax, sortPriority=9}
  }
  MaterialExpressionScalarParameter_54 {
   class=MaterialExpressionScalarParameter
   properties {
    defaultValue=1
    desc=Distribution of the flow map effect on shells between root and tip of the fur strand.
    group=Fur_Shell_03 - WPO
    parameterName=FlowMapRootTipDistribution
    sortPriority=2
   }
  }
  MaterialExpressionScalarParameter_55 {
   class=MaterialExpressionScalarParameter
   properties {defaultValue=1, group=Fur_Shell_03 - WPO, parameterName=FlowMapStrength, sortPriority=1}
  }
  MaterialExpressionScalarParameter_56 {
   class=MaterialExpressionScalarParameter
   properties {
    desc="Shells offset strength along the vert normals. Controls distance between shells at root "
    group=Fur_Shell_03 - WPO
    parameterName=NormalVSRootStrength
    sortPriority=4
   }
  }
  MaterialExpressionScalarParameter_57 {
   class=MaterialExpressionScalarParameter
   properties {
    defaultValue=1
    desc=Distribution of the shells along the vertex normals between root and tip values. Controlling overall distance between shells along strand.
    group=Fur_Shell_03 - WPO
    parameterName=NormalVSRootTipDistribution
    sortPriority=6
   }
  }
  MaterialExpressionScalarParameter_58 {
   class=MaterialExpressionScalarParameter
   properties {
    defaultValue=0.4
    desc="Shells offset strength along the vert normals. Controls distance between shells at tip. "
    group=Fur_Shell_03 - WPO
    parameterName=NormalVSTipStrength
    sortPriority=5
   }
  }
  MaterialExpressionScalarParameter_59 {
   class=MaterialExpressionScalarParameter
   properties {defaultValue=1, group=Fur_Shell_02 - Strands, parameterName=FurNormalMaskExponent, sortPriority=17}
  }
  MaterialExpressionScalarParameter_6 {
   class=MaterialExpressionScalarParameter
   properties {
    controlType=Enumeration
    desc="Flatten the tip of the modeleded clumps \r\n\r\nUses Length Map to control clump areas"
    enumeration=/Game/Global/Materials/Enums/E_Mat_YesNo
    group=Fur_Shell_03 - WPO
    parameterName=UseModeledClumpTipStrengthControl
    sortPriority=5
   }
  }
  MaterialExpressionScalarParameter_60 {class=MaterialExpressionScalarParameter, properties {defaultValue=1, group=Fur_Shell_02 - Strands, parameterName=FurBacklightContribution}}
  MaterialExpressionScalarParameter_61 {
   class=MaterialExpressionScalarParameter
   properties {
    desc="Reduces the Mesh's normal map against the fur "
    group=Fur_Shell_01 - Base Settings
    parameterName=MeshNormalMapFlatten
    sortPriority=40
   }
  }
  MaterialExpressionScalarParameter_62 {class=MaterialExpressionScalarParameter, properties {defaultValue=0.5, group=Fur_Shell_02 - Strands, parameterName=FurColorRootMulitplier}}
  MaterialExpressionScalarParameter_63 {class=MaterialExpressionScalarParameter, properties {defaultValue=1, group=Fur_Shell_02 - Strands, parameterName=FurColorRootExponent}}
  MaterialExpressionScalarParameter_64 {class=MaterialExpressionScalarParameter, properties {defaultValue=1, group=Fur_Shell_02 - Strands, parameterName=FurColorTipMulitplier}}
  MaterialExpressionScalarParameter_65 {class=MaterialExpressionScalarParameter, properties {defaultValue=1, group=Fur_Shell_02 - Strands, parameterName=FurColorTipExponent}}
  MaterialExpressionScalarParameter_66 {
   class=MaterialExpressionScalarParameter
   properties {desc="Fur Clothing Mask Toggle ", group=Fur_Shell_01 - Base Settings, parameterName=FurClothing, sortPriority=41}
  }
  MaterialExpressionScalarParameter_67 {
   class=MaterialExpressionScalarParameter
   properties {
    defaultValue=0.1
    desc=Buffer stoping the flow map from effecting the first few shells to avoid shell missalignment relative to the mesh
    group=Fur_Shell_03 - WPO
    parameterName=FlowMapMaskRootBuffer
    sortPriority=3
   }
  }
  MaterialExpressionScalarParameter_68 {class=MaterialExpressionScalarParameter, properties {group=Fur_Shell_04 - Aniso, parameterName=AnisoNoiseStrength, sortPriority=41}}
  MaterialExpressionScalarParameter_69 {
   class=MaterialExpressionScalarParameter
   properties {defaultValue=3, group=Fur_Shell_04 - Aniso, parameterName=AnisoNoiseTiling, sortPriority=42}
  }
  MaterialExpressionScalarParameter_70 {class=MaterialExpressionScalarParameter, properties {group=Fur_Shell_04 - Aniso, parameterName=AnisoRootAmount, sortPriority=31}}
  MaterialExpressionScalarParameter_71 {class=MaterialExpressionScalarParameter, properties {defaultValue=0.6, group=Fur_Shell_04 - Aniso, parameterName=AnisoTipAmount}}
  MaterialExpressionScalarParameter_72 {
   class=MaterialExpressionScalarParameter
   properties {defaultValue=1, group=Fur_Shell_04 - Aniso, parameterName=AnisoMaskStrength, sortPriority=35}
  }
  MaterialExpressionScalarParameter_73 {
   class=MaterialExpressionScalarParameter
   properties {
    defaultValue=1
    desc=How is the transition between root and tip aniso effect distributed along the strand.
    group=Fur_Shell_04 - Aniso
    parameterName=AnisoRootTipDistribution
    sortPriority=33
   }
  }
  MaterialExpressionScalarParameter_74 {class=MaterialExpressionScalarParameter, properties {defaultValue=50, group=Fur_Shell_06 - Debug, parameterName=FurDebugClipShellIndex}}
  MaterialExpressionScalarParameter_75 {class=MaterialExpressionScalarParameter, properties {defaultValue=1, group=Fur_Shell_06 - Debug, parameterName=FurDebugColorStrength}}
  MaterialExpressionScalarParameter_76 {class=MaterialExpressionScalarParameter, properties {group=Fur_Shell_05 - Culling, parameterName=FurShellCullAmount, sortPriority=2}}
  MaterialExpressionScalarParameter_77 {
   class=MaterialExpressionScalarParameter
   properties {defaultValue=50, group=Fur_Shell_05 - Culling, parameterName=FurShellCullNearDistance, sortPriority=3}
  }
  MaterialExpressionScalarParameter_78 {
   class=MaterialExpressionScalarParameter
   properties {defaultValue=100, group=Fur_Shell_05 - Culling, parameterName=FurShellCullRange, sortPriority=4}
  }
  MaterialExpressionScalarParameter_79 {class=MaterialExpressionScalarParameter, properties {group=Fur_Shell_05 - Culling, parameterName=RenormalizeShellLayers, sortPriority=10}}
  MaterialExpressionScalarParameter_80 {
   class=MaterialExpressionScalarParameter
   properties {defaultValue=-1, group=Fur_Shell_02 - Strands, parameterName=FurPatternMipBias, sortPriority=1}
  }
  MaterialExpressionScalarParameter_81 {
   class=MaterialExpressionScalarParameter
   properties {defaultValue=1, group=Fur_Shell_05 - Culling, parameterName=FurShellMinCullLayer, sortPriority=5}
  }
  MaterialExpressionScalarParameter_82 {class=MaterialExpressionScalarParameter, properties {group=Fur_Shell_07 - Far Fur, parameterName=FarFurNearDistance, sortPriority=1}}
  MaterialExpressionScalarParameter_83 {class=MaterialExpressionScalarParameter, properties {group=Fur_Shell_07 - Far Fur, parameterName=FarFurRange, sortPriority=2}}
  MaterialExpressionScalarParameter_84 {
   class=MaterialExpressionScalarParameter
   properties {defaultValue=2, group=Fur_Shell_05 - Culling, parameterName=ShellLODCullMin, sortPriority=103}
  }
  MaterialExpressionScalarParameter_85 {
   class=MaterialExpressionScalarParameter
   properties {defaultValue=50, group=Fur_Shell_05 - Culling, parameterName=ShellLODCullDistance, sortPriority=101}
  }
  MaterialExpressionScalarParameter_86 {
   class=MaterialExpressionScalarParameter
   properties {defaultValue=250, group=Fur_Shell_05 - Culling, parameterName=ShellLODCullRange, sortPriority=102}
  }
  MaterialExpressionScalarParameter_87 {
   class=MaterialExpressionScalarParameter
   properties {defaultValue=16, group=Fur_Shell_05 - Culling, parameterName=ShellLODCullMax, sortPriority=104}
  }
  MaterialExpressionScalarParameter_88 {
   class=MaterialExpressionScalarParameter
   properties {defaultValue=8, group=Fur_Shell_05 - Culling, parameterName=ShellLODMinLayerCount, sortPriority=105}
  }
  MaterialExpressionScalarParameter_89 {
   class=MaterialExpressionScalarParameter
   properties {defaultValue=10, group=Fur_Shell_07 - Far Fur, parameterName=FarFurPattern_01_UVTile, sortPriority=10}
  }
  MaterialExpressionScalarParameter_90 {
   class=MaterialExpressionScalarParameter
   properties {
    defaultValue=1
    desc=Strength of the normal map.
    group=Fur_Shell_07 - Far Fur
    parameterName=FarFurPattern_01_NormalStrength
    sortPriority=20
   }
  }
  MaterialExpressionScalarParameter_91 {
   class=MaterialExpressionScalarParameter
   properties {
    defaultValue=1.5
    desc=How thick is the fur strand at the base/root.
    group=Fur_Shell_07 - Far Fur
    parameterName=FarFurThicknessRoot
    sortPriority=11
   }
  }
  MaterialExpressionScalarParameter_92 {
   class=MaterialExpressionScalarParameter
   properties {
    defaultValue=0.5
    desc=How thick is the fur strand at the tip.
    group=Fur_Shell_07 - Far Fur
    parameterName=FarFurThicknessTip
    sortPriority=12
   }
  }
  MaterialExpressionScalarParameter_93 {
   class=MaterialExpressionScalarParameter
   properties {defaultValue=1, group=Fur_Shell_07 - Far Fur, parameterName=FarFurLengthMultiplier, sortPriority=13}
  }
  MaterialExpressionScalarParameter_94 {class=MaterialExpressionScalarParameter, properties {defaultValue=99}}
  MaterialExpressionScalarParameter_95 {
   class=MaterialExpressionScalarParameter
   properties {defaultValue=1, group=Fur_Shell_07 - Far Fur, parameterName=FarFlowMapStrength, sortPriority=13}
  }
  MaterialExpressionScalarParameter_96 {class=MaterialExpressionScalarParameter, properties {group=Fur_Shell_05 - Culling, parameterName=FurShellCullClothingBias, sortPriority=100}}
  MaterialExpressionScalarParameter_97 {
   class=MaterialExpressionScalarParameter
   properties {defaultValue=0.5, group=Fur_Shell_05 - Culling, parameterName=FOVDistanceMultiplier, sortPriority=100}
  }
  MaterialExpressionScalarParameter_98 {
   class=MaterialExpressionScalarParameter
   properties {defaultValue=10, group=Fur_Shell_02 - Strands, parameterName=Secondary_FurPatternUVTile, sortPriority=6}
  }
  MaterialExpressionScalarParameter_99 {
   class=MaterialExpressionScalarParameter
   properties {group=Fur_Shell_02 - Strands, parameterName=Secondary_FurPatternUVOffsetX, sortPriority=6}
  }
  MaterialExpressionSetMaterialAttributes_0 {
   class=MaterialExpressionSetMaterialAttributes
   inputs [
    MaterialAttributes {from=MaterialExpressionNamedRerouteUsage_16, output=""}
    BaseColor {from=MaterialExpressionStaticSwitch_10, output=""}
    Anisotropy {from=MaterialExpressionNamedRerouteUsage_98, output=""}
    OpacityMask {from=MaterialExpressionStaticSwitch_12, output=""}
    Normal {from=MaterialExpressionNamedRerouteUsage_14, output=""}
    Tangent {from=MaterialExpressionNamedRerouteUsage_99, output=""}
    WorldPositionOffset {from=MaterialExpressionNamedRerouteUsage_39, output=""}
    ShadingModel {from=MaterialExpressionShadingModel_0, output=""}
   ]
   properties {attributeSetTypes [
     #1
     #4
     #5
     #2
     #3
     #6
     #7
    ]}
  }
  MaterialExpressionShaderStageSwitch_0 {
   class=MaterialExpressionShaderStageSwitch
   inputs [
    PixelShader {from=MaterialExpressionVertexInterpolator_2, output=PS}
    VertexShader {from=ShellMeshMaterialExpressionNormalizedShellLayer_1, output=""}
   ]
   properties {}
  }
  MaterialExpressionShadingModel_0 {class=MaterialExpressionShadingModel, inputs [], properties {}}
  MaterialExpressionSign_7 {class=MaterialExpressionSign, inputs [None {from=MaterialExpressionSubtract_14, output=""}], properties {}}
  MaterialExpressionSine_10 {class=MaterialExpressionSine, inputs [None {from=MaterialExpressionMultiply_127}]}
  MaterialExpressionSine_11 {class=MaterialExpressionSine, inputs [None {from=MaterialExpressionDotProduct_3}]}
  MaterialExpressionStaticBoolParameter_3 {class=MaterialExpressionStaticBoolParameter, inputs [], properties {group=Fur_Shell_06 - Debug, parameterName=UseFurShellDebug}}
  MaterialExpressionStaticBoolParameter_4 {class=MaterialExpressionStaticBoolParameter, inputs [], properties {group=Fur_Shell_06 - Debug, parameterName=UseFurShellCullingDebug}}
  MaterialExpressionStaticBoolParameter_5 {class=MaterialExpressionStaticBoolParameter, inputs [], properties {group=Fur_Shell_07 - Far Fur, parameterName=UseFarFur, sortPriority=-1}}
  MaterialExpressionStaticBoolParameter_6 {
   class=MaterialExpressionStaticBoolParameter
   inputs []
   properties {group=Fur_Shell_01 - Base Settings, parameterName=UseSecondaryFurStyle, sortPriority=0}
  }
  MaterialExpressionStaticSwitchParameter_10 {
   class=MaterialExpressionStaticSwitchParameter
   inputs [
    True {from=MaterialExpressionAdd_19, output=""}
    False {from=MaterialExpressionAdd_4, output=""}
   ]
   properties {group=Fur_Shell_08 - Distortion, parameterName=UseDistortion, sortPriority=-5}
  }
  MaterialExpressionStaticSwitchParameter_12 {
   class=MaterialExpressionStaticSwitchParameter
   inputs [
    True {from=MaterialExpressionStaticSwitch_23, output=""}
    False {from=MaterialExpressionConstant3Vector_6, output=""}
   ]
   properties {group=Fur_Shell_01 - Base Settings, parameterName=UseSecondaryNormalMap, sortPriority=0}
  }
  MaterialExpressionStaticSwitchParameter_5 {
   class=MaterialExpressionStaticSwitchParameter
   inputs [
    True {from=MaterialExpressionLinearInterpolate_24, output=""}
    False {from=MaterialExpressionNormalize_1, output=""}
   ]
   properties {
    defaultValue=true
    desc="Reduces fur normal map intensiy on areas of the mask "
    group=Fur_Shell_02 - Strands
    parameterName=UseFurNormalMask
    sortPriority=16
   }
  }
  MaterialExpressionStaticSwitchParameter_6 {
   class=MaterialExpressionStaticSwitchParameter
   inputs [
    True {from=MaterialExpressionNamedRerouteUsage_96, output=""}
    False {from=MaterialExpressionNamedRerouteUsage_97, output=""}
   ]
   properties {group=Fur_Shell_01 - Base Settings, parameterName=UseFurShells, sortPriority=-2}
  }
  MaterialExpressionStaticSwitchParameter_7 {
   class=MaterialExpressionStaticSwitchParameter
   inputs [
    True {from=MaterialExpressionOneMinus_13, output=""}
    False {from=MaterialExpressionOneMinus_11, output=""}
   ]
   properties {group=Fur_Shell_05 - Culling, parameterName=UseCameraDistanceFurCulling, sortPriority=1}
  }
  MaterialExpressionStaticSwitchParameter_8 {
   class=MaterialExpressionStaticSwitchParameter
   inputs [
    True {from=MaterialExpressionMultiply_47, output=""}
    False {from=MaterialExpressionLinearInterpolate_27, output=""}
   ]
   properties {group=Fur_Shell_04 - Aniso, parameterName=UseAnisoNoise, sortPriority=40}
  }
  MaterialExpressionStaticSwitchParameter_9 {
   class=MaterialExpressionStaticSwitchParameter
   inputs [
    True {from=MaterialExpressionAdd_14, output=""}
    False {from=MaterialExpressionMultiply_50, output=""}
   ]
   properties {group=Fur_Shell_05 - Culling, parameterName=UseShellLODCulling, sortPriority=100}
  }
  MaterialExpressionStaticSwitch_0 {
   class=MaterialExpressionStaticSwitch
   inputs [
    True {from=MaterialExpressionLinearInterpolate_1, output=""}
    False {from=MaterialExpressionScalarParameter_0, output=""}
    Value {from=MaterialExpressionNamedRerouteUsage_5, output=""}
   ]
   properties {}
  }
  MaterialExpressionStaticSwitch_1 {
   class=MaterialExpressionStaticSwitch
   inputs [
    True {from=MaterialExpressionLinearInterpolate_2, output=""}
    False {from=MaterialExpressionScalarParameter_1, output=""}
    Value {from=MaterialExpressionNamedRerouteUsage_6, output=""}
   ]
   properties {}
  }
  MaterialExpressionStaticSwitch_10 {
   class=MaterialExpressionStaticSwitch
   inputs [
    True {from=MaterialExpressionNamedRerouteUsage_107, output=""}
    False {from=MaterialExpressionNamedRerouteUsage_106, output=""}
    Value {from=MaterialExpressionNamedRerouteUsage_108, output=""}
   ]
   properties {}
  }
  MaterialExpressionStaticSwitch_12 {
   class=MaterialExpressionStaticSwitch
   inputs [
    True {from=MaterialExpressionNamedRerouteUsage_109, output=""}
    False {from=MaterialExpressionNamedRerouteUsage_110, output=""}
    Value {from=MaterialExpressionNamedRerouteUsage_111, output=""}
   ]
   properties {}
  }
  MaterialExpressionStaticSwitch_13 {
   class=MaterialExpressionStaticSwitch
   inputs [
    True {from=MaterialExpressionConstant_13, output=""}
    False {from=MaterialExpressionIf_3, output=""}
    Value {from=MaterialExpressionNamedRerouteUsage_123, output=""}
   ]
   properties {}
  }
  MaterialExpressionStaticSwitch_14 {
   class=MaterialExpressionStaticSwitch
   inputs [
    True {from=MaterialExpressionLinearInterpolate_35, output=""}
    False {from=MaterialExpressionTextureSample_7, output=RGB}
    Value {from=MaterialExpressionNamedRerouteUsage_139, output=""}
   ]
   properties {}
  }
  MaterialExpressionStaticSwitch_15 {
   class=MaterialExpressionStaticSwitch
   inputs [
    True {from=MaterialExpressionLinearInterpolate_36, output=""}
    False {from=MaterialExpressionNormalize_6, output=""}
    Value {from=MaterialExpressionNamedRerouteUsage_140, output=""}
   ]
   properties {}
  }
  MaterialExpressionStaticSwitch_16 {
   class=MaterialExpressionStaticSwitch
   inputs [
    True {from=MaterialExpressionLinearInterpolate_37, output=""}
    False {from=MaterialExpressionScalarParameter_13, output=""}
    Value {from=MaterialExpressionNamedRerouteUsage_141, output=""}
   ]
   properties {}
  }
  MaterialExpressionStaticSwitch_17 {
   class=MaterialExpressionStaticSwitch
   inputs [
    True {from=MaterialExpressionLinearInterpolate_38, output=""}
    False {from=MaterialExpressionScalarParameter_24, output=""}
    Value {from=MaterialExpressionNamedRerouteUsage_142, output=""}
   ]
   properties {}
  }
  MaterialExpressionStaticSwitch_18 {
   class=MaterialExpressionStaticSwitch
   inputs [
    True {from=MaterialExpressionMultiply_53, output=""}
    False {from=MaterialExpressionScalarParameter_12, output=""}
    Value {from=MaterialExpressionNamedRerouteUsage_143, output=""}
   ]
   properties {}
  }
  MaterialExpressionStaticSwitch_19 {
   class=MaterialExpressionStaticSwitch
   inputs [
    True {from=MaterialExpressionMultiply_54, output=""}
    False {from=MaterialExpressionScalarParameter_53, output=""}
    Value {from=MaterialExpressionNamedRerouteUsage_144, output=""}
   ]
   properties {}
  }
  MaterialExpressionStaticSwitch_2 {
   class=MaterialExpressionStaticSwitch
   inputs [
    True {from=MaterialExpressionLinearInterpolate_3, output=""}
    False {from=MaterialExpressionLinearInterpolate_7, output=""}
    Value {from=MaterialExpressionNamedRerouteUsage_7, output=""}
   ]
   properties {}
  }
  MaterialExpressionStaticSwitch_20 {
   class=MaterialExpressionStaticSwitch
   inputs [
    True {from=MaterialExpressionLinearInterpolate_40, output=""}
    False {from=MaterialExpressionScalarParameter_55, output=""}
    Value {from=MaterialExpressionNamedRerouteUsage_154, output=""}
   ]
   properties {}
  }
  MaterialExpressionStaticSwitch_21 {
   class=MaterialExpressionStaticSwitch
   inputs [
    True {from=MaterialExpressionLinearInterpolate_42, output=""}
    False {from=MaterialExpressionTextureSample_10, output=RGB}
    Value {from=MaterialExpressionNamedRerouteUsage_160, output=""}
   ]
   properties {}
  }
  MaterialExpressionStaticSwitch_22 {
   class=MaterialExpressionStaticSwitch
   inputs [
    True {from=MaterialExpressionLinearInterpolate_43, output=""}
    False {from=MaterialExpressionNamedRerouteUsage_163, output=""}
    Value {from=MaterialExpressionNamedRerouteUsage_162, output=""}
   ]
   properties {}
  }
  MaterialExpressionStaticSwitch_23 {
   class=MaterialExpressionStaticSwitch
   inputs [
    True {from=MaterialExpressionLinearInterpolate_46, output=""}
    False {from=MaterialExpressionNormalize_9, output=""}
    Value {from=MaterialExpressionNamedRerouteUsage_168, output=""}
   ]
   properties {}
  }
  MaterialExpressionStaticSwitch_24 {
   class=MaterialExpressionStaticSwitch
   inputs [
    True {from=MaterialExpressionLinearInterpolate_47, output=""}
    False {from=MaterialExpressionScalarParameter_105, output=""}
    Value {from=MaterialExpressionNamedRerouteUsage_181, output=""}
   ]
   properties {}
  }
  MaterialExpressionStaticSwitch_27 {
   class=MaterialExpressionStaticSwitch
   inputs [
    True {from=MaterialExpressionLinearInterpolate_45, output=""}
    False {from=MaterialExpressionNamedRerouteDeclaration_69, output=""}
    Value {from=MaterialExpressionNamedRerouteUsage_173, output=""}
   ]
   properties {}
  }
  MaterialExpressionStep_0 {
   class=MaterialExpressionStep
   inputs [
    Y {from="", output=""}
    X {from=MaterialExpressionScalarParameter_37, output=""}
   ]
   properties {constY=0.1}
  }
  MaterialExpressionStep_4 {
   class=MaterialExpressionStep
   inputs [
    Y {from=MaterialExpressionNamedRerouteUsage_112, output=""}
    X {from=MaterialExpressionNamedRerouteDeclaration_87, output=""}
   ]
   properties {}
  }
  MaterialExpressionStep_5 {
   class=MaterialExpressionStep
   inputs [
    Y {from=MaterialExpressionScalarParameter_81, output=""}
    X {from=ShellMeshMaterialExpressionShellLayerIndex_4, output=""}
   ]
   properties {}
  }
  MaterialExpressionStep_6 {
   class=MaterialExpressionStep
   inputs [
    Y {from=MaterialExpressionSubtract_6, output=""}
    X {from=MaterialExpressionFmod_1, output=""}
   ]
   properties {}
  }
  MaterialExpressionStep_7 {
   class=MaterialExpressionStep
   inputs [
    Y {from="", output=""}
    X {from=MaterialExpressionNamedRerouteUsage_146, output=""}
   ]
   properties {}
  }
  MaterialExpressionSubtract_1 {
   class=MaterialExpressionSubtract
   inputs [
    A {from=MaterialExpressionNamedRerouteUsage_164}
    B {from=MaterialExpressionStaticSwitch_2}
   ]
  }
  MaterialExpressionSubtract_14 {
   class=MaterialExpressionSubtract
   inputs [
    A {from=MaterialExpressionMultiply_142}
    B {from=""}
   ]
   properties {constB=0.5}
  }
  MaterialExpressionSubtract_4 {
   class=MaterialExpressionSubtract
   inputs [
    A {from=MaterialExpressionNamedRerouteUsage_120}
    B {from=MaterialExpressionScalarParameter_67}
   ]
  }
  MaterialExpressionSubtract_5 {
   class=MaterialExpressionSubtract
   inputs [
    A {from=MaterialExpressionMultiply_44}
    B {from=""}
   ]
  }
  MaterialExpressionSubtract_6 {
   class=MaterialExpressionSubtract
   inputs [
    A {from=MaterialExpressionPower_16}
    B {from=""}
   ]
  }
  MaterialExpressionTextureCoordinate_0 {class=MaterialExpressionTextureCoordinate}
  MaterialExpressionTextureCoordinate_3 {class=MaterialExpressionTextureCoordinate}
  MaterialExpressionTextureCoordinate_4 {class=MaterialExpressionTextureCoordinate}
  MaterialExpressionTextureCoordinate_5 {class=MaterialExpressionTextureCoordinate}
  MaterialExpressionTextureCoordinate_7 {class=MaterialExpressionTextureCoordinate}
  MaterialExpressionTextureObjectParameter_2 {
   class=MaterialExpressionTextureObjectParameter
   inputs [
    UVs {from="", output=""}
    Apply View MipBias {from="", output=""}
   ]
   properties {
    group=Fur_Shell_02 - Strands
    parameterName=FurPattern
    samplerType=SAMPLERTYPE_Masks
    sortPriority=2
    texture=&T_WhiteMask
   }
  }
  MaterialExpressionTextureObjectParameter_3 {
   class=MaterialExpressionTextureObjectParameter
   inputs [
    UVs {from="", output=""}
    Apply View MipBias {from="", output=""}
   ]
   properties {
    group=Fur_Shell_02 - Strands
    parameterName=FurPatternNormal
    samplerType=SAMPLERTYPE_Normal
    sortPriority=10
    texture=&FlatNormalBlue
   }
  }
  MaterialExpressionTextureObjectParameter_4 {
   class=MaterialExpressionTextureObjectParameter
   inputs [
    UVs {from="", output=""}
    Apply View MipBias {from="", output=""}
   ]
   properties {
    group=Fur_Shell_02 - Strands
    parameterName=Secondary_FurPattern
    samplerType=SAMPLERTYPE_Masks
    sortPriority=2
    texture=&T_WhiteMask
   }
  }
  MaterialExpressionTextureObjectParameter_5 {
   class=MaterialExpressionTextureObjectParameter
   inputs [
    UVs {from="", output=""}
    Apply View MipBias {from="", output=""}
   ]
   properties {
    group=Fur_Shell_02 - Strands
    parameterName=Secondary_FurPatternNormal
    samplerType=SAMPLERTYPE_Normal
    sortPriority=10
    texture=&FlatNormalBlue
   }
  }
  MaterialExpressionTextureSampleParameter2D_0 {
   class=MaterialExpressionTextureSampleParameter2D
   inputs [
    UVs {from="", output=""}
    Apply View MipBias {from="", output=""}
   ]
   properties {
    desc="Map controls:\r\n\r\n- FurStyleBlendMask\r\n- FurLengthMap\r\n- FurClothingMask\r\n- FurNormalMask"
    group=Fur_Shell_01 - Base Settings
    parameterName=FurShellMap
    samplerSource=SSM_Wrap_WorldGroupSettings
    samplerType=SAMPLERTYPE_Masks
    sortPriority=0
    texture=/Game/Global/Textures/Default/Blanks/T_Blank_Mask_Green_R0G1B0A1
   }
  }
  MaterialExpressionTextureSampleParameter2D_1 {
   class=MaterialExpressionTextureSampleParameter2D
   inputs [
    UVs {from="", output=""}
    Apply View MipBias {from="", output=""}
   ]
   properties {
    desc=Flow Map is used to offset the fur UVs and Shell verts to drive the overall fur direction. Also used to inlfuence the distortion controls and as the Tangent for aniso
    group=Fur_Shell_03 - WPO
    parameterName=FurFlowMap
    samplerSource=SSM_Wrap_WorldGroupSettings
    samplerType=SAMPLERTYPE_LinearColor
    sortPriority=0
    texture=/Game/Global/Textures/T_Default_Aniso
   }
  }
  MaterialExpressionTextureSample_10 {
   class=MaterialExpressionTextureSample
   inputs [
    UVs {from=MaterialExpressionNamedRerouteDeclaration_114}
    Tex {from=MaterialExpressionNamedRerouteDeclaration_115}
    Bias {from=MaterialExpressionNamedRerouteUsage_156}
    Apply View MipBias {from=""}
   ]
   properties {mipValueMode=TMVM_MipBias, samplerSource=SSM_Wrap_WorldGroupSettings}
  }
  MaterialExpressionTextureSample_11 {
   class=MaterialExpressionTextureSample
   inputs [
    UVs {from=MaterialExpressionNamedRerouteDeclaration_117}
    Tex {from=MaterialExpressionNamedRerouteUsage_158}
    Bias {from=MaterialExpressionNamedRerouteUsage_157}
    Apply View MipBias {from=""}
   ]
   properties {mipValueMode=TMVM_MipBias, samplerSource=SSM_Wrap_WorldGroupSettings}
  }
  MaterialExpressionTextureSample_12 {
   class=MaterialExpressionTextureSample
   inputs [
    UVs {from=MaterialExpressionNamedRerouteUsage_169}
    Tex {from=MaterialExpressionNamedRerouteDeclaration_122}
    Apply View MipBias {from=""}
   ]
   properties {samplerSource=SSM_Wrap_WorldGroupSettings, samplerType=SAMPLERTYPE_Normal}
  }
  MaterialExpressionTextureSample_13 {
   class=MaterialExpressionTextureSample
   inputs [
    UVs {from=MaterialExpressionNamedRerouteUsage_170}
    Tex {from=MaterialExpressionNamedRerouteUsage_171}
    Apply View MipBias {from=""}
   ]
   properties {samplerSource=SSM_Wrap_WorldGroupSettings, samplerType=SAMPLERTYPE_Normal}
  }
  MaterialExpressionTextureSample_5 {
   class=MaterialExpressionTextureSample
   inputs [
    UVs {from=MaterialExpressionMultiply_43}
    Tex {from=""}
    Apply View MipBias {from=""}
   ]
   properties {samplerSource=SSM_Wrap_WorldGroupSettings, texture=/Game/Global/Textures/Default/Noise/Noise}
  }
  MaterialExpressionTextureSample_6 {
   class=MaterialExpressionTextureSample
   inputs [
    UVs {from=MaterialExpressionNamedRerouteUsage_13}
    Tex {from=MaterialExpressionNamedRerouteDeclaration_112}
    Apply View MipBias {from=""}
   ]
   properties {samplerSource=SSM_Wrap_WorldGroupSettings, samplerType=SAMPLERTYPE_Normal}
  }
  MaterialExpressionTextureSample_7 {
   class=MaterialExpressionTextureSample
   inputs [
    UVs {from=MaterialExpressionNamedRerouteDeclaration_21}
    Tex {from=MaterialExpressionNamedRerouteDeclaration_111}
    Bias {from=MaterialExpressionNamedRerouteDeclaration_100}
    Apply View MipBias {from=""}
   ]
   properties {mipValueMode=TMVM_MipBias, samplerSource=SSM_Wrap_WorldGroupSettings}
  }
  MaterialExpressionTextureSample_8 {
   class=MaterialExpressionTextureSample
   inputs [
    UVs {from=MaterialExpressionNamedRerouteDeclaration_107}
    Tex {from=MaterialExpressionNamedRerouteUsage_147}
    Bias {from=MaterialExpressionNamedRerouteUsage_145}
    Apply View MipBias {from=""}
   ]
   properties {mipValueMode=TMVM_MipBias, samplerSource=SSM_Wrap_WorldGroupSettings}
  }
  MaterialExpressionTextureSample_9 {
   class=MaterialExpressionTextureSample
   inputs [
    UVs {from=MaterialExpressionNamedRerouteUsage_132}
    Tex {from=MaterialExpressionNamedRerouteUsage_148}
    Apply View MipBias {from=""}
   ]
   properties {samplerSource=SSM_Wrap_WorldGroupSettings, samplerType=SAMPLERTYPE_Normal}
  }
  MaterialExpressionTransform_19 {class=MaterialExpressionTransform, inputs [None {from=MaterialExpressionNormalize_14}]}
  MaterialExpressionTransform_2 {
   class=MaterialExpressionTransform
   inputs [None {from=ShellMeshMaterialExpressionLocalShellOffset_1}]
   properties {transformSourceType=TRANSFORMSOURCE_Local}
  }
  MaterialExpressionTransform_3 {class=MaterialExpressionTransform, inputs [None {from=MaterialExpressionNamedRerouteUsage_41}]}
  MaterialExpressionTwoSidedSign_0 {class=MaterialExpressionTwoSidedSign, inputs [], properties {}}
  MaterialExpressionVertexInterpolator_2 {class=MaterialExpressionVertexInterpolator, inputs [VS {from=ShellMeshMaterialExpressionNormalizedShellLayer_1, output=""}], properties {}}
  MaterialExpressionVertexInterpolator_3 {class=MaterialExpressionVertexInterpolator, inputs [VS {from=ShellMeshMaterialExpressionShellLayerIndex_3, output=""}], properties {}}
  MaterialExpressionVertexNormalWS_3 {class=MaterialExpressionVertexNormalWS, inputs [], properties {}}
  MaterialExpressionWorldPosition_1 {class=MaterialExpressionWorldPosition, inputs [], properties {worldPositionShaderOffset=WPT_CameraRelative}}
  ShellMeshMaterialExpressionLocalShellOffset_1 {class=ShellMeshMaterialExpressionLocalShellOffset, inputs [], properties {}}
  ShellMeshMaterialExpressionNormalizedShellLayer_1 {class=ShellMeshMaterialExpressionNormalizedShellLayer, inputs [], properties {}}
  ShellMeshMaterialExpressionShellCount_1 {class=ShellMeshMaterialExpressionShellCount, inputs [], properties {}}
  ShellMeshMaterialExpressionShellLayerIndex_3 {class=ShellMeshMaterialExpressionShellLayerIndex, inputs [], properties {}}
  ShellMeshMaterialExpressionShellLayerIndex_4 {class=ShellMeshMaterialExpressionShellLayerIndex, inputs [], properties {}}
  ShellMeshMaterialExpressionShellLayerIndex_5 {class=ShellMeshMaterialExpressionShellLayerIndex, inputs [], properties {}}
 }}
